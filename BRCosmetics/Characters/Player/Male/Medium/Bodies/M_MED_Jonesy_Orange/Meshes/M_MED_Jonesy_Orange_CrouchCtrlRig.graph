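# Control Rig for the M_MED_Jonesy_Orange body that applies pose-driven offsets to loincloth and waist-knot sleeve bones, apparently for crouching.
# Five spherical pose readers track each thigh against spine_04 or pelvis, and their outputs weight the offsets.
# A shared Set Transforms for Bone function adds translate, Euler rotate and scale values to the bone's current local transform.
# Fixed-offset modify chains and a thigh_r angle remap sit unconnected to execution.
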ControlRigBlueprint /BRCosmetics/Characters/Player/Male/Medium/Bodies/M_MED_Jonesy_Orange/Meshes/M_MED_Jonesy_Orange_CrouchCtrlRig
graphs [
 ContainedGraph {nodes [
   n1 = RigVMFunctionEntryNode()
   RigUnit_GetTransform = &RigUnit_GetTransform(Space=LocalSpace, bInitial=false, Item=@n1.Joint)
   n2 = &RigVMFunction_MathQuaternionToEuler(RotationOrder=ZYX, Value=@RigUnit_GetTransform.Transform.Rotation)
   > RigUnit_SetTransform = &RigUnit_SetTransform(
    Space=LocalSpace
    Value=(Rotation=(X=0.0132,Y=0.4099,Z=-0.0222,W=0.9118),Translation=(X=23.6045,Y=-6.37,Z=140.1115),Scale3D=(X=1,Y=1,Z=1))
    bInitial=false
    Item=@n1.Joint
    Weight=@n1.Weight
    Value.Translation.Z=&RigVMFunction_MathFloatAdd(B=@n1.Translate Z, A=@RigUnit_GetTransform.Transform.Translation.Z) {template="Add::Execute(in A,in B,out Result)"}.Result
    Value.Scale3D.Z=&RigVMFunction_MathFloatAdd(B=@n1.Scale Z, A=@RigUnit_GetTransform.Transform.Scale3D.Z) {template="Add::Execute(in A,in B,out Result)"}.Result
    Value.Translation.Y=&RigVMFunction_MathFloatAdd(B=@n1.Translate Y, A=@RigUnit_GetTransform.Transform.Translation.Y) {template="Add::Execute(in A,in B,out Result)"}.Result
    Value.Scale3D.Y=&RigVMFunction_MathFloatAdd(B=@n1.Scale Y, A=@RigUnit_GetTransform.Transform.Scale3D.Y) {template="Add::Execute(in A,in B,out Result)"}.Result
    Value.Translation.X=&RigVMFunction_MathFloatAdd(B=@n1.Translate X, A=@RigUnit_GetTransform.Transform.Translation.X) {template="Add::Execute(in A,in B,out Result)"}.Result
    Value.Scale3D.X=&RigVMFunction_MathFloatAdd(B=@n1.Scale X, A=@RigUnit_GetTransform.Transform.Scale3D.X) {template="Add::Execute(in A,in B,out Result)"}.Result
    Value.Rotation=&RigVMFunction_MathQuaternionFromEuler(
     Euler=(X=0,Y=0,Z=0)
     Euler.X=&RigVMFunction_MathFloatAdd(B=@n1.Rotate X, A=@n2.Result.X) {template="Add::Execute(in A,in B,out Result)"}.Result
     Euler.Y=&RigVMFunction_MathFloatAdd(B=@n1.Rotate Y, A=@n2.Result.Y) {template="Add::Execute(in A,in B,out Result)"}.Result
     Euler.Z=&RigVMFunction_MathFloatAdd(B=@n1.Rotate Z, A=@n2.Result.Z) {template="Add::Execute(in A,in B,out Result)"}.Result
    ).Result
   )
   > RigVMFunctionReturnNode()
  ]}
 RigVMFunctionLibrary {nodes [Set Transforms for Bone = RigVMCollapseNode(
    ExecuteContext=()
    Joint=()
    Rotate X=0
    Rotate Y=0
    Rotate Z=0
    Scale X=0
    Scale Y=0
    Scale Z=0
    Translate X=0
    Translate Y=0
    Translate Z=0
    Weight=1
   )]}
 RigVMModel {nodes [
   /Script/ControlRig.RigUnit_ModifyBoneTransforms(
    BoneToModify=((Bone="dyn_loincloth_knot_MSTR",Transform=(Rotation=(X=0,Y=0,Z=0,W=1),Translation=(X=0,Y=3,Z=0),Scale3D=(X=1,Y=1,Z=1))),(Transform=(Rotation=(X=0,Y=0,Z=0,W=1),Translation=(X=0,Y=0,Z=0),Scale3D=(X=1,Y=1,Z=1))))
    ExecutePin=()
    Mode=AdditiveLocal
    Weight=1
    WeightMaximum=1
    WeightMinimum=0
    BoneToModify.0.Transform.Rotation=&RigVMFunction_MathQuaternionFromEuler(Euler=(X=0,Y=0,Z=10)).Result
   ) {is=struct RigVMUnitNode}
   > &RigUnit_ModifyBoneTransforms(
    BoneToModify=((Bone="dyn_sleeve_r_MSTR",Transform=(Rotation=(X=0,Y=0,Z=0,W=1),Translation=(X=0,Y=0,Z=0),Scale3D=(X=1,Y=1,Z=1))),(Transform=(Rotation=(X=0,Y=0,Z=0,W=1),Translation=(X=0,Y=0,Z=0),Scale3D=(X=1,Y=1,Z=1))))
    Mode=AdditiveLocal
    Weight=1
    WeightMaximum=1
    WeightMinimum=0
    BoneToModify.0.Transform.Rotation=&RigVMFunction_MathQuaternionFromEuler(Euler=(X=0,Y=0,Z=0)).Result
   )
   > &RigUnit_ModifyBoneTransforms(
    BoneToModify=((Bone="dyn_sleeve_l_MSTR",Transform=(Rotation=(X=0,Y=0,Z=0,W=1),Translation=(X=-2,Y=2,Z=3),Scale3D=(X=1,Y=1,Z=1))),(Transform=(Rotation=(X=0,Y=0,Z=0,W=1),Translation=(X=0,Y=0,Z=0),Scale3D=(X=1,Y=1,Z=1))))
    Mode=AdditiveLocal
    Weight=1
    WeightMaximum=1
    WeightMinimum=0
    BoneToModify.0.Transform.Rotation=&RigVMFunction_MathQuaternionFromEuler(Euler=(X=0,Y=0,Z=-30)).Result
   )
   &RigUnit_ModifyTransforms(
    ExecutePin=()
    ItemToModify=((Item=(Type=Bone,Name="dyn_loincloth_fr_l_MSTR"),Transform=(Rotation=(X=0,Y=0,Z=0,W=1),Translation=(X=-10,Y=-3,Z=10),Scale3D=(X=1,Y=1,Z=1))))
    Weight=1
    ItemToModify.0.Transform.Rotation=&RigVMFunction_MathQuaternionFromEuler(Euler=(X=20,Y=0,Z=-10)).Result
   )
   > &RigUnit_ModifyTransforms(
    ItemToModify=((Item=(Type=Bone,Name="dyn_loincloth_fr_l_02"),Transform=(Rotation=(X=0,Y=0,Z=0,W=1),Translation=(X=0,Y=0,Z=0),Scale3D=(X=1,Y=1,Z=1))))
    Weight=1
    ItemToModify.0.Transform.Rotation=&RigVMFunction_MathQuaternionFromEuler(Euler=(X=5,Y=-10,Z=0)).Result
   )
   > &RigUnit_ModifyTransforms(
    ItemToModify=((Item=(Type=Bone,Name="dyn_loincloth_fr_r_MSTR"),Transform=(Rotation=(X=0,Y=0,Z=0,W=1),Translation=(X=-5,Y=2,Z=0),Scale3D=(X=1,Y=1,Z=1))))
    Weight=1
    ItemToModify.0.Transform.Rotation=&RigVMFunction_MathQuaternionFromEuler(Euler=(X=0,Y=0,Z=0)).Result
   )
   > &RigUnit_ModifyTransforms(
    ItemToModify=((Item=(Type=Bone,Name="dyn_loincloth_fr_r_02"),Transform=(Rotation=(X=0,Y=0,Z=0,W=1),Translation=(X=0,Y=0,Z=0),Scale3D=(X=1,Y=1,Z=1))))
    Weight=1
    ItemToModify.0.Transform.Rotation=&RigVMFunction_MathQuaternionFromEuler(Euler=(X=0,Y=0,Z=0), RotationOrder=XYZ).Result
   )
   > &RigUnit_ModifyTransforms(
    ItemToModify=((Item=(Type=Bone,Name="dyn_loincloth_side_r_MSTR"),Transform=(Rotation=(X=0,Y=0,Z=0,W=1),Translation=(X=0,Y=0,Z=0),Scale3D=(X=1,Y=1,Z=1))))
    Weight=1
    ItemToModify.0.Transform.Rotation=&RigVMFunction_MathQuaternionFromEuler(Euler=(X=0,Y=0,Z=0)).Result
   )
   RigUnit_BeginExecution = &RigUnit_BeginExecution()
   n1 = &RigUnit_SphericalPoseReader(
    ActiveRegionScaleFactors=(PositiveWidth=2,NegativeWidth=2.5,PositiveHeight=2,NegativeHeight=2)
    ActiveRegionSize=0.05
    Debug=(bDrawDebug=false,bDraw2D=false,bDrawLocalAxes=false,DebugScale=25,DebugSegments=20,DebugThickness=0.25)
    DriverAxis=(X=-1,Y=0,Z=0)
    DriverItem=(Type=Bone,Name="thigh_l")
    FalloffRegionScaleFactors=(PositiveWidth=1,NegativeWidth=0.8,PositiveHeight=1,NegativeHeight=1)
    FalloffSize=0.5
    FlipWidthScaling=true
    OptionalParentItem=(Type=Bone,Name="spine_04")
    RotationOffset=(X=70,Y=-160,Z=-35)
    ExecutePin=@RigUnit_BeginExecution.ExecutePin
   )
   > Set Transforms for Bone_1_1_1_1_1_3 = RigVMFunctionReferenceNode(
    Joint=(Type=Bone,Name="dyn_sleeve_l_MSTR")
    Rotate X=-5
    Rotate Y=0
    Rotate Z=-10
    Scale X=0
    Scale Y=0
    Scale Z=0
    Translate X=-5
    Translate Y=5
    Translate Z=1
    Weight=@n1.OutputParam
   )
   > n2 = &RigUnit_SphericalPoseReader(
    ActiveRegionScaleFactors=(PositiveWidth=2,NegativeWidth=3,PositiveHeight=3,NegativeHeight=3)
    ActiveRegionSize=0.05
    Debug=(bDrawDebug=false,bDraw2D=false,bDrawLocalAxes=false,DebugScale=25,DebugSegments=20,DebugThickness=0.25)
    DriverAxis=(X=2,Y=0,Z=-0.2)
    DriverItem=(Type=Bone,Name="thigh_r")
    FalloffRegionScaleFactors=(PositiveWidth=2,NegativeWidth=2.3,PositiveHeight=2,NegativeHeight=2)
    FalloffSize=0.3
    FlipWidthScaling=true
    OptionalParentItem=(Type=Bone,Name="spine_04")
    RotationOffset=(X=-90,Y=0,Z=-20)
   )
   > Set Transforms for Bone_1_1_1_1_1_1_1 = RigVMFunctionReferenceNode(
    Joint=(Type=Bone,Name="dyn_sleeve_r_MSTR")
    Rotate X=0
    Rotate Y=0
    Rotate Z=0
    Scale X=0
    Scale Y=0
    Scale Z=0
    Translate X=0
    Translate Y=-5
    Translate Z=5
    Weight=@n2.OutputParam
   )
   > n3 = &RigUnit_SphericalPoseReader(
    ActiveRegionScaleFactors=(PositiveWidth=1,NegativeWidth=1,PositiveHeight=2,NegativeHeight=2.5)
    ActiveRegionSize=0.05
    Debug=(bDrawDebug=false,bDraw2D=false,bDrawLocalAxes=false,DebugScale=25,DebugSegments=20,DebugThickness=0.25)
    DriverAxis=(X=-1,Y=0,Z=0)
    DriverItem=(Type=Bone,Name="thigh_l")
    FalloffRegionScaleFactors=(PositiveWidth=0.7,NegativeWidth=0.7,PositiveHeight=1,NegativeHeight=1.1)
    FalloffSize=0.5
    FlipWidthScaling=true
    OptionalParentItem=(Type=Bone,Name="pelvis")
    RotationOffset=(X=100,Y=0,Z=25)
   )
   > Set Transforms for Bone_1_1_1_1_1 = RigVMFunctionReferenceNode(
    Joint=(Type=Bone,Name="dyn_loincloth_fr_l_MSTR")
    Rotate X=0
    Rotate Y=5
    Rotate Z=-10
    Scale X=0
    Scale Y=0
    Scale Z=0
    Translate X=0
    Translate Y=0
    Translate Z=0
    Weight=@n3.OutputParam
   )
   > Set Transforms for Bone_1_1_1_1_1_4 = RigVMFunctionReferenceNode(
    Joint=(Type=Bone,Name="dyn_loincloth_side_l_MSTR")
    Rotate X=0
    Rotate Y=0
    Rotate Z=-15
    Scale X=0
    Scale Y=0
    Scale Z=0
    Translate X=0
    Translate Y=0
    Translate Z=0
    Weight=@n3.OutputParam
   )
   > Set Transforms for Bone_1_1_1_1_1_1_3 = RigVMFunctionReferenceNode(
    Joint=(Type=Bone,Name="dyn_loincloth_bk_l_MSTR")
    Rotate X=0
    Rotate Y=0
    Rotate Z=5
    Scale X=0
    Scale Y=0
    Scale Z=0
    Translate X=0
    Translate Y=0
    Translate Z=2
    Weight=@n3.OutputParam
   )
   > n4 = &RigUnit_SphericalPoseReader(
    ActiveRegionScaleFactors=(PositiveWidth=0.5,NegativeWidth=3,PositiveHeight=3.5,NegativeHeight=1.2)
    ActiveRegionSize=0.05
    Debug=(bDrawDebug=false,bDraw2D=false,bDrawLocalAxes=false,DebugScale=25,DebugSegments=20,DebugThickness=0.25)
    DriverAxis=(X=2,Y=0,Z=-0.2)
    DriverItem=(Type=Bone,Name="thigh_r")
    FalloffRegionScaleFactors=(PositiveWidth=1,NegativeWidth=3,PositiveHeight=3,NegativeHeight=2)
    FalloffSize=0.3
    FlipHeightScaling=false
    FlipWidthScaling=false
    OptionalParentItem=(Type=Bone,Name="pelvis")
    RotationOffset=(X=-90,Y=0,Z=-20)
   )
   > Set Transforms for Bone_1_1_1_1_1_1 = RigVMFunctionReferenceNode(
    Joint=(Type=Bone,Name="dyn_loincloth_fr_r_MSTR")
    Rotate X=20
    Rotate Y=0
    Rotate Z=10
    Scale X=0
    Scale Y=0
    Scale Z=0
    Translate X=3
    Translate Y=0
    Translate Z=1
    Weight=@n4.OutputParam
   )
   > Set Transforms for Bone_1_1_1_1_1_1_2 = RigVMFunctionReferenceNode(
    Joint=(Type=Bone,Name="dyn_loincloth_side_r_MSTR")
    Rotate X=0
    Rotate Y=0
    Rotate Z=0
    Scale X=0
    Scale Y=0
    Scale Z=0
    Translate X=0
    Translate Y=0
    Translate Z=1.5
    Weight=@n4.OutputParam
   )
   > n5 = &RigUnit_SphericalPoseReader(
    ActiveRegionScaleFactors=(PositiveWidth=1,NegativeWidth=1,PositiveHeight=2,NegativeHeight=2.5)
    ActiveRegionSize=0.05
    Debug=(bDrawDebug=false,bDraw2D=false,bDrawLocalAxes=false,DebugScale=25,DebugSegments=20,DebugThickness=0.25)
    DriverAxis=(X=-1,Y=0,Z=0)
    DriverItem=(Type=Bone,Name="thigh_l")
    FalloffRegionScaleFactors=(PositiveWidth=0.3,NegativeWidth=0.3,PositiveHeight=0.5,NegativeHeight=0.5)
    FalloffSize=0.5
    FlipHeightScaling=false
    FlipWidthScaling=true
    OptionalParentItem=(Type=Bone,Name="pelvis")
    RotationOffset=(X=145,Y=0,Z=25)
   )
   > Set Transforms for Bone_1_1_1_1_1_2 = RigVMFunctionReferenceNode(
    Joint=(Type=Bone,Name="dyn_loincloth_fr_l_02")
    Rotate X=0
    Rotate Y=0
    Rotate Z=0
    Scale X=0
    Scale Y=0
    Scale Z=0
    Translate X=0
    Translate Y=-5
    Translate Z=0
    Weight=@n5.OutputParam
   )
   n6 = &RigUnit_GetRelativeBoneTransform(Bone=thigh_r, Space=pelvis)
   &RigVMFunction_MathFloatRemap(
    SourceMaximum=104
    SourceMinimum=20
    TargetMinimum=0
    bClamp=true
    Value=/Script/RigVM.RigVMFunction_MathFloatAbs(Value=&RigVMFunction_MathQuaternionToEuler(RotationOrder=ZYX, Value=@n6.Transform.Rotation).Result.Z) {is=struct RigVMUnitNode, template="Absolute::Execute(in Value,out Result)"}.Result
   )
   &RigUnit_ModifyTransforms(
    ExecutePin=()
    ItemToModify=((Item=(Type=Bone,Name="dyn_loincloth_side_l_MSTR"),Transform=(Rotation=(X=0,Y=0,Z=0,W=1),Translation=(X=0,Y=0,Z=10),Scale3D=(X=1,Y=1,Z=1))))
    Weight=1
   )
   &RigVMFunction_MathQuaternionFromEuler(Euler=(X=0,Y=0,Z=20))
  ]}
]
hierarchy [
 root {loc=<0 0 0>, rot=<0 0 0 1>}
 pelvis {parent=root, loc=<0 -0.7071 93.1191>, rot=<0.0127 0.707 -0.0127 -0.707>}
 spine_01 {parent=pelvis, loc=<6.7494 -0.6388 0>, rot=<0 0 0.034 -0.9994>}
 spine_02 {parent=spine_01, loc=<7.2518 -0.035 0>, rot=<0 0 0.0238 -0.9997>}
 spine_03 {parent=spine_02, loc=<6.0104 0 0>, rot=<0 0 0.0149 -0.9999>}
 spine_04 {parent=spine_03, loc=<8.1024 0 0>, rot=<0 0 -0.0132 -0.9999>}
 spine_05 {parent=spine_04, loc=<10.0146 0 0>, rot=<0 0 -0.0015 -1>}
 clavicle_l {parent=spine_05, loc=<4.3575 -0.1466 -2.3223>, rot=<0.0465 -0.7289 -0.009 -0.6829>}
 upperarm_l {parent=clavicle_l, loc=<15.5111 0 0>, rot=<0.0226 -0.3234 -0.0088 -0.9459>}
 lowerarm_l {parent=upperarm_l, loc=<26.6514 0 0>, rot=<0.0391 -0.0099 -0.2453 -0.9686>}
 hand_l {parent=lowerarm_l, loc=<22.3491 0 0>, rot=<-0.7081 -0.0507 0.0975 -0.6975>}
 index_metacarpal_l {parent=hand_l, loc=<1.6535 0.0183 -2.96>, rot=<0.0158 -0.0124 -0.0173 -0.9996>}
 index_01_l {parent=index_metacarpal_l, loc=<8.3008 0.0357 -0.0524>, rot=<-0.0044 0.0485 0.0924 -0.9945>}
 index_02_l {parent=index_01_l, loc=<3.3587 -0.3004 0.0452>, rot=<-0.002 0.0027 0.3225 -0.9466>}
 index_03_l {parent=index_02_l, loc=<2.5578 -0.123 -0.0252>, rot=<-0.002 0.0067 0.2558 -0.9667>}
 middle_metacarpal_l {parent=hand_l, loc=<2.1856 -0.1725 -0.5984>, rot=<0.006 0.0168 -0.0329 -0.9993>}
 middle_01_l {parent=middle_metacarpal_l, loc=<7.3275 -0.2447 0.0129>, rot=<-0.0208 0.0042 0.0858 -0.9961>}
 middle_02_l {parent=middle_01_l, loc=<4.0123 0.1401 0.1175>, rot=<-0.0218 0.0258 0.3512 -0.9357>}
 middle_03_l {parent=middle_02_l, loc=<2.6077 0.0534 0.1373>, rot=<0.0082 0.0188 0.1708 -0.9851>}
 pinky_metacarpal_l {parent=hand_l, loc=<1.7616 -0.9753 3.3814>, rot=<-0.0397 0.0814 -0.069 -0.9935>}
 pinky_01_l {parent=pinky_metacarpal_l, loc=<7.1709 -0.4831 -0.1094>, rot=<-0.0654 -0.0035 0.2242 -0.9723>}
 pinky_02_l {parent=pinky_01_l, loc=<2.8274 0.0893 -0.1043>, rot=<0.0037 0.0011 0.2854 -0.9584>}
 pinky_03_l {parent=pinky_02_l, loc=<2.2046 0.0307 -0.0078>, rot=<-0.0002 0.0098 0.1445 -0.9895>}
 ring_metacarpal_l {parent=hand_l, loc=<2.2074 -0.3625 1.3638>, rot=<-0.0284 0.0722 -0.0458 -0.9959>}
 ring_01_l {parent=ring_metacarpal_l, loc=<7.0295 -0.6197 -0.1446>, rot=<-0.0489 -0.0252 0.1892 -0.9804>}
 ring_02_l {parent=ring_01_l, loc=<3.6174 0.4647 -0.1027>, rot=<0 0 0.2806 -0.9598>}
 ring_03_l {parent=ring_02_l, loc=<2.4152 0.114 0.0199>, rot=<0 0 0.1189 -0.9929>}
 thumb_01_l {parent=hand_l, loc=<1.3118 -0.3191 -2.459>, rot=<0.6546 -0.4042 0.0664 -0.6354>}
 thumb_02_l {parent=thumb_01_l, loc=<5.9336 -0.0448 0.0401>, rot=<-0.0037 0.0213 0.298 -0.9543>}
 thumb_03_l {parent=thumb_02_l, loc=<2.2223 0.3682 -0.0907>, rot=<0.0118 -0.013 0.0066 -0.9998>}
 weapon_l {parent=hand_l, loc=<-0.1974 -3.4111 0.3811>, rot=<0 0 0 1>}
 deform_wrist_dn_base_l {parent=hand_l, loc=<0 -2.2342 0>, rot=<0 0 0 1>}
 deform_wrist_dn_l {parent=deform_wrist_dn_base_l, loc=<0 0 0>, rot=<0 0 0 1>}
 deform_wrist_up_base_l {parent=hand_l, loc=<0 2.2682 0>, rot=<0 0 0 1>}
 deform_wrist_up_l {parent=deform_wrist_up_base_l, loc=<0 0 0>, rot=<0 0 0 1>}
 lowerarm_twist_01_l {parent=lowerarm_l, loc=<17.8549 0 0>, rot=<0 0 0 1>}
 dyn_cuff_l {parent=lowerarm_twist_01_l, loc=<0.7459 -0.15 0.0454>, rot=<-0.0678 -0.031 0.0104 -0.9972>}
 lowerarm_twist_02_l {parent=lowerarm_l, loc=<9.6199 0 0>, rot=<0 0 0 1>}
 deform_elbow_in_base_l {parent=lowerarm_l, loc=<0.2112 5.259 0.8274>, rot=<0 0 0 1>}
 deform_elbow_in_l {parent=deform_elbow_in_base_l, loc=<0 0 0>, rot=<0 0 0 1>}
 deform_elbow_out_base_l {parent=lowerarm_l, loc=<0 -4 0>, rot=<0 0 0 1>}
 deform_elbow_out_l {parent=deform_elbow_out_base_l, loc=<0 0 0>, rot=<0 0 0 1>}
 upperarm_twist_01_l {parent=upperarm_l, loc=<0.0041 0 0>, rot=<0 0 0 1>}
 deform_upperarm_bk_base_l {parent=upperarm_twist_01_l, loc=<1.0813 -7.2193 -1.9453>, rot=<0 0 0 1>}
 deform_upperarm_bk_l {parent=deform_upperarm_bk_base_l, loc=<0 0 0>, rot=<0 0 0 1>}
 deform_upperarm_fr_base_l {parent=upperarm_twist_01_l, loc=<-0.3691 7.0506 -2.765>, rot=<0 0 0 1>}
 deform_upperarm_fr_l {parent=deform_upperarm_fr_base_l, loc=<0 0 0>, rot=<0 0 0 1>}
 deform_upperarm_up_base_l {parent=upperarm_twist_01_l, loc=<-3.2216 -1.1391 3.3311>, rot=<0 0 0 1>}
 deform_upperarm_up_l {parent=deform_upperarm_up_base_l, loc=<0 0 0>, rot=<0 0 0 1>}
 upperarm_twist_02_l {parent=upperarm_l, loc=<13.7941 0 0>, rot=<0 0 0 1>}
 dyn_strap_top_l {parent=clavicle_l, loc=<8.5236 0.9735 7.0452>, rot=<0.0008 0.0336 -0.0252 -0.9991>}
 clavicle_r {parent=spine_05, loc=<4.3575 -0.1466 2.3223>, rot=<0.7289 0.0465 -0.6829 0.009>}
 upperarm_r {parent=clavicle_r, loc=<-15.5111 0 0>, rot=<0.0226 -0.3234 -0.0088 -0.9459>}
 lowerarm_r {parent=upperarm_r, loc=<-26.6515 0 0>, rot=<0.0391 -0.0099 -0.2453 -0.9686>}
 hand_r {parent=lowerarm_r, loc=<-22.3492 0 0>, rot=<-0.7081 -0.0507 0.0975 -0.6975>}
 index_metacarpal_r {parent=hand_r, loc=<-1.6534 -0.0181 2.9601>, rot=<0.0158 -0.0124 -0.0173 -0.9996>}
 index_01_r {parent=index_metacarpal_r, loc=<-8.3008 -0.0357 0.0524>, rot=<-0.0044 0.0485 0.0924 -0.9945>}
 index_02_r {parent=index_01_r, loc=<-3.3589 0.3003 -0.0453>, rot=<-0.002 0.0027 0.3225 -0.9466>}
 index_03_r {parent=index_02_r, loc=<-2.5578 0.1231 0.0251>, rot=<-0.002 0.0067 0.2558 -0.9667>}
 middle_metacarpal_r {parent=hand_r, loc=<-2.1856 0.1726 0.5984>, rot=<0.006 0.0168 -0.0329 -0.9993>}
 middle_01_r {parent=middle_metacarpal_r, loc=<-7.3273 0.2447 -0.0129>, rot=<-0.0208 0.0042 0.0858 -0.9961>}
 middle_02_r {parent=middle_01_r, loc=<-4.0142 -0.1411 -0.07>, rot=<-0.0217 0.0259 0.3512 -0.9357>}
 middle_03_r {parent=middle_02_r, loc=<-2.6074 -0.0534 -0.1312>, rot=<0.0083 0.0188 0.1708 -0.9851>}
 pinky_metacarpal_r {parent=hand_r, loc=<-1.7617 0.9753 -3.3814>, rot=<-0.0397 0.0814 -0.069 -0.9935>}
 pinky_01_r {parent=pinky_metacarpal_r, loc=<-7.1707 0.4832 0.1094>, rot=<-0.0654 -0.0035 0.2242 -0.9723>}
 pinky_02_r {parent=pinky_01_r, loc=<-2.8274 -0.0892 0.1043>, rot=<0.0037 0.0011 0.2854 -0.9584>}
 pinky_03_r {parent=pinky_02_r, loc=<-2.2045 -0.031 0.0078>, rot=<-0.0002 0.0098 0.1445 -0.9895>}
 ring_metacarpal_r {parent=hand_r, loc=<-2.2073 0.3627 -1.3638>, rot=<-0.0284 0.0722 -0.0458 -0.9959>}
 ring_01_r {parent=ring_metacarpal_r, loc=<-7.0297 0.6197 0.1446>, rot=<-0.0489 -0.0252 0.1892 -0.9804>}
 ring_02_r {parent=ring_01_r, loc=<-3.6171 -0.4648 0.1027>, rot=<0 0 0.2806 -0.9598>}
 ring_03_r {parent=ring_02_r, loc=<-2.4153 -0.1141 -0.0199>, rot=<0 0 0.1189 -0.9929>}
 thumb_01_r {parent=hand_r, loc=<-1.3119 0.3192 2.459>, rot=<0.6546 -0.4042 0.0664 -0.6354>}
 thumb_02_r {parent=thumb_01_r, loc=<-5.9335 0.0447 -0.0401>, rot=<-0.0037 0.0213 0.298 -0.9543>}
 thumb_03_r {parent=thumb_02_r, loc=<-2.2223 -0.3682 0.0907>, rot=<0.0118 -0.013 0.0066 -0.9998>}
 weapon_r {parent=hand_r, loc=<0.1975 3.4112 -0.3811>, rot=<0 0 0 -1>}
 deform_wrist_dn_base_r {parent=hand_r, loc=<-0.0003 2.2346 -0.0001>, rot=<0 0 0 1>}
 deform_wrist_dn_r {parent=deform_wrist_dn_base_r, loc=<0 0 0>, rot=<0 0 0 1>}
 deform_wrist_up_base_r {parent=hand_r, loc=<0.0004 -2.2683 0.0001>, rot=<0 0 0 1>}
 deform_wrist_up_r {parent=deform_wrist_up_base_r, loc=<0 0 0>, rot=<0 0 0 1>}
 lowerarm_twist_01_r {parent=lowerarm_r, loc=<-17.855 0 0>, rot=<0 0 0 1>}
 dyn_cuff_r {parent=lowerarm_twist_01_r, loc=<-0.758 0.023 -0.206>, rot=<0.031 -0.0678 0.9972 0.0104>}
 lowerarm_twist_02_r {parent=lowerarm_r, loc=<-9.6199 0 0>, rot=<0 0 0 1>}
 deform_elbow_in_base_r {parent=lowerarm_r, loc=<-0.2111 -5.2591 -0.8275>, rot=<0 0 0 1>}
 deform_elbow_in_r {parent=deform_elbow_in_base_r, loc=<-0.0001 0 0.0002>, rot=<0 0 0 1>}
 deform_elbow_out_base_r {parent=lowerarm_r, loc=<0 4 0>, rot=<0 0 0 1>}
 deform_elbow_out_r {parent=deform_elbow_out_base_r, loc=<0 0 0>, rot=<0 0 0 1>}
 upperarm_twist_01_r {parent=upperarm_r, loc=<-0.0041 0 0>, rot=<0 0 0 1>}
 deform_upperarm_bk_base_r {parent=upperarm_twist_01_r, loc=<-1.0815 7.2192 1.9456>, rot=<0 0 0 1>}
 deform_upperarm_bk_r {parent=deform_upperarm_bk_base_r, loc=<0 0 0>, rot=<0 0 0 1>}
 deform_upperarm_fr_base_r {parent=upperarm_twist_01_r, loc=<0.3689 -7.0506 2.7652>, rot=<0 0 0 1>}
 deform_upperarm_fr_r {parent=deform_upperarm_fr_base_r, loc=<0 0 0>, rot=<0 0 0 1>}
 deform_upperarm_up_base_r {parent=upperarm_twist_01_r, loc=<3.2213 1.1391 -3.3308>, rot=<0 0 0 1>}
 deform_upperarm_up_r {parent=deform_upperarm_up_base_r, loc=<0 0 0>, rot=<0 0 0 -1>}
 upperarm_twist_02_r {parent=upperarm_r, loc=<-13.7942 0 0>, rot=<0 0 0 1>}
 dyn_strap_top_r {parent=clavicle_r, loc=<-5.6594 -1.6494 -11.7117>, rot=<0 -0.0436 0 -0.999>, scale=<1 1.7488 1>}
 neck_01 {parent=spine_05, loc=<11.3138 0.1529 0>, rot=<0 0 -0.2003 -0.9797>}
 neck_02 {parent=neck_01, loc=<4.5627 0 0>, rot=<0 0 0.0504 -0.9987>}
 head {parent=neck_02, loc=<5.3914 0 0>, rot=<0 0 0.115 -0.9934>}
 deform_pec_aim_base_l {parent=spine_04, loc=<7.8732 12.6388 -2.4123>, rot=<0.0102 -0.4602 0.3979 -0.7936>}
 deform_pec_aim_l {parent=deform_pec_aim_base_l, loc=<0 0 0>, rot=<0 0 0 1>}
 deform_pec_base_l {parent=deform_pec_aim_l, loc=<10.766 0 0>, rot=<0 0 0 1>}
 deform_pec_l {parent=deform_pec_base_l, loc=<0.055 0 0>, rot=<0 0 0 1>}
 deform_pec_aim_base_r {parent=spine_04, loc=<7.8727 12.6388 2.4123>, rot=<-0.4602 -0.0102 0.7936 0.3979>}
 deform_pec_aim_r {parent=deform_pec_aim_base_r, loc=<0 0 0>, rot=<0 0 0 1>}
 deform_pec_base_r {parent=deform_pec_aim_r, loc=<-10.7662 -0.0006 -0.0007>, rot=<0 0 0 1>}
 deform_pec_r {parent=deform_pec_base_r, loc=<0 0 0>, rot=<0 0 0 1>}
 dyn_loincloth_knot_MSTR {parent=spine_01, loc=<-1.019 11.1658 -2.3454>, rot=<-0.0043 -0.7486 -0.6629 -0.0126>}
 dyn_sleeve_l_MSTR {parent=dyn_loincloth_knot_MSTR, loc=<0.0301 -3.6629 0.3994>, rot=<-0.0862 0.0155 0.1461 -0.9854>}
 dyn_sleeve_l_01 {parent=dyn_sleeve_l_MSTR, loc=<0.3 0 0>, rot=<0 0 0 1>}
 dyn_sleeve_l_02 {parent=dyn_sleeve_l_01, loc=<7.3061 -0.1968 0.3379>, rot=<0.0862 -0.0155 -0.1461 -0.9854>}
 dyn_sleeve_l_03 {parent=dyn_sleeve_l_02, loc=<6.876 0 0>, rot=<0 0 0 1>}
 dyn_sleeve_r_MSTR {parent=dyn_loincloth_knot_MSTR, loc=<0.9598 5.2271 -1.1324>, rot=<0.0396 -0.0851 -0.053 -0.9942>}
 dyn_sleeve_r_01 {parent=dyn_sleeve_r_MSTR, loc=<0.2999 0 0>, rot=<0 0 0 1>}
 dyn_sleeve_r_02 {parent=dyn_sleeve_r_01, loc=<9.1119 0 0.6953>, rot=<0 0.0978 0 -0.9952>}
 dyn_sleeve_r_03 {parent=dyn_sleeve_r_02, loc=<7.6753 0 0>, rot=<-0.0502 0.0034 0.0673 -0.9965>}
 dyn_loincloth_fr_l_MSTR {parent=spine_01, loc=<-0.1683 8.1177 -9.5147>, rot=<0.1035 -0.9423 -0.2968 -0.1154>}
 dyn_loincloth_fr_l_01 {parent=dyn_loincloth_fr_l_MSTR, loc=<0.2999 0 0>, rot=<0 0 0 1>}
 dyn_loincloth_fr_l_02 {parent=dyn_loincloth_fr_l_01, loc=<14.4162 0 0.6891>, rot=<0 -0.0031 0 -1>}
 dyn_loincloth_fr_l_03 {parent=dyn_loincloth_fr_l_02, loc=<12.1274 0 0>, rot=<-0.206 -0.0908 0.0848 -0.9706>}
 dyn_loincloth_fr_l_04 {parent=dyn_loincloth_fr_l_03, loc=<11.3067 0 0>, rot=<0 -0.0674 0 -0.9977>}
 dyn_zipper_01 {parent=dyn_loincloth_fr_l_04, loc=<9.1698 2.8447 0.3985>, rot=<0.3587 0.0647 -0.159 -0.9175>}
 dyn_zipper_02 {parent=dyn_zipper_01, loc=<0 0 0>, rot=<0 0 0 1>}
 dyn_loincloth_side_l_MSTR {parent=spine_01, loc=<1.6881 -1.0555 -13.1991>, rot=<0.0703 -0.995 0.0389 -0.06>}
 dyn_loincloth_side_l_01 {parent=dyn_loincloth_side_l_MSTR, loc=<0.2999 0 0>, rot=<0 0 0 1>}
 dyn_loincloth_side_l_02 {parent=dyn_loincloth_side_l_01, loc=<16.7058 0 0.6833>, rot=<0 -0.0008 0 -1>}
 dyn_loincloth_side_l_03 {parent=dyn_loincloth_side_l_02, loc=<12.9135 0 0>, rot=<0 0 0 1>}
 dyn_loincloth_bk_l_MSTR {parent=spine_01, loc=<2.1407 -8.5326 -8.9547>, rot=<0.0437 -0.8954 0.4402 -0.0505>}
 dyn_loincloth_bk_l_01 {parent=dyn_loincloth_bk_l_MSTR, loc=<0.3 0 0>, rot=<0 0 0 1>}
 dyn_loincloth_bk_l_02 {parent=dyn_loincloth_bk_l_01, loc=<16.0198 0 0.7799>, rot=<0 0 0 1>}
 dyn_loincloth_bk_l_03 {parent=dyn_loincloth_bk_l_02, loc=<11.2348 0 0>, rot=<0 0 0 1>}
 dyn_loincloth_fr_r_MSTR {parent=spine_01, loc=<-0.2708 8.2646 9.2542>, rot=<0.122 0.3752 0.9159 -0.0742>}
 dyn_loincloth_fr_r_01 {parent=dyn_loincloth_fr_r_MSTR, loc=<0.2999 0 0>, rot=<0 0 0 1>}
 dyn_loincloth_fr_r_02 {parent=dyn_loincloth_fr_r_01, loc=<13.3582 -0.0668 0.6052>, rot=<-0.055 -0.0333 -0.0018 -0.9979>}
 dyn_loincloth_fr_r_03 {parent=dyn_loincloth_fr_r_02, loc=<9.8721 -0.3719 0.2283>, rot=<0.152 -0.0595 -0.1424 -0.9762>}
 dyn_loincloth_fr_r_04 {parent=dyn_loincloth_fr_r_03, loc=<11.688 0 0>, rot=<0 0 0 1>}
 dyn_loincloth_side_r_MSTR {parent=spine_01, loc=<1.1879 -0.8852 13.2946>, rot=<0.0517 -0.0554 0.9953 -0.0609>}
 dyn_loincloth_side_r_01 {parent=dyn_loincloth_side_r_MSTR, loc=<0.2999 0 0>, rot=<0 0 0 1>}
 dyn_loincloth_side_r_02 {parent=dyn_loincloth_side_r_01, loc=<15.4451 0 0.6853>, rot=<0 -0.0008 0 -1>}
 dyn_loincloth_side_r_03 {parent=dyn_loincloth_side_r_02, loc=<12.9135 0 0>, rot=<0 0 0 1>}
 dyn_loincloth_bk_r_MSTR {parent=spine_01, loc=<1.2162 -9.2117 8.9302>, rot=<0.053 -0.4854 0.8718 -0.0386>}
 dyn_loincloth_bk_r_01 {parent=dyn_loincloth_bk_r_MSTR, loc=<0.3 0 0>, rot=<0 0 0 1>}
 dyn_loincloth_bk_r_02 {parent=dyn_loincloth_bk_r_01, loc=<14.1907 0 0.5376>, rot=<0 0.0122 0 -0.9999>}
 dyn_loincloth_bk_r_03 {parent=dyn_loincloth_bk_r_02, loc=<11.4092 0 0>, rot=<0 0 0 1>}
 dyn_loincloth_bk_MSTR {parent=spine_01, loc=<1.7256 -10.913 0>, rot=<0.0524 -0.7052 0.7052 -0.0524>}
 dyn_loincloth_bk_01 {parent=dyn_loincloth_bk_MSTR, loc=<0.3 0 0>, rot=<0 0 0 1>}
 dyn_loincloth_bk_02 {parent=dyn_loincloth_bk_01, loc=<14.513 0 0.2801>, rot=<0 -0.0628 0 -0.998>}
 dyn_loincloth_bk_03 {parent=dyn_loincloth_bk_02, loc=<11.2348 0 0>, rot=<0 0 0 1>}
 thigh_l {parent=pelvis, loc=<-2.5693 -0.0109 -7.7501>, rot=<0.0012 0.0572 0.0155 -0.9982>}
 calf_l {parent=thigh_l, loc=<-39.0484 0 0>, rot=<0.0006 -0.0161 -0.0391 -0.9991>}
 calf_twist_01_l {parent=calf_l, loc=<-35.813 -0.0967 0.0397>, rot=<0.0237 0.0095 -0.0035 -0.9997>}
 calf_twist_02_l {parent=calf_l, loc=<-18.8064 -0.0967 0.0397>, rot=<0.0237 0.0095 -0.0035 -0.9997>}
 foot_l {parent=calf_l, loc=<-42.561 0.1077 0.4292>, rot=<0.0418 -0.0455 0.0343 -0.9975>}
 ball_l {parent=foot_l, loc=<-6.1599 13.8264 0.6806>, rot=<0.0001 -0.0001 -0.7063 -0.7079>}
 dyn_shoe_knot_l {parent=foot_l, loc=<5.7087 5.6569 0.0278>, rot=<-0.0357 -0.7045 -0.0284 -0.7082>}
 dyn_shoetie_l_out_MSTR {parent=dyn_shoe_knot_l, loc=<0.5465 0.6033 -0.1773>, rot=<0.0267 -0.2055 0.1261 -0.9701>}
 dyn_shoetie_l_out_01 {parent=dyn_shoetie_l_out_MSTR, loc=<0 0 0>, rot=<0.0063 -0.0959 -0.1909 -0.9769>}
 dyn_shoetie_l_out_02 {parent=dyn_shoetie_l_out_01, loc=<6.1448 -0.1074 0.0041>, rot=<-0.0002 -0.001 -0.0271 -0.9996>}
 dyn_shoelace_l_out_MSTR {parent=dyn_shoe_knot_l, loc=<0.5374 1.0069 -1.0542>, rot=<-0.0735 -0.2663 -0.2558 -0.9264>}
 dyn_shoelace_l_out_01 {parent=dyn_shoelace_l_out_MSTR, loc=<0 0 0>, rot=<0 0 0 1>}
 dyn_shoelace_l_out_02 {parent=dyn_shoelace_l_out_01, loc=<4.5984 -0.489 0.1578>, rot=<-0.0391 -0.0902 0.1289 -0.9868>}
 dyn_shoelace_l_out_03 {parent=dyn_shoelace_l_out_02, loc=<3.9639 0 0>, rot=<-0.034 -0.1011 -0.045 -0.9933>}
 dyn_shoelace_l_out_04 {parent=dyn_shoelace_l_out_03, loc=<6.2162 0.2913 -0.5074>, rot=<0.0091 0.021 -0.0379 -0.999>}
 dyn_shoetie_l_in_MSTR {parent=dyn_shoe_knot_l, loc=<-0.7484 0.7979 -0.1337>, rot=<-0.1159 -0.95 -0.0351 -0.2878>}
 dyn_shoetie_l_in_01 {parent=dyn_shoetie_l_in_MSTR, loc=<0 0 0>, rot=<0.0334 0.1342 -0.0008 -0.9904>}
 dyn_shoetie_l_in_02 {parent=dyn_shoetie_l_in_01, loc=<2.672 0.0495 -0.0756>, rot=<0.0067 0.0295 0.0196 -0.9994>}
 dyn_shoelace_l_in_MSTR {parent=dyn_shoe_knot_l, loc=<-0.7954 1.4735 -1.0475>, rot=<-0.2661 -0.8638 -0.1259 -0.4089>}
 dyn_shoelace_l_in_01 {parent=dyn_shoelace_l_in_MSTR, loc=<0 0 0>, rot=<0 0 0 1>}
 dyn_shoelace_l_in_02 {parent=dyn_shoelace_l_in_01, loc=<5.4566 -0.5221 0.2838>, rot=<0.0413 0.0914 0.1714 -0.9801>}
 dyn_shoelace_l_in_03 {parent=dyn_shoelace_l_in_02, loc=<3.9986 -0.1033 0.1328>, rot=<0.0086 0.0445 0.0596 -0.9972>}
 deform_knee_in_base_l {parent=calf_l, loc=<1.298 -5.2957 -0.0496>, rot=<0 0 0 1>}
 deform_knee_in_l {parent=deform_knee_in_base_l, loc=<0 0 0>, rot=<0 0 0 1>}
 deform_knee_out_base_l {parent=calf_l, loc=<0 4 0>, rot=<0 0 0 1>}
 deform_knee_out_l {parent=deform_knee_out_base_l, loc=<0 0 0>, rot=<0 0 0 1>}
 thigh_twist_01_l {parent=thigh_l, loc=<-21.6303 0 0>, rot=<0.0013 0 0 -1>}
 deform_glute_base_l {parent=thigh_l, loc=<-6 -9 0>, rot=<0 0 -0.4617 -0.887>}
 deform_glute_l {parent=deform_glute_base_l, loc=<0 0 0>, rot=<0 0 0 1>}
 deform_groin_base_l {parent=thigh_l, loc=<-2.1908 8.1142 -1.7442>, rot=<0 0 0 1>}
 deform_groin_l {parent=deform_groin_base_l, loc=<0 0 0>, rot=<0 0 0 1>}
 thigh_r {parent=pelvis, loc=<-2.5691 -0.0109 7.7501>, rot=<0.0572 -0.0012 0.9982 0.0155>}
 calf_r {parent=thigh_r, loc=<39.0487 0 0>, rot=<0.0006 -0.0161 -0.0391 -0.9991>}
 calf_twist_01_r {parent=calf_r, loc=<35.8132 0.0967 -0.0397>, rot=<0.0237 0.0095 -0.0035 -0.9997>}
 calf_twist_02_r {parent=calf_r, loc=<18.8064 0.0967 -0.0397>, rot=<0.0237 0.0095 -0.0035 -0.9997>}
 foot_r {parent=calf_r, loc=<42.5612 -0.1077 -0.4292>, rot=<0.0418 -0.0455 0.0343 -0.9975>}
 ball_r {parent=foot_r, loc=<6.1598 -13.8264 -0.6806>, rot=<0.0001 -0.0001 -0.7063 -0.7079>}
 dyn_shoe_knot_r {parent=foot_r, loc=<-5.7089 -5.6569 -0.0278>, rot=<0.7082 0.0284 -0.7045 -0.0357>}
 dyn_shoetie_r_out_MSTR {parent=dyn_shoe_knot_r, loc=<-0.5465 0.6033 -0.1773>, rot=<0.0267 -0.2055 0.1261 -0.9701>}
 dyn_shoetie_r_out_01 {parent=dyn_shoetie_r_out_MSTR, loc=<0 0 0>, rot=<0.0063 -0.0959 -0.1909 -0.9769>}
 dyn_shoetie_r_out_02 {parent=dyn_shoetie_r_out_01, loc=<-2.001 0.9372 -5.7348>, rot=<-0.0002 -0.001 -0.0271 -0.9996>}
 dyn_shoelace_r_out_MSTR {parent=dyn_shoe_knot_r, loc=<-0.5374 1.0069 -1.0541>, rot=<-0.0735 -0.2663 -0.2558 -0.9264>}
 dyn_shoelace_r_out_01 {parent=dyn_shoelace_r_out_MSTR, loc=<0 0 0>, rot=<-0.3036 -0.6386 -0.317 -0.6321>}
 dyn_shoelace_r_out_02 {parent=dyn_shoelace_r_out_01, loc=<4.4028 -1.271 -0.6404>, rot=<-0.0565 -0.0673 0.1973 -0.9764>}
 dyn_shoelace_r_out_03 {parent=dyn_shoelace_r_out_02, loc=<3.9043 -0.5408 0.4203>, rot=<-0.0478 0.1099 0.0599 -0.991>}
 dyn_shoelace_r_out_04 {parent=dyn_shoelace_r_out_03, loc=<6.109 0.0163 1.2898>, rot=<0.0119 0.1012 -0.1559 -0.9825>}
 dyn_shoetie_r_in_MSTR {parent=dyn_shoe_knot_r, loc=<0.7484 0.7979 -0.1337>, rot=<-0.1159 -0.95 -0.0351 -0.2878>}
 dyn_shoetie_r_in_01 {parent=dyn_shoetie_r_in_MSTR, loc=<0 0 0>, rot=<0.0334 0.1342 -0.0008 -0.9904>}
 dyn_shoetie_r_in_02 {parent=dyn_shoetie_r_in_01, loc=<0.4966 0.5924 2.5594>, rot=<0.0067 0.0295 0.0196 -0.9994>}
 dyn_shoelace_r_in_MSTR {parent=dyn_shoe_knot_r, loc=<0.7954 1.4735 -1.0474>, rot=<-0.2661 -0.8638 -0.1259 -0.4089>}
 dyn_shoelace_r_in_01 {parent=dyn_shoelace_r_in_MSTR, loc=<0 0 0>, rot=<0.7184 0.5077 0.2193 -0.4219>}
 dyn_shoelace_r_in_02 {parent=dyn_shoelace_r_in_01, loc=<5.2435 0.3039 -1.5944>, rot=<0.0753 -0.468 -0.035 -0.8798>}
 dyn_shoelace_r_in_03 {parent=dyn_shoelace_r_in_02, loc=<3.7242 0.1562 1.4571>, rot=<-0.2705 0.1018 -0.0618 -0.9553>}
 deform_knee_in_base_r {parent=calf_r, loc=<-1.2981 5.2958 0.0496>, rot=<0 0 0 1>}
 deform_knee_in_r {parent=deform_knee_in_base_r, loc=<0 0 0>, rot=<0 0 0 1>}
 deform_knee_out_base_r {parent=calf_r, loc=<0 -4 0>, rot=<0 0 0 1>}
 deform_knee_out_r {parent=deform_knee_out_base_r, loc=<0 0 0>, rot=<0 0 0 1>}
 dyn_chain_MSTR {parent=calf_r, loc=<29.1381 7.5485 -4.418>, rot=<0.9991 -0.0415 -0.0059 -0.0017>}
 dyn_chain_01 {parent=dyn_chain_MSTR, loc=<0 0 0>, rot=<0 0 0 1>}
 dyn_chain_02 {parent=dyn_chain_01, loc=<3.2209 0 0>, rot=<0 0.0178 0 -0.9998>}
 dyn_chain_03 {parent=dyn_chain_02, loc=<3.2526 0 0>, rot=<0 -0.0011 0 -1>}
 thigh_twist_01_r {parent=thigh_r, loc=<21.6303 0 0>, rot=<0.0013 0 0 -1>}
 deform_glute_base_r {parent=thigh_r, loc=<6 9 0>, rot=<0 0 -0.4617 -0.887>}
 deform_glute_r {parent=deform_glute_base_r, loc=<0 0 0>, rot=<0 0 0 1>}
 deform_groin_base_r {parent=thigh_r, loc=<2.1909 -8.1142 1.7442>, rot=<0 0 0 1>}
 deform_groin_r {parent=deform_groin_base_r, loc=<0 0 0>, rot=<0 0 0 1>}
]
variables [
 Idle {type=double, default=0}
 Left Direct {type=double, default=0}
 Right Direct {type=double, default=0}
]
