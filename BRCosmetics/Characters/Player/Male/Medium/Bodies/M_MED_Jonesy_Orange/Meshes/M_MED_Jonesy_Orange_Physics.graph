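# Physics asset for the M_MED_Jonesy_Orange body, covering dangling loincloth and sleeve bone chains plus pelvis, spine, thigh and calf capsules.
# Mostly capsules, with boxes on the lower front-right panel, form chains with limited swing and tight twist under small MSTR anchor bodies.
# Panel and knot anchors join spine_01 with free rotation, and sleeve anchors hang from the knot.
# Free-rotation cross-constraints tie front panels to neighboring side panels.
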PhysicsAsset /BRCosmetics/Characters/Player/Male/Medium/Bodies/M_MED_Jonesy_Orange/Meshes/M_MED_Jonesy_Orange_Physics
bodies {
 calf_l [{shapeType=Capsule, length=30, radius=5, center=<-8 0 0>, rotation=<rot 90 0 0>}]
 calf_r [{shapeType=Capsule, length=30, radius=5, center=<8 0 0>, rotation=<rot 90 0 0>}]
 dyn_loincloth_bk_01 [{shapeType=Capsule, length=9, radius=8, center=<8 0 6.5>, rotation=<rot 0 0 90>}]
 dyn_loincloth_bk_02 [{shapeType=Capsule, length=9, radius=8, shapeName=dyn_loincloth_bk_02_capsule, center=<8 0 6.5>, rotation=<rot 0 0 90>}]
 dyn_loincloth_bk_03 [{shapeType=Capsule, length=9, radius=8, shapeName=dyn_loincloth_bk_03_capsule, center=<8 0 6.5>, rotation=<rot 0 0 90>}]
 dyn_loincloth_bk_MSTR [{shapeType=Capsule, length=0.505, radius=0.505, shapeName=dyn_loincloth_bk_MSTR_capsule, rotation=<rot 0 -90 0>}]
 dyn_loincloth_bk_l_01 [{shapeType=Capsule, length=9, radius=8, center=<8 0 6.5>, rotation=<rot 0 0 90>}]
 dyn_loincloth_bk_l_02 [{shapeType=Capsule, length=9, radius=8, shapeName=dyn_loincloth_bk_l_02_capsule, center=<8 0 6.5>, rotation=<rot 0 0 90>}]
 dyn_loincloth_bk_l_03 [{shapeType=Capsule, length=9, radius=8, shapeName=dyn_loincloth_bk_l_03_capsule, center=<8 0 6.5>, rotation=<rot 0 0 90>}]
 dyn_loincloth_bk_l_MSTR [{shapeType=Capsule, length=0.505, radius=0.505, shapeName=dyn_loincloth_bk_l_MSTR_capsule, rotation=<rot 0 -90 0>}]
 dyn_loincloth_bk_r_01 [{shapeType=Capsule, length=9, radius=8, center=<8 0 6.5>, rotation=<rot 0 0 90>}]
 dyn_loincloth_bk_r_02 [{shapeType=Capsule, length=9, radius=8, shapeName=dyn_loincloth_bk_r_02_capsule, center=<8 0 6.5>, rotation=<rot 0 0 90>}]
 dyn_loincloth_bk_r_03 [{shapeType=Capsule, length=9, radius=8, shapeName=dyn_loincloth_bk_r_03_capsule, center=<8 0 6.5>, rotation=<rot 0 0 90>}]
 dyn_loincloth_bk_r_MSTR [{shapeType=Capsule, length=0.505, radius=0.505, shapeName=dyn_loincloth_bk_r_MSTR_capsule, rotation=<rot 0 -90 0>}]
 dyn_loincloth_fr_l_01 [{shapeType=Capsule, length=9.35, radius=3.2, shapeName=dyn_loincloth_fr_l_01_capsule, center=<2 0 4>, rotation=<rot 0 0 90>}]
 dyn_loincloth_fr_l_02 [{shapeType=Capsule, length=9.5, radius=2.5, shapeName=dyn_loincloth_fr_l_02_capsule, center=<2.2011 0 3>, rotation=<rot 90 0 0>}]
 dyn_loincloth_fr_l_03 [{shapeType=Capsule, length=8, radius=4, shapeName=dyn_loincloth_fr_l_03_capsule, center=<2.2011 0 3>, rotation=<rot 90 0 0>}]
 dyn_loincloth_fr_l_04 [{shapeType=Capsule, length=8, radius=4, shapeName=dyn_loincloth_fr_l_04_capsule, center=<2.2011 0 3>, rotation=<rot 90 0 0>}]
 dyn_loincloth_fr_l_MSTR [{shapeType=Capsule, length=0.505, radius=0.505, shapeName=dyn_loincloth_fr_l_MSTR_capsule, rotation=<rot 0 -90 0>}]
 dyn_loincloth_fr_r_01 [{shapeType=Capsule, length=9.35, radius=3.2, center=<2 0 4>, rotation=<rot 0 0 90>}]
 dyn_loincloth_fr_r_02 [{shapeType=Box, extentX=9.5, extentY=9, extentZ=1.5, shapeName=dyn_loincloth_fr_r_02_box, center=<1 0 0>}]
 dyn_loincloth_fr_r_03 [{shapeType=Box, extentX=9.5, extentY=9, extentZ=1.5, shapeName=dyn_loincloth_fr_r_03_box, center=<3 0 0>}]
 dyn_loincloth_fr_r_04 [{shapeType=Box, extentX=9.5, extentY=9, extentZ=1.5, shapeName=dyn_loincloth_fr_r_04_box, center=<5.3711 -1.2461 -0.2603>}]
 dyn_loincloth_fr_r_MSTR [{shapeType=Capsule, length=0.505, radius=0.505, shapeName=dyn_loincloth_fr_r_MSTR_capsule, rotation=<rot 0 -90 0>}]
 dyn_loincloth_knot_MSTR [{shapeType=Capsule, length=0.505, radius=0.505, shapeName=dyn_loincloth_knot_MSTR_capsule, rotation=<rot 0 -90 0>}]
 dyn_loincloth_side_l_01 [{shapeType=Capsule, length=13.5, radius=5, center=<8 -5 4>, rotation=<rot 0 0.0001 65>}]
 dyn_loincloth_side_l_02 [{shapeType=Capsule, length=9.25, radius=8.75, shapeName=dyn_loincloth_side_l_02_capsule, center=<8 -5 6.5>, rotation=<rot 0 0.0001 65>}]
 dyn_loincloth_side_l_03 [{shapeType=Capsule, length=9.25, radius=8.75, shapeName=dyn_loincloth_side_l_03_capsule, center=<8 -5 6.5>, rotation=<rot 0 0.0001 65>}]
 dyn_loincloth_side_l_MSTR [{shapeType=Capsule, length=0.505, radius=0.505, shapeName=dyn_loincloth_side_l_MSTR_capsule, rotation=<rot 0 -90 0>}]
 dyn_loincloth_side_r_01 [{shapeType=Capsule, length=13.5, radius=5, center=<8 -5 4>, rotation=<rot 0 0 90>}]
 dyn_loincloth_side_r_02 [{shapeType=Capsule, length=10, radius=10, shapeName=dyn_loincloth_side_r_02_capsule, center=<8 -5 6.5>, rotation=<rot 0 0 90>}]
 dyn_loincloth_side_r_03 [{shapeType=Capsule, length=10, radius=10, shapeName=dyn_loincloth_side_r_03_capsule, center=<8 -5 6.5>, rotation=<rot 0 0 90>}]
 dyn_loincloth_side_r_MSTR [{shapeType=Capsule, length=0.505, radius=0.505, shapeName=dyn_loincloth_side_r_MSTR_capsule, rotation=<rot 0 -90 0>}]
 dyn_sleeve_l_01 [{shapeType=Capsule, length=4, radius=2, shapeName=dyn_sleeve_l_01_capsule, center=<4 0 1.5>, rotation=<rot 0 0 90>}]
 dyn_sleeve_l_02 [{shapeType=Capsule, length=4, radius=2, shapeName=dyn_sleeve_l_02_capsule, center=<4 0 0>, rotation=<rot 0 0 90>}]
 dyn_sleeve_l_03 [{shapeType=Capsule, length=4, radius=2, shapeName=dyn_sleeve_l_03_capsule, center=<4 0 0>, rotation=<rot 0 0 90>}]
 dyn_sleeve_l_MSTR [{shapeType=Capsule, length=0.505, radius=0.505, shapeName=dyn_sleeve_l_MSTR_capsule, rotation=<rot 0 -90 0>}]
 dyn_sleeve_r_01 [{shapeType=Capsule, length=4, radius=2, center=<4 0 1.5>, rotation=<rot 0 0 90>}]
 dyn_sleeve_r_02 [{shapeType=Capsule, length=4, radius=2, shapeName=dyn_sleeve_r_02_capsule, center=<4 0 0>, rotation=<rot 0 0 90>}]
 dyn_sleeve_r_03 [{shapeType=Capsule, length=4, radius=2, shapeName=dyn_sleeve_r_03_capsule, center=<4 0 0>, rotation=<rot 0 0 90>}]
 dyn_sleeve_r_MSTR [{shapeType=Capsule, length=0.505, radius=0.505, shapeName=dyn_sleeve_r_MSTR_capsule, rotation=<rot 0 -90 0>}]
 pelvis [{shapeType=Capsule, length=4.28, radius=10.25, center=<0.1284 0 0>}]
 spine_01 [{shapeType=Capsule, length=16.3607, radius=10.4893, center=<6.2216 0 0>, rotation=<rot 90 0 0>}]
 thigh_l [{shapeType=Capsule, length=39.0865, radius=7, center=<-13.636 0 0>, rotation=<rot 90 0 0>}]
 thigh_r [{shapeType=Capsule, length=39.645, radius=7.5, center=<13.9352 0 0>, rotation=<rot 90 0 0>}]
}
constraints [
 {bone1Name=spine_01, bone2Name=pelvis, swing1LimitDegrees=45, swing2LimitDegrees=45, twistLimitDegrees=45}
 {bone1Name=dyn_sleeve_r_01, bone2Name=dyn_sleeve_r_MSTR, swing1LimitDegrees=30, swing2LimitDegrees=35, twistLimitDegrees=2}
 {bone1Name=dyn_loincloth_fr_l_02, bone2Name=dyn_loincloth_fr_l_01, swing1LimitDegrees=45, swing2LimitDegrees=25, twistLimitDegrees=1}
 {bone1Name=dyn_loincloth_side_l_01, bone2Name=dyn_loincloth_side_l_MSTR, swing1LimitDegrees=45, swing2LimitDegrees=30, twistLimitDegrees=1}
 {bone1Name=dyn_loincloth_bk_l_01, bone2Name=dyn_loincloth_bk_l_MSTR, swing1LimitDegrees=25, swing2LimitDegrees=45, twistLimitDegrees=1}
 {bone1Name=dyn_loincloth_fr_r_01, bone2Name=dyn_loincloth_fr_r_MSTR, swing1LimitDegrees=30, swing2LimitDegrees=45, twistLimitDegrees=1}
 {bone1Name=dyn_loincloth_side_r_01, bone2Name=dyn_loincloth_side_r_MSTR, swing1LimitDegrees=25, swing2LimitDegrees=45, twistLimitDegrees=1}
 {bone1Name=dyn_loincloth_bk_r_01, bone2Name=dyn_loincloth_bk_r_MSTR, swing1LimitDegrees=25, swing2LimitDegrees=45, twistLimitDegrees=1}
 {bone1Name=dyn_loincloth_bk_01, bone2Name=dyn_loincloth_bk_MSTR, swing1LimitDegrees=25, swing2LimitDegrees=45, twistLimitDegrees=1}
 {bone1Name=thigh_l, bone2Name=pelvis, swing1LimitDegrees=45, swing2LimitDegrees=45, twistLimitDegrees=45}
 {bone1Name=calf_l, bone2Name=thigh_l, swing1LimitDegrees=45, swing2LimitDegrees=45, twistLimitDegrees=45}
 {bone1Name=thigh_r, bone2Name=pelvis, swing1LimitDegrees=45, swing2LimitDegrees=45, twistLimitDegrees=45}
 {bone1Name=calf_r, bone2Name=thigh_r, swing1LimitDegrees=45, swing2LimitDegrees=45, twistLimitDegrees=45}
 {bone1Name=dyn_sleeve_r_02, bone2Name=dyn_sleeve_r_01, swing1LimitDegrees=30, swing2LimitDegrees=35, twistLimitDegrees=2}
 {bone1Name=dyn_sleeve_r_03, bone2Name=dyn_sleeve_r_02, swing1LimitDegrees=30, swing2LimitDegrees=35, twistLimitDegrees=2}
 {bone1Name=dyn_sleeve_l_01, bone2Name=dyn_sleeve_l_MSTR, swing1LimitDegrees=30, swing2LimitDegrees=35, twistLimitDegrees=2}
 {bone1Name=dyn_sleeve_l_02, bone2Name=dyn_sleeve_l_01, swing1LimitDegrees=30, swing2LimitDegrees=35, twistLimitDegrees=2}
 {bone1Name=dyn_sleeve_l_03, bone2Name=dyn_sleeve_l_02, swing1LimitDegrees=30, swing2LimitDegrees=35, twistLimitDegrees=2}
 {bone1Name=dyn_loincloth_fr_l_03, bone2Name=dyn_loincloth_fr_l_02, swing1LimitDegrees=45, swing2LimitDegrees=25, twistLimitDegrees=1}
 {bone1Name=dyn_loincloth_fr_l_04, bone2Name=dyn_loincloth_fr_l_03, swing1LimitDegrees=45, swing2LimitDegrees=25, twistLimitDegrees=1}
 {bone1Name=dyn_loincloth_side_l_02, bone2Name=dyn_loincloth_side_l_01, swing1LimitDegrees=25, swing2LimitDegrees=45, twistLimitDegrees=5}
 {bone1Name=dyn_loincloth_side_l_03, bone2Name=dyn_loincloth_side_l_02, swing1LimitDegrees=45, swing2LimitDegrees=45, twistLimitDegrees=5}
 {bone1Name=dyn_loincloth_bk_l_02, bone2Name=dyn_loincloth_bk_l_01, swing1LimitDegrees=25, swing2LimitDegrees=45, twistLimitDegrees=5}
 {bone1Name=dyn_loincloth_bk_l_03, bone2Name=dyn_loincloth_bk_l_02, swing1LimitDegrees=35, swing2LimitDegrees=60, twistLimitDegrees=5}
 {bone1Name=dyn_loincloth_fr_r_02, bone2Name=dyn_loincloth_fr_r_01, swing1LimitDegrees=25, swing2LimitDegrees=45, twistLimitDegrees=5}
 {bone1Name=dyn_loincloth_fr_r_03, bone2Name=dyn_loincloth_fr_r_02, swing1LimitDegrees=45, swing2LimitDegrees=45, twistLimitDegrees=5}
 {bone1Name=dyn_loincloth_fr_r_04, bone2Name=dyn_loincloth_fr_r_03, swing1LimitDegrees=5, swing2LimitDegrees=25, twistLimitDegrees=0}
 {bone1Name=dyn_loincloth_side_r_02, bone2Name=dyn_loincloth_side_r_01, swing1LimitDegrees=25, swing2LimitDegrees=45, twistLimitDegrees=5}
 {bone1Name=dyn_loincloth_side_r_03, bone2Name=dyn_loincloth_side_r_02, swing1LimitDegrees=45, swing2LimitDegrees=45, twistLimitDegrees=5}
 {bone1Name=dyn_loincloth_bk_r_02, bone2Name=dyn_loincloth_bk_r_01, swing1LimitDegrees=25, swing2LimitDegrees=45, twistLimitDegrees=5}
 {bone1Name=dyn_loincloth_bk_r_03, bone2Name=dyn_loincloth_bk_r_02, swing1LimitDegrees=35, swing2LimitDegrees=60, twistLimitDegrees=5}
 {bone1Name=dyn_loincloth_bk_02, bone2Name=dyn_loincloth_bk_01, swing1LimitDegrees=25, swing2LimitDegrees=45, twistLimitDegrees=5}
 {bone1Name=dyn_loincloth_bk_03, bone2Name=dyn_loincloth_bk_02, swing1LimitDegrees=35, swing2LimitDegrees=60, twistLimitDegrees=3}
 {bone1Name=dyn_loincloth_fr_l_01, bone2Name=dyn_loincloth_fr_l_MSTR, swing1LimitDegrees=30, swing2LimitDegrees=45, twistLimitDegrees=1}
 {
  bone1Name=dyn_loincloth_knot_MSTR
  bone2Name=spine_01
  swing1LimitDegrees=45
  swing1Motion=Free
  swing2LimitDegrees=45
  swing2Motion=Free
  twistLimitDegrees=45
  twistMotion=Free
 }
 {
  bone1Name=dyn_sleeve_l_MSTR
  bone2Name=dyn_loincloth_knot_MSTR
  swing1LimitDegrees=45
  swing1Motion=Free
  swing2LimitDegrees=45
  swing2Motion=Free
  twistLimitDegrees=45
  twistMotion=Free
 }
 {
  bone1Name=dyn_sleeve_r_MSTR
  bone2Name=dyn_loincloth_knot_MSTR
  swing1LimitDegrees=45
  swing1Motion=Free
  swing2LimitDegrees=45
  swing2Motion=Free
  twistLimitDegrees=45
  twistMotion=Free
 }
 {
  bone1Name=dyn_loincloth_fr_l_MSTR
  bone2Name=spine_01
  swing1LimitDegrees=45
  swing1Motion=Free
  swing2LimitDegrees=45
  swing2Motion=Free
  twistLimitDegrees=45
  twistMotion=Free
 }
 {
  bone1Name=dyn_loincloth_side_l_MSTR
  bone2Name=spine_01
  swing1LimitDegrees=45
  swing1Motion=Free
  swing2LimitDegrees=45
  swing2Motion=Free
  twistLimitDegrees=45
  twistMotion=Free
 }
 {
  bone1Name=dyn_loincloth_bk_l_MSTR
  bone2Name=spine_01
  swing1LimitDegrees=45
  swing1Motion=Free
  swing2LimitDegrees=45
  swing2Motion=Free
  twistLimitDegrees=45
  twistMotion=Free
 }
 {
  bone1Name=dyn_loincloth_fr_r_MSTR
  bone2Name=spine_01
  swing1LimitDegrees=45
  swing1Motion=Free
  swing2LimitDegrees=45
  swing2Motion=Free
  twistLimitDegrees=45
  twistMotion=Free
 }
 {
  bone1Name=dyn_loincloth_side_r_MSTR
  bone2Name=spine_01
  swing1LimitDegrees=45
  swing1Motion=Free
  swing2LimitDegrees=45
  swing2Motion=Free
  twistLimitDegrees=45
  twistMotion=Free
 }
 {
  bone1Name=dyn_loincloth_bk_r_MSTR
  bone2Name=spine_01
  swing1LimitDegrees=45
  swing1Motion=Free
  swing2LimitDegrees=45
  swing2Motion=Free
  twistLimitDegrees=45
  twistMotion=Free
 }
 {
  bone1Name=dyn_loincloth_bk_MSTR
  bone2Name=spine_01
  swing1LimitDegrees=45
  swing1Motion=Free
  swing2LimitDegrees=45
  swing2Motion=Free
  twistLimitDegrees=45
  twistMotion=Free
 }
 {
  bone1Name=dyn_loincloth_fr_r_02
  bone2Name=dyn_loincloth_side_r_02
  swing1LimitDegrees=45
  swing1Motion=Free
  swing2LimitDegrees=45
  swing2Motion=Free
  twistLimitDegrees=45
  twistMotion=Free
 }
 {
  bone1Name=dyn_loincloth_fr_l_02
  bone2Name=dyn_loincloth_side_l_02
  swing1LimitDegrees=45
  swing1Motion=Free
  swing2LimitDegrees=45
  swing2Motion=Free
  twistLimitDegrees=45
  twistMotion=Free
 }
 {
  bone1Name=dyn_loincloth_fr_l_03
  bone2Name=dyn_loincloth_side_l_03
  swing1LimitDegrees=45
  swing1Motion=Free
  swing2LimitDegrees=45
  swing2Motion=Free
  twistLimitDegrees=45
  twistMotion=Free
 }
 {
  bone1Name=dyn_loincloth_side_r_02
  bone2Name=dyn_loincloth_fr_r_03
  swing1LimitDegrees=45
  swing1Motion=Free
  swing2LimitDegrees=45
  swing2Motion=Free
  twistLimitDegrees=45
  twistMotion=Free
 }
]
