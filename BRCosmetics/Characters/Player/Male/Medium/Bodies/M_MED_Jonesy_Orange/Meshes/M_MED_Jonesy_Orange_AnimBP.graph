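# Animation Blueprint for the M_MED_Jonesy_Orange player body that poses the loincloth and sleeve bones over a cached body pose.
# Chains of Constraint, Look At and Modify Bone nodes follow thighs and pelvis, weighted by thigh rotation and movement-direction variables, often with interpolated, clamped alphas.
# Separate branches handle front-end and emote states.
# The front-end branch adds pelvis-space RigidBody simulation, while another branch runs a crouch Control Rig.
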AnimBlueprint /BRCosmetics/Characters/Player/Male/Medium/Bodies/M_MED_Jonesy_Orange/Meshes/M_MED_Jonesy_Orange_AnimBP
target_skeleton=&/M_MED_Jonesy_Orange_Skeleton
graphs [
 AnimGraph {nodes [
   n1 = "Constraint - dyn_loincloth_fr_l_02 (P)"(ComponentPose=Local To Component(LocalPose=Use cached pose 'Pose'()), Alpha=1, ConstraintWeights_0=1) {
    alphaBoolBlend {blendOption=Linear}
    alphaScaleBiasClamp {bClampResult=true, bInterpResult=true, clampMax=0.6, interpSpeedDecreasing=5, interpSpeedIncreasing=7}
    boneToModify {boneName=dyn_loincloth_fr_l_02}
    constraintSetup [{targetBone {boneName=thigh_l}}]
    constraintWeights [1]
   }
   n2 = "Constraint - dyn_loincloth_fr_r_02 (P)"(ComponentPose=@n1, Alpha=1, ConstraintWeights_0=1) {
    alphaBoolBlend {blendOption=Linear}
    alphaScaleBiasClamp {bClampResult=true, bInterpResult=true, clampMax=0.6, interpSpeedDecreasing=5, interpSpeedIncreasing=7}
    boneToModify {boneName=dyn_loincloth_fr_r_02}
    constraintSetup [{targetBone {boneName=thigh_r}}]
    constraintWeights [1]
   }
   n3 = "Transform (Modify) Bone - Bone: dyn_loincloth_bk_l_MSTR"!(ComponentPose=@n2, Alpha=1, Rotation="15.000,0.000,-0.000") -> (Pose) {
    rotationMode=BMM_Replace
    rotationSpace=BCS_BoneSpace
    scaleSpace=BCS_BoneSpace
    translationSpace=BCS_BoneSpace
    alphaBoolBlend {blendOption=Linear}
    alphaScaleBiasClamp {bInterpResult=true, interpSpeedDecreasing=5, interpSpeedIncreasing=4}
    boneToModify {boneName=dyn_loincloth_bk_l_MSTR}
    rotation=<rot 15 0 0>
    translation=<-7.723 3.33 0.199>
   }
   n4 = "Transform (Modify) Bone - Bone: dyn_loincloth_bk_r_MSTR"!(ComponentPose=@n3, Alpha=1, Rotation="15.000,-0.000,0.000") -> (Pose) {
    rotationMode=BMM_Replace
    rotationSpace=BCS_BoneSpace
    scaleSpace=BCS_BoneSpace
    translationSpace=BCS_BoneSpace
    alphaBoolBlend {blendOption=Linear}
    alphaScaleBiasClamp {bInterpResult=true, interpSpeedDecreasing=5, interpSpeedIncreasing=4}
    boneToModify {boneName=dyn_loincloth_bk_r_MSTR}
    rotation=<rot 15 0 0>
    translation=<-7.723 3.33 0.199>
   }
   n5 = "Transform (Modify) Bone - Bone: dyn_loincloth_bk_MSTR"!(ComponentPose=@n4, Alpha=1, Rotation="20.000,-0.000,10.000") -> (Pose) {
    rotationMode=BMM_Replace
    rotationSpace=BCS_BoneSpace
    scaleSpace=BCS_BoneSpace
    translationSpace=BCS_BoneSpace
    alphaBoolBlend {blendOption=Linear}
    alphaScaleBiasClamp {bInterpResult=true, interpSpeedDecreasing=5, interpSpeedIncreasing=4}
    boneToModify {boneName=dyn_loincloth_bk_MSTR}
    rotation=<rot 20 0 10>
    translation=<-7.723 3.33 0.199>
   }
   n6 = "Transform (Modify) Bone - Bone: dyn_sleeve_l_02"!(ComponentPose=@n5, Alpha=1, Translation="0,0,5") -> (Pose) {
    rotationSpace=BCS_BoneSpace
    scaleSpace=BCS_BoneSpace
    translationMode=BMM_Additive
    translationSpace=BCS_BoneSpace
    alphaBoolBlend {blendOption=Linear}
    alphaScaleBiasClamp {bInterpResult=true, interpSpeedDecreasing=5, interpSpeedIncreasing=4}
    boneToModify {boneName=dyn_sleeve_l_02}
    rotation {pitch=15}
    translation {z=5}
   }
   n7 = "Transform (Modify) Bone - Bone: dyn_loincloth_side_l_MSTR"!(ComponentPose=@n6, Alpha=1, Rotation="-6.812,-33.775,-6.072") -> (Pose) {
    rotationMode=BMM_Replace
    rotationSpace=BCS_BoneSpace
    scaleSpace=BCS_BoneSpace
    translationSpace=BCS_BoneSpace
    alphaBoolBlend {blendOption=Linear}
    alphaScaleBiasClamp {bClampResult=true, bInterpResult=true, clampMax=0.45, interpSpeedDecreasing=5, interpSpeedIncreasing=7}
    boneToModify {boneName=dyn_loincloth_side_l_MSTR}
    rotation=<rot -6.812 -33.775 -6.072>
    translation=<-7.723 3.33 0.199>
   }
   n8 = "Transform (Modify) Bone - Bone: dyn_loincloth_side_r_MSTR"!(ComponentPose=@n7, Alpha=1, Rotation="-0.943,19.574,4.217") -> (Pose) {
    rotationMode=BMM_Replace
    rotationSpace=BCS_BoneSpace
    scaleSpace=BCS_BoneSpace
    translationSpace=BCS_BoneSpace
    alphaBoolBlend {blendOption=Linear}
    alphaScaleBiasClamp {bClampResult=true, bInterpResult=true, clampMax=0.45, interpSpeedDecreasing=5, interpSpeedIncreasing=7}
    boneToModify {boneName=dyn_loincloth_side_r_MSTR}
    rotation=<rot -0.943 19.574 4.217>
    translation=<-7.723 3.33 0.199>
   }
   n9 = "Transform (Modify) Bone - Bone: dyn_loincloth_bk_l_MSTR"!(ComponentPose=@n8, Alpha=1, Translation:Vector="-0.372,0.000,4.238", Rotation="5,0.000,-0.000") -> (Pose) {
    rotationMode=BMM_Replace
    rotationSpace=BCS_BoneSpace
    scaleSpace=BCS_BoneSpace
    translationMode=BMM_Replace
    translationSpace=BCS_BoneSpace
    alphaBoolBlend {blendOption=Linear}
    alphaScaleBiasClamp {bClampResult=true, bInterpResult=true, clampMax=0.5, interpSpeedDecreasing=5, interpSpeedIncreasing=7}
    boneToModify {boneName=dyn_loincloth_bk_l_MSTR}
    rotation {pitch=5}
    translation {x=-0.372, z=4.238}
   }
   n10 = "Transform (Modify) Bone - Bone: dyn_loincloth_bk_r_MSTR"!(ComponentPose=@n9, Alpha=1, Translation:Vector="-0.674,-0.000,7.706", Rotation="5,0,0") -> (Pose) {
    rotationMode=BMM_Replace
    rotationSpace=BCS_BoneSpace
    scaleSpace=BCS_BoneSpace
    translationMode=BMM_Replace
    translationSpace=BCS_BoneSpace
    alphaBoolBlend {blendOption=Linear}
    alphaScaleBiasClamp {bClampResult=true, bInterpResult=true, clampMax=0.5, interpSpeedDecreasing=5, interpSpeedIncreasing=7}
    boneToModify {boneName=dyn_loincloth_bk_r_MSTR}
    rotation {pitch=5}
    translation {x=-0.674, z=7.706}
   }
   n11 = "Transform (Modify) Bone - Bone: dyn_loincloth_bk_MSTR"!(ComponentPose=@n10, Alpha=1, Translation:Vector="-1.738,0.000,19.879", Rotation="5,0,0") -> (Pose) {
    rotationMode=BMM_Replace
    rotationSpace=BCS_BoneSpace
    scaleSpace=BCS_BoneSpace
    translationMode=BMM_Replace
    translationSpace=BCS_BoneSpace
    alphaBoolBlend {blendOption=Linear}
    alphaScaleBiasClamp {bClampResult=true, bInterpResult=true, clampMax=0.5, interpSpeedDecreasing=5, interpSpeedIncreasing=7}
    boneToModify {boneName=dyn_loincloth_bk_MSTR}
    rotation {pitch=5}
    translation {x=-1.738, z=19.879}
   }
   n12 = "Constraint - dyn_sleeve_l_02 (P)"(ComponentPose=@n11, Alpha=1, ConstraintWeights_0=1) {
    alphaBoolBlend {blendOption=Linear}
    alphaScaleBiasClamp {bClampResult=true, bInterpResult=true, interpSpeedDecreasing=5, interpSpeedIncreasing=7}
    boneToModify {boneName=dyn_sleeve_l_02}
    constraintSetup [{targetBone {boneName=thigh_l}}]
    constraintWeights [1]
   }
   n13 = "Transform (Modify) Bone - Bone: dyn_sleeve_l_02"!(ComponentPose=@n12, Alpha=1, Translation="10.553,11.431,0.517", Rotation:Rotator="-15.000,0.037,0.082") -> (Pose) {
    rotationMode=BMM_Replace
    rotationSpace=BCS_BoneSpace
    scaleSpace=BCS_BoneSpace
    translationMode=BMM_Replace
    translationSpace=BCS_BoneSpace
    alphaBoolBlend {blendOption=Linear}
    alphaScaleBiasClamp {bClampResult=true, bInterpResult=true, clampMax=0.5, interpSpeedDecreasing=5, interpSpeedIncreasing=7}
    boneToModify {boneName=dyn_sleeve_l_02}
    rotation=<rot -15 0.037 0.082>
    translation=<10.553 11.431 0.517>
   }
   n14 = "Constraint - dyn_sleeve_r_02 (P)"(ComponentPose=@n13, Alpha=1, ConstraintWeights_0=1) {
    alphaBoolBlend {blendOption=Linear}
    alphaScaleBiasClamp {bClampResult=true, bInterpResult=true, interpSpeedDecreasing=5, interpSpeedIncreasing=7}
    boneToModify {boneName=dyn_sleeve_r_02}
    constraintSetup [{targetBone {boneName=thigh_r}}]
    constraintWeights [1]
   }
   n15 = "Transform (Modify) Bone - Bone: dyn_loincloth_side_l_MSTR"(ComponentPose=@n14, Alpha=Get thigh_rot_l(), Rotation="10.000,0.020,0.033") {
    rotationMode=BMM_Replace
    rotationSpace=BCS_BoneSpace
    scaleSpace=BCS_BoneSpace
    translationSpace=BCS_BoneSpace
    alphaBoolBlend {blendOption=Linear}
    boneToModify {boneName=dyn_loincloth_side_l_MSTR}
    rotation=<rot 10 0.02 0.033>
    translation=<-7.723 3.33 0.199>
   }
   n16 = "Transform (Modify) Bone - Bone: dyn_loincloth_side_r_MSTR"(ComponentPose=@n15, Alpha=Get thigh_rot_r(), Rotation="10,0,0") {
    rotationMode=BMM_Replace
    rotationSpace=BCS_BoneSpace
    scaleSpace=BCS_BoneSpace
    translationSpace=BCS_BoneSpace
    alphaBoolBlend {blendOption=Linear}
    boneToModify {boneName=dyn_loincloth_side_r_MSTR}
    rotation {pitch=10}
    translation=<-7.723 3.33 0.199>
   }
   n17 = Get Left_Direct() -> (Left_Direct:double)
   n18 = Look At - Bone: dyn_loincloth_fr_l_MSTR(ComponentPose=@n16, Alpha=@n17) {
    alphaBoolBlend {blendOption=Linear}
    alphaScaleBiasClamp {bClampResult=true, bInterpResult=true, clampMax=0.8, interpSpeedDecreasing=5, interpSpeedIncreasing=7}
    boneToModify {boneName=dyn_loincloth_fr_l_MSTR}
    lookAtLocation=<-9.1694 5.599 -15.0267>
    lookAtTarget {boneReference {boneName=thigh_l}}
    lookAt_Axis {axis=<1 0>}
   }
   n19 = Look At - Bone: dyn_loincloth_side_l_MSTR(ComponentPose=@n18, Alpha=@n17) {
    alphaBoolBlend {blendOption=Linear}
    alphaScaleBiasClamp {bClampResult=true, bInterpResult=true, clampMax=0.8, interpSpeedDecreasing=5, interpSpeedIncreasing=7}
    boneToModify {boneName=dyn_loincloth_side_l_MSTR}
    lookAtLocation=<-3.6095 -8.9752 -20.1742>
    lookAtTarget {boneReference {boneName=thigh_l}}
    lookAt_Axis {axis=<1 0>}
   }
   n20 = Look At - Bone: dyn_loincloth_bk_l_MSTR(ComponentPose=@n19, Alpha=@n17) {
    alphaBoolBlend {blendOption=Linear}
    alphaScaleBiasClamp {bClampResult=true, bInterpResult=true, clampMax=0.8, interpSpeedDecreasing=5, interpSpeedIncreasing=7}
    boneToModify {boneName=dyn_loincloth_bk_l_MSTR}
    lookAtLocation=<-1.0626 -12.5492 -9.5576>
    lookAtTarget {boneReference {boneName=thigh_l}}
    lookAt_Axis {axis=<1 0>}
   }
   n21 = Get Backward_Direct() -> (Backward_Direct:double)
   n22 = Look At - Bone: dyn_loincloth_bk_l_MSTR(ComponentPose=@n20, Alpha=@n21) {
    alphaBoolBlend {blendOption=Linear}
    alphaScaleBiasClamp {bInterpResult=true, interpSpeedDecreasing=5, interpSpeedIncreasing=7}
    boneToModify {boneName=dyn_loincloth_bk_l_MSTR}
    lookAtLocation=<-4.9982 -9.108 -15.5626>
    lookAtTarget {boneReference {boneName=pelvis}}
    lookAt_Axis {axis=<1 0>}
   }
   n23 = Look At - Bone: dyn_loincloth_bk_r_MSTR(ComponentPose=@n22, Alpha=@n21) {
    alphaBoolBlend {blendOption=Linear}
    alphaScaleBiasClamp {bInterpResult=true, interpSpeedDecreasing=5, interpSpeedIncreasing=7}
    boneToModify {boneName=dyn_loincloth_bk_r_MSTR}
    lookAtLocation=<-4.7251 -15.7471 10.0725>
    lookAtTarget {boneReference {boneName=pelvis}}
    lookAt_Axis {axis=<1 0>}
   }
   n24 = Look At - Bone: dyn_loincloth_bk_MSTR(ComponentPose=@n23, Alpha=@n21) {
    alphaBoolBlend {blendOption=Linear}
    alphaScaleBiasClamp {bInterpResult=true, interpSpeedDecreasing=5, interpSpeedIncreasing=7}
    boneToModify {boneName=dyn_loincloth_bk_MSTR}
    lookAtLocation=<-4.8727 -15.2368 0.001>
    lookAtTarget {boneReference {boneName=pelvis}}
    lookAt_Axis {axis=<1 0>}
   }
   n25 = Get Right_Direct() -> (Right_Direct:double)
   n26 = Look At - Bone: dyn_loincloth_fr_r_MSTR(ComponentPose=@n24, Alpha=@n25) {
    alphaBoolBlend {blendOption=Linear}
    alphaScaleBiasClamp {bInterpResult=true, interpSpeedDecreasing=5, interpSpeedIncreasing=7}
    boneToModify {boneName=dyn_loincloth_fr_r_MSTR}
    lookAtLocation=<39.6412 -8.6813 17.1199>
    lookAtTarget {boneReference {boneName=thigh_r}}
    lookAt_Axis {axis=<1 0>}
   }
   n27 = Look At - Bone: dyn_loincloth_side_r_MSTR(ComponentPose=@n26, Alpha=@n25) {
    alphaBoolBlend {blendOption=Linear}
    alphaScaleBiasClamp {bInterpResult=true, interpSpeedDecreasing=5, interpSpeedIncreasing=7}
    boneToModify {boneName=dyn_loincloth_side_r_MSTR}
    lookAtLocation=<8.446 6.6285 22.0747>
    lookAtTarget {boneReference {boneName=thigh_r}}
    lookAt_Axis {axis=<1 0>}
   }
   n28 = Look At - Bone: dyn_loincloth_bk_r_MSTR(ComponentPose=@n27, Alpha=@n25) {
    alphaBoolBlend {blendOption=Linear}
    alphaScaleBiasClamp {bInterpResult=true, interpSpeedDecreasing=5, interpSpeedIncreasing=7}
    boneToModify {boneName=dyn_loincloth_bk_r_MSTR}
    lookAtLocation=<4.9666 14.4204 6.6651>
    lookAtTarget {boneReference {boneName=thigh_r}}
    lookAt_Axis {axis=<1 0>}
   }
   n29 = "Transform (Modify) Bone - Bone: dyn_sleeve_l_MSTR"!(ComponentPose=@n28, Alpha=@n25, Rotation="12,-2,0") -> (Pose) {
    rotationMode=BMM_Replace
    rotationSpace=BCS_BoneSpace
    scaleSpace=BCS_BoneSpace
    translationSpace=BCS_BoneSpace
    alphaBoolBlend {blendOption=Linear}
    boneToModify {boneName=dyn_sleeve_l_MSTR}
    rotation {pitch=12, yaw=-2}
    translation=<-7.723 3.33 0.199>
   }
   n30 = "Transform (Modify) Bone - Bone: dyn_sleeve_r_MSTR"!(ComponentPose=@n29, Alpha=@n25, Rotation="5,8,0") -> (Pose) {
    rotationMode=BMM_Replace
    rotationSpace=BCS_BoneSpace
    scaleSpace=BCS_BoneSpace
    translationSpace=BCS_BoneSpace
    alphaBoolBlend {blendOption=Linear}
    boneToModify {boneName=dyn_sleeve_r_MSTR}
    rotation {pitch=5, yaw=8}
    translation=<-7.723 3.33 0.199>
   }
   n31 = "Transform (Modify) Bone - Bone: dyn_loincloth_bk_r_MSTR"(ComponentPose=@n30, Alpha=1, Translation="0,0,2", Rotation="5,0,0") {
    rotationMode=BMM_Replace
    rotationSpace=BCS_BoneSpace
    scaleSpace=BCS_BoneSpace
    translationMode=BMM_Replace
    translationSpace=BCS_BoneSpace
    alphaBoolBlend {blendOption=Linear}
    alphaScaleBiasClamp {bInterpResult=true, clampMax=0.5, interpSpeedIncreasing=1}
    boneToModify {boneName=dyn_loincloth_bk_r_MSTR}
    rotation {pitch=5}
    translation {z=2}
   }
   n32 = "Transform (Modify) Bone - Bone: dyn_loincloth_bk_MSTR"(ComponentPose=@n31, Alpha=1, Translation="0,0,2", Rotation="5,-15,0") {
    rotationMode=BMM_Replace
    rotationSpace=BCS_BoneSpace
    scaleSpace=BCS_BoneSpace
    translationMode=BMM_Replace
    translationSpace=BCS_BoneSpace
    alphaBoolBlend {blendOption=Linear}
    alphaScaleBiasClamp {bInterpResult=true, clampMax=0.5, interpSpeedIncreasing=1}
    boneToModify {boneName=dyn_loincloth_bk_MSTR}
    rotation {pitch=5, yaw=-15}
    translation {z=2}
   }
   n33 = "Transform (Modify) Bone - Bone: dyn_loincloth_bk_l_MSTR"(ComponentPose=@n32, Alpha=1, Translation="0,0,2", Rotation="5,-15,0") {
    rotationMode=BMM_Replace
    rotationSpace=BCS_BoneSpace
    scaleSpace=BCS_BoneSpace
    translationMode=BMM_Replace
    translationSpace=BCS_BoneSpace
    alphaBoolBlend {blendOption=Linear}
    alphaScaleBiasClamp {bInterpResult=true, clampMax=0.5, interpSpeedIncreasing=1}
    boneToModify {boneName=dyn_loincloth_bk_l_MSTR}
    rotation {pitch=5, yaw=-15}
    translation {z=2}
   }
   n34 = "Constraint - dyn_loincloth_fr_l_MSTR (P)"(ComponentPose=Local To Component(LocalPose=Use cached pose 'Pose'()), Alpha=1, ConstraintWeights_0=1) {
    alphaBoolBlend {blendOption=Linear}
    alphaScaleBiasClamp {bInterpResult=true}
    boneToModify {boneName=dyn_loincloth_fr_l_MSTR}
    constraintSetup [{targetBone {boneName=thigh_l}}]
    constraintWeights [1]
   }
   n35 = "Constraint - dyn_loincloth_fr_r_MSTR (P)"(ComponentPose=@n34, Alpha=1, ConstraintWeights_0=1) {
    alphaBoolBlend {blendOption=Linear}
    alphaScaleBiasClamp {bInterpResult=true}
    boneToModify {boneName=dyn_loincloth_fr_r_MSTR}
    constraintSetup [{targetBone {boneName=thigh_r}}]
    constraintWeights [1]
   }
   n36 = Look At - Bone: dyn_loincloth_bk_l_MSTR(ComponentPose=@n35, Alpha=1) {
    alphaBoolBlend {blendOption=Linear}
    alphaScaleBiasClamp {bInterpResult=true}
    boneToModify {boneName=dyn_loincloth_bk_l_MSTR}
    lookAtLocation=<-87.4612 -21.0135 14.6116>
    lookAtTarget {boneReference {boneName=thigh_l}}
    lookAt_Axis {axis=<1 0>}
   }
   n37 = Look At - Bone: dyn_loincloth_bk_MSTR(ComponentPose=@n36, Alpha=1) {
    alphaBoolBlend {blendOption=Linear}
    alphaScaleBiasClamp {bInterpResult=true}
    boneToModify {boneName=dyn_loincloth_bk_MSTR}
    lookAtLocation=<-95.3406 -23.9551>
    lookAtTarget {boneReference {boneName=pelvis}}
    lookAt_Axis {axis=<1 0>}
   }
   n38 = Look At - Bone: dyn_loincloth_bk_r_MSTR(ComponentPose=@n37, Alpha=1) {
    alphaBoolBlend {blendOption=Linear}
    alphaScaleBiasClamp {bInterpResult=true}
    boneToModify {boneName=dyn_loincloth_bk_r_MSTR}
    lookAtLocation=<89.8493 19.9182>
    lookAtTarget {boneReference {boneName=thigh_r}}
    lookAt_Axis {axis=<1 0>}
   }
   n39 = "Transform (Modify) Bone - Bone: dyn_loincloth_fr_l_MSTR"!(ComponentPose=@n38, Alpha=1, Translation="-2.906,0.024,1.799", Rotation="-7.156,-7.519,0.966") -> (Pose) {
    rotationMode=BMM_Replace
    rotationSpace=BCS_BoneSpace
    translationMode=BMM_Replace
    translationSpace=BCS_BoneSpace
    alphaBoolBlend {blendOption=Linear}
    alphaScaleBiasClamp {bInterpResult=true, clampMax=0.6, interpSpeedDecreasing=5, interpSpeedIncreasing=7}
    boneToModify {boneName=dyn_loincloth_fr_l_MSTR}
    rotation=<rot -7.156 -7.519 0.966>
    translation=<-2.906 0.024 1.799>
   }
   n40 = "Transform (Modify) Bone - Bone: dyn_loincloth_fr_r_MSTR"!(ComponentPose=@n39, Alpha=1, Translation="0.963,0.076,0.1", Rotation="-4.686,-5.025,-0.043") -> (Pose) {
    rotationMode=BMM_Replace
    rotationSpace=BCS_BoneSpace
    translationMode=BMM_Replace
    translationSpace=BCS_BoneSpace
    alphaBoolBlend {blendOption=Linear}
    alphaScaleBiasClamp {bInterpResult=true, clampMax=0.6, interpSpeedDecreasing=5, interpSpeedIncreasing=7}
    boneToModify {boneName=dyn_loincloth_fr_r_MSTR}
    rotation=<rot -4.686 -5.025 -0.043>
    translation=<0.963 0.076 0.1>
   }
   n41 = "Transform (Modify) Bone - Bone: dyn_loincloth_fr_l_04"!(ComponentPose=@n40, Alpha=1, Translation="0,0,0", Rotation="2.878,-20.407,0.002") -> (Pose) {
    rotationMode=BMM_Replace
    rotationSpace=BCS_BoneSpace
    translationMode=BMM_Replace
    translationSpace=BCS_BoneSpace
    alphaBoolBlend {blendOption=Linear}
    alphaScaleBiasClamp {bInterpResult=true, clampMax=0.6, interpSpeedDecreasing=5, interpSpeedIncreasing=7}
    boneToModify {boneName=dyn_loincloth_fr_l_04}
    rotation=<rot 2.878 -20.407 0.002>
   }
   n42 = "Transform (Modify) Bone - Bone: dyn_loincloth_fr_r_04"!(ComponentPose=@n41, Alpha=1, Translation="0,0,0", Rotation="0.003,12.683,-0.003") -> (Pose) {
    rotationMode=BMM_Replace
    rotationSpace=BCS_BoneSpace
    translationMode=BMM_Replace
    translationSpace=BCS_BoneSpace
    alphaBoolBlend {blendOption=Linear}
    alphaScaleBiasClamp {bInterpResult=true, clampMax=0.6, interpSpeedDecreasing=5, interpSpeedIncreasing=7}
    boneToModify {boneName=dyn_loincloth_fr_r_04}
    rotation=<rot 0.003 12.683 -0.003>
   }
   n43 = RigidBody!(ComponentPose=@n42, Alpha, OverrideWorldGravity:Vector="0,0,-980", ExternalForce) -> (Pose) {
    bClampLinearTranslationLimitToRefPose=true
    bOverrideWorldGravity=true
    overridePhysicsAsset=&/M_MED_Jonesy_Orange_Physics
    simulationSpace=BaseBoneSpace
    alphaBoolBlend {blendOption=Linear}
    baseBoneRef {boneName=pelvis}
    componentAppliedLinearAccClamp=<3000 3000 3000>
    componentLinearAccScale=<0.5 0.3 0.3>
    componentLinearVelScale=<0.5 0.3 0.3>
    simSpaceSettings {maxAngularVelocity=3, maxLinearAcceleration=700, maxLinearVelocity=500, velocityScaleZ=0, worldAlpha=1, externalLinearDragV=<2 0.2>}
   }
   n44 = "Transform (Modify) Bone - Bone: dyn_loincloth_fr_l_MSTR"(ComponentPose=@n43, Alpha=1, Translation="10,0,1.5", Rotation="0,0,6") {
    rotationMode=BMM_Replace
    rotationSpace=BCS_BoneSpace
    translationMode=BMM_Replace
    translationSpace=BCS_BoneSpace
    alphaBoolBlend {blendOption=Linear}
    alphaScaleBiasClamp {bInterpResult=true, clampMax=0.6, interpSpeedDecreasing=5, interpSpeedIncreasing=7}
    boneToModify {boneName=dyn_loincloth_fr_l_MSTR}
    rotation {roll=6}
    translation {x=10, z=1.5}
   }
   n45 = Blend Poses by bool(BlendPose_0=Component To Local(ComponentPose=@n44), BlendPose_1=Component To Local(ComponentPose=@n33), bActiveValue=Get bIsInFrontEnd()) {blendType=Linear}
   n46 = "Transform (Modify) Bone - Bone: dyn_loincloth_fr_l_MSTR"(ComponentPose=Local To Component(LocalPose=Use cached pose 'Pose'()), Alpha=1, Translation="0.287,0.023,3", Rotation="-7.156,-2,0.966") {
    rotationMode=BMM_Replace
    rotationSpace=BCS_BoneSpace
    translationMode=BMM_Replace
    translationSpace=BCS_BoneSpace
    alphaBoolBlend {blendOption=Linear}
    alphaScaleBiasClamp {bInterpResult=true, clampMax=0.6, interpSpeedDecreasing=5, interpSpeedIncreasing=7}
    boneToModify {boneName=dyn_loincloth_fr_l_MSTR}
    rotation=<rot -7.156 -2 0.966>
    translation=<0.287 0.023 3>
   }
   n47 = "Transform (Modify) Bone - Bone: dyn_loincloth_fr_r_MSTR"(ComponentPose=@n46, Alpha=1, Translation="0.963,0.076,0.1", Rotation="-4.686,-5.025,-0.043") {
    rotationMode=BMM_Replace
    rotationSpace=BCS_BoneSpace
    translationMode=BMM_Replace
    translationSpace=BCS_BoneSpace
    alphaBoolBlend {blendOption=Linear}
    alphaScaleBiasClamp {bInterpResult=true, clampMax=0.6, interpSpeedDecreasing=5, interpSpeedIncreasing=7}
    boneToModify {boneName=dyn_loincloth_fr_r_MSTR}
    rotation=<rot -4.686 -5.025 -0.043>
    translation=<0.963 0.076 0.1>
   }
   n48 = "Transform (Modify) Bone - Bone: dyn_loincloth_fr_l_04"(ComponentPose=@n47, Alpha=1, Translation="0,0,0", Rotation="2.878,-20.407,0.002") {
    rotationMode=BMM_Replace
    rotationSpace=BCS_BoneSpace
    translationMode=BMM_Replace
    translationSpace=BCS_BoneSpace
    alphaBoolBlend {blendOption=Linear}
    alphaScaleBiasClamp {bInterpResult=true, clampMax=0.6, interpSpeedDecreasing=5, interpSpeedIncreasing=7}
    boneToModify {boneName=dyn_loincloth_fr_l_04}
    rotation=<rot 2.878 -20.407 0.002>
   }
   n49 = "Transform (Modify) Bone - Bone: dyn_loincloth_fr_r_04"(ComponentPose=@n48, Alpha=1, Translation="0,0,0", Rotation="0.003,12.683,-0.003") {
    rotationMode=BMM_Replace
    rotationSpace=BCS_BoneSpace
    translationMode=BMM_Replace
    translationSpace=BCS_BoneSpace
    alphaBoolBlend {blendOption=Linear}
    alphaScaleBiasClamp {bInterpResult=true, clampMax=0.6, interpSpeedDecreasing=5, interpSpeedIncreasing=7}
    boneToModify {boneName=dyn_loincloth_fr_r_04}
    rotation=<rot 0.003 12.683 -0.003>
   }
   n50 = Get thigh_rot_l()
   n51 = "Constraint - dyn_sleeve_l_02 (P)"(ComponentPose=@n49, Alpha=@n50, ConstraintWeights_0=1) {
    alphaBoolBlend {blendOption=Linear}
    alphaScaleBiasClamp {bClampResult=true, bInterpResult=true, clampMax=0.5, interpSpeedDecreasing=5, interpSpeedIncreasing=7}
    boneToModify {boneName=dyn_sleeve_l_02}
    constraintSetup [{targetBone {boneName=thigh_l}}]
    constraintWeights [1]
   }
   n52 = "Constraint - dyn_sleeve_l_03 (P)"(ComponentPose=@n51, Alpha=@n50, ConstraintWeights_0=1) {
    alphaBoolBlend {blendOption=Linear}
    alphaScaleBiasClamp {bClampResult=true, bInterpResult=true, clampMax=0.2, interpSpeedDecreasing=5, interpSpeedIncreasing=7}
    boneToModify {boneName=dyn_sleeve_l_03}
    constraintSetup [{targetBone {boneName=thigh_l}}]
    constraintWeights [1]
   }
   n53 = "Constraint - dyn_loincloth_fr_l_02 (PPPPPPPPPP)"!(
    ComponentPose=@n52
    Alpha=@n50
    ConstraintWeights_0=0.2424
    ConstraintWeights_1:float=0
    ConstraintWeights_2:float=0.0002
    ConstraintWeights_3:float=0
    ConstraintWeights_4:float=0.0001
    ConstraintWeights_5:float=0.6202
    ConstraintWeights_6:float=0.0023
    ConstraintWeights_7:float=0.1302
    ConstraintWeights_8:float=0.0034
    ConstraintWeights_9:float=0.0011
   ) -> (Pose) {
    alphaBoolBlend {blendOption=Linear}
    boneToModify {boneName=dyn_loincloth_fr_l_02}
    constraintSetup [
     {targetBone {boneName=pelvis}}
     {targetBone {boneName=spine_01}}
     {targetBone {boneName=spine_02}}
     {targetBone {boneName=spine_03}}
     {targetBone {boneName=thigh_l}}
     {targetBone {boneName=thigh_twist_01_l}}
     {targetBone {boneName=deform_glute_l}}
     {targetBone {boneName=deform_groin_l}}
     {targetBone {boneName=thigh_twist_01_r}}
     {targetBone {boneName=deform_groin_r}}
    ]
    constraintWeights [1,1,1,1,1,1,1,1,1,1]
   }
   n54 = Get thigh_rot_r()
   n55 = "Constraint - dyn_sleeve_r_02 (P)"(ComponentPose=@n53, Alpha=@n54, ConstraintWeights_0=1) {
    alphaBoolBlend {blendOption=Linear}
    alphaScaleBiasClamp {bClampResult=true, bInterpResult=true, clampMax=0.5, interpSpeedDecreasing=5, interpSpeedIncreasing=7}
    boneToModify {boneName=dyn_sleeve_r_02}
    constraintSetup [{targetBone {boneName=thigh_r}}]
    constraintWeights [1]
   }
   n56 = "Constraint - dyn_sleeve_r_03 (P)"(ComponentPose=@n55, Alpha=@n54, ConstraintWeights_0=1) {
    alphaBoolBlend {blendOption=Linear}
    alphaScaleBiasClamp {bClampResult=true, bInterpResult=true, clampMax=0.2, interpSpeedDecreasing=5, interpSpeedIncreasing=7}
    boneToModify {boneName=dyn_sleeve_r_03}
    constraintSetup [{targetBone {boneName=thigh_r}}]
    constraintWeights [1]
   }
   n57 = "Constraint - dyn_loincloth_fr_r_02 (PPPPPPPPPPPPPP)"!(
    ComponentPose=@n56
    Alpha=@n54
    ConstraintWeights_0=0.265
    ConstraintWeights_1:float=0.0003
    ConstraintWeights_2:float=0.0004
    ConstraintWeights_3:float=0.0002
    ConstraintWeights_4:float=0.0018
    ConstraintWeights_5:float=0.0026
    ConstraintWeights_6:float=0
    ConstraintWeights_7:float=0.0002
    ConstraintWeights_8:float=0.0011
    ConstraintWeights_9:float=0.3921
    ConstraintWeights_10:float=0.0864
    ConstraintWeights_11:float=0.0048
    ConstraintWeights_12:float=0.1778
    ConstraintWeights_13:float=0.0672
   ) -> (Pose) {
    alphaBoolBlend {blendOption=Linear}
    boneToModify {boneName=dyn_loincloth_fr_r_02}
    constraintSetup [
     {targetBone {boneName=pelvis}}
     {targetBone {boneName=spine_01}}
     {targetBone {boneName=spine_02}}
     {targetBone {boneName=spine_03}}
     {targetBone {boneName=thigh_l}}
     {targetBone {boneName=thigh_twist_01_l}}
     {targetBone {boneName=deform_glute_l}}
     {targetBone {boneName=deform_groin_base_l}}
     {targetBone {boneName=deform_groin_l}}
     {targetBone {boneName=thigh_r}}
     {targetBone {boneName=thigh_twist_01_r}}
     {targetBone {boneName=deform_glute_r}}
     {targetBone {boneName=deform_groin_base_r}}
     {targetBone {boneName=deform_groin_r}}
    ]
    constraintWeights [1,1,1,1,1,1,1,1,1,1,1,1,1,1]
   }
   n58 = "Constraint - dyn_loincloth_fr_r_03 (PPPP)"!(
    ComponentPose=@n57
    Alpha=@n54
    ConstraintWeights_0=0.067
    ConstraintWeights_1:float=0
    ConstraintWeights_2:float=0.8186
    ConstraintWeights_3:float=0.1144
   ) -> (Pose) {
    alphaBoolBlend {blendOption=Linear}
    boneToModify {boneName=dyn_loincloth_fr_r_03}
    constraintSetup [
     {targetBone {boneName=pelvis}}
     {targetBone {boneName=thigh_l}}
     {targetBone {boneName=thigh_r}}
     {targetBone {boneName=deform_groin_base_r}}
    ]
    constraintWeights [1,1,1,1]
   }
   n59 = "Constraint - dyn_loincloth_fr_r_04 (PPPP)"!(
    ComponentPose=@n58
    Alpha=@n54
    ConstraintWeights_0=0.0153
    ConstraintWeights_1:float=0.9397
    ConstraintWeights_2:float=0.0034
    ConstraintWeights_3:float=0.0417
   ) -> (Pose) {
    alphaBoolBlend {blendOption=Linear}
    boneToModify {boneName=dyn_loincloth_fr_r_04}
    constraintSetup [{targetBone {boneName=pelvis}}, {targetBone {boneName=thigh_r}}, {targetBone {boneName=calf_r}}, {targetBone {boneName=deform_groin_base_r}}]
    constraintWeights [1,1,1,1]
   }
   n60 = "Constraint - dyn_loincloth_bk_l_MSTR (P)"(ComponentPose=@n59, Alpha=0.5, ConstraintWeights_0=1) {
    alphaBoolBlend {blendOption=Linear}
    alphaScaleBiasClamp {bClampResult=true, bInterpResult=true, interpSpeedDecreasing=5, interpSpeedIncreasing=7}
    boneToModify {boneName=dyn_loincloth_bk_l_MSTR}
    constraintSetup [{targetBone {boneName=thigh_l}}]
    constraintWeights [1]
   }
   n61 = "Constraint - dyn_loincloth_bk_MSTR (PP)"!(ComponentPose=@n60, Alpha=0.2, ConstraintWeights_0=0.5, ConstraintWeights_1:float=0.5) -> (Pose) {
    alphaBoolBlend {blendOption=Linear}
    alphaScaleBiasClamp {bClampResult=true, bInterpResult=true, interpSpeedDecreasing=5, interpSpeedIncreasing=7}
    boneToModify {boneName=dyn_loincloth_bk_MSTR}
    constraintSetup [{targetBone {boneName=thigh_r}}, {targetBone {boneName=thigh_l}}]
    constraintWeights [1,1]
   }
   n62 = "Constraint - dyn_loincloth_bk_r_MSTR (P)"(ComponentPose=@n61, Alpha=0.5, ConstraintWeights_0=1) {
    alphaBoolBlend {blendOption=Linear}
    alphaScaleBiasClamp {bClampResult=true, bInterpResult=true, interpSpeedDecreasing=5, interpSpeedIncreasing=7}
    boneToModify {boneName=dyn_loincloth_bk_r_MSTR}
    constraintSetup [{targetBone {boneName=thigh_r}}]
    constraintWeights [1]
   }
   n63 = "Transform (Modify) Bone - Bone: dyn_loincloth_bk_l_MSTR"(ComponentPose=@n62, Alpha=1, Translation="0,0,5", Rotation="20,0,0") {
    rotationMode=BMM_Replace
    rotationSpace=BCS_BoneSpace
    translationMode=BMM_Replace
    translationSpace=BCS_BoneSpace
    alphaBoolBlend {blendOption=Linear}
    alphaScaleBiasClamp {bInterpResult=true, clampMax=0.6, interpSpeedDecreasing=5, interpSpeedIncreasing=7}
    boneToModify {boneName=dyn_loincloth_bk_l_MSTR}
    rotation {pitch=20}
    translation {z=5}
   }
   n64 = "Transform (Modify) Bone - Bone: dyn_loincloth_bk_MSTR"(ComponentPose=@n63, Alpha=1, Translation="0,0,5", Rotation="15,0,0") {
    rotationMode=BMM_Replace
    rotationSpace=BCS_BoneSpace
    translationMode=BMM_Replace
    translationSpace=BCS_BoneSpace
    alphaBoolBlend {blendOption=Linear}
    alphaScaleBiasClamp {bInterpResult=true, clampMax=0.6, interpSpeedDecreasing=5, interpSpeedIncreasing=7}
    boneToModify {boneName=dyn_loincloth_bk_MSTR}
    rotation {pitch=15}
    translation {z=5}
   }
   n65 = "Transform (Modify) Bone - Bone: dyn_loincloth_bk_r_MSTR"(ComponentPose=@n64, Alpha=1, Translation="0,0,5", Rotation="20,0,0") {
    rotationMode=BMM_Replace
    rotationSpace=BCS_BoneSpace
    translationMode=BMM_Replace
    translationSpace=BCS_BoneSpace
    alphaBoolBlend {blendOption=Linear}
    alphaScaleBiasClamp {bInterpResult=true, clampMax=0.6, interpSpeedDecreasing=5, interpSpeedIncreasing=7}
    boneToModify {boneName=dyn_loincloth_bk_r_MSTR}
    rotation {pitch=20}
    translation {z=5}
   }
   n66 = Component To Local(ComponentPose=@n65)
   n67 = Get bIsPlayingEmote()
   n68 = "Transform (Modify) Bone - Bone: dyn_loincloth_fr_l_MSTR"(
    ComponentPose=Local To Component(LocalPose=Use cached pose 'Pose'())
    Alpha=Get thigh_rot_l()
    Translation="-4.592,-5.547,5.930"
    Rotation="-19.809,-1.924,0"
   ) {
    rotationMode=BMM_Replace
    rotationSpace=BCS_BoneSpace
    translationMode=BMM_Replace
    translationSpace=BCS_BoneSpace
    alphaBoolBlend {blendOption=Linear}
    alphaScaleBiasClamp {bClampResult=true, bInterpResult=true, clampMax=0.85, interpSpeedDecreasing=5, interpSpeedIncreasing=7}
    boneToModify {boneName=dyn_loincloth_fr_l_MSTR}
    rotation {pitch=-19.809, yaw=-1.924}
    translation=<-4.592 -5.547 5.93>
   }
   n69 = "Transform (Modify) Bone - Bone: dyn_loincloth_fr_r_MSTR"(ComponentPose=@n68, Alpha=Get thigh_rot_r(), Translation="-0.293,1.396,1.101", Rotation="14.982,0.036,-0.011") {
    rotationMode=BMM_Replace
    rotationSpace=BCS_BoneSpace
    translationMode=BMM_Replace
    translationSpace=BCS_BoneSpace
    alphaBoolBlend {blendOption=Linear}
    alphaScaleBiasClamp {bClampResult=true, bInterpResult=true, clampMax=0.85, interpSpeedDecreasing=5, interpSpeedIncreasing=7}
    boneToModify {boneName=dyn_loincloth_fr_r_MSTR}
    rotation=<rot 14.982 0.036 -0.011>
    translation=<-0.293 1.396 1.101>
   }
   n70 = "Transform (Modify) Bone - Bone: dyn_loincloth_fr_r_MSTR"!(ComponentPose=@n69, Alpha=1, Rotation="19.934,0.814,-1.030", Scale:Vector="0.750,1.000,1.000") -> (Pose) {
    rotationMode=BMM_Additive
    rotationSpace=BCS_BoneSpace
    scaleMode=BMM_Additive
    scaleSpace=BCS_BoneSpace
    translationSpace=BCS_BoneSpace
    alphaBoolBlend {blendOption=Linear}
    alphaScaleBiasClamp {bClampResult=true, bInterpResult=true, clampMax=0.45, interpSpeedDecreasing=5, interpSpeedIncreasing=7}
    boneToModify {boneName=dyn_loincloth_fr_r_MSTR}
    rotation=<rot 19.934 0.814 -1.03>
    scale {x=0.75}
    translation=<-7.723 3.33 0.199>
   }
   n71 = "Transform (Modify) Bone - Bone: dyn_loincloth_fr_l_02"!(ComponentPose=@n70, Alpha=1, Translation="0.5,3,-3", Rotation="0,0,-45") -> (Pose) {
    rotationMode=BMM_Additive
    rotationSpace=BCS_BoneSpace
    scaleSpace=BCS_BoneSpace
    translationMode=BMM_Additive
    translationSpace=BCS_BoneSpace
    alphaBoolBlend {blendOption=Linear}
    alphaScaleBiasClamp {bClampResult=true, bInterpResult=true, clampMax=0.45, interpSpeedDecreasing=5, interpSpeedIncreasing=7}
    boneToModify {boneName=dyn_loincloth_fr_l_02}
    rotation {roll=-45}
    translation=<0.5 3 -3>
   }
   n72 = "Transform (Modify) Bone - Bone: dyn_loincloth_side_r_MSTR"!(ComponentPose=@n71, Alpha=1, Rotation="30,10,0", Scale:Vector="0.750,1.000,1.000") -> (Pose) {
    rotationMode=BMM_Additive
    rotationSpace=BCS_BoneSpace
    scaleMode=BMM_Additive
    scaleSpace=BCS_BoneSpace
    translationSpace=BCS_BoneSpace
    alphaBoolBlend {blendOption=Linear}
    alphaScaleBiasClamp {bClampResult=true, bInterpResult=true, clampMax=0.45, interpSpeedDecreasing=5, interpSpeedIncreasing=7}
    boneToModify {boneName=dyn_loincloth_side_r_MSTR}
    rotation {pitch=30, yaw=10}
    scale {x=0.75}
    translation=<-7.723 3.33 0.199>
   }
   n73 = "Transform (Modify) Bone - Bone: dyn_loincloth_fr_r_03"!(ComponentPose=@n72, Alpha=1, Rotation="35.366,-1.202,-0.017") -> (Pose) {
    rotationMode=BMM_Additive
    rotationSpace=BCS_BoneSpace
    scaleMode=BMM_Additive
    scaleSpace=BCS_BoneSpace
    translationSpace=BCS_BoneSpace
    alphaBoolBlend {blendOption=Linear}
    alphaScaleBiasClamp {bClampResult=true, bInterpResult=true, clampMax=0.45, interpSpeedDecreasing=5, interpSpeedIncreasing=7}
    boneToModify {boneName=dyn_loincloth_fr_r_03}
    rotation=<rot 35.366 -1.202 -0.017>
    translation=<-7.723 3.33 0.199>
   }
   n74 = Control Rig(Source=Component To Local(ComponentPose=@n73)) {
    alphaBoolBlend {blendOption=Linear}
    alphaScaleBiasClamp {bInterpResult=true, interpSpeedDecreasing=5, interpSpeedIncreasing=7}
    controlRigAssetReference {blueprintRigClass=&/M_MED_Jonesy_Orange_CrouchCtrlRig.M_MED_Jonesy_Orange_CrouchCtrlRig_C}
   }
   n75 = Get Right_Direct()
   n76 = Look At - Bone: dyn_sleeve_l_MSTR(ComponentPose=Local To Component(LocalPose=@n74), Alpha=@n75) {
    alphaBoolBlend {blendOption=Linear}
    alphaScaleBiasClamp {bInterpResult=true, interpSpeedDecreasing=5, interpSpeedIncreasing=7}
    boneToModify {boneName=dyn_sleeve_l_MSTR}
    lookAtLocation=<-44.974 37.5788 -2.5075>
    lookAtTarget {boneReference {boneName=pelvis}}
    lookAt_Axis {axis=<1 0>}
   }
   n77 = Look At - Bone: dyn_sleeve_r_MSTR(ComponentPose=@n76, Alpha=@n75) {
    alphaBoolBlend {blendOption=Linear}
    alphaScaleBiasClamp {bInterpResult=true, interpSpeedDecreasing=5, interpSpeedIncreasing=7}
    boneToModify {boneName=dyn_sleeve_r_MSTR}
    lookAtLocation=<-44.9911 18.5754 19.5996>
    lookAtTarget {boneReference {boneName=pelvis}}
    lookAt_Axis {axis=<1 0>}
   }
   n78 = Get Left_Direct()
   n79 = Look At - Bone: dyn_sleeve_l_MSTR(ComponentPose=@n77, Alpha=@n78) {
    alphaBoolBlend {blendOption=Linear}
    alphaScaleBiasClamp {bInterpResult=true, interpSpeedDecreasing=5, interpSpeedIncreasing=7}
    boneToModify {boneName=dyn_sleeve_l_MSTR}
    lookAtLocation=<-79.0825 38.0676 -59.3529>
    lookAtTarget {boneReference {boneName=pelvis}}
    lookAt_Axis {axis=<1 0>}
   }
   n80 = Look At - Bone: dyn_sleeve_r_MSTR(ComponentPose=@n79, Alpha=@n78) {
    alphaBoolBlend {blendOption=Linear}
    alphaScaleBiasClamp {bInterpResult=true, interpSpeedDecreasing=5, interpSpeedIncreasing=7}
    boneToModify {boneName=dyn_sleeve_r_MSTR}
    lookAtLocation=<-98.1936 -9.4217 -58.3129>
    lookAtTarget {boneReference {boneName=pelvis}}
    lookAt_Axis {axis=<1 0>}
   }
   n81 = Get Backward_Direct()
   n82 = Look At - Bone: dyn_loincloth_fr_r_MSTR(ComponentPose=@n80, Alpha=@n81) {
    alphaBoolBlend {blendOption=Linear}
    alphaScaleBiasClamp {bInterpResult=true, interpSpeedDecreasing=5, interpSpeedIncreasing=7}
    boneToModify {boneName=dyn_loincloth_fr_r_MSTR}
    lookAtLocation=<16.716 29.9814 16.8587>
    lookAtTarget {boneReference {boneName=thigh_r}}
    lookAt_Axis {axis=<1 0>}
   }
   n83 = Look At - Bone: dyn_loincloth_side_r_MSTR(ComponentPose=@n82, Alpha=@n81) {
    alphaBoolBlend {blendOption=Linear}
    alphaScaleBiasClamp {bInterpResult=true, interpSpeedDecreasing=5, interpSpeedIncreasing=7}
    boneToModify {boneName=dyn_loincloth_side_r_MSTR}
    lookAtLocation=<5.325 26.0999 21.2892>
    lookAtTarget {boneReference {boneName=thigh_r}}
    lookAt_Axis {axis=<1 0>}
   }
   n84 = "Constraint - dyn_loincloth_bk_l_02 (P)"(ComponentPose=@n83, Alpha=0.3, ConstraintWeights_0=1) {
    alphaBoolBlend {blendOption=Linear}
    boneToModify {boneName=dyn_loincloth_bk_l_02}
    constraintSetup [{targetBone {boneName=dyn_loincloth_bk_02}}]
    constraintWeights [1]
   }
   n85 = "Constraint - dyn_loincloth_bk_r_02 (P)"(ComponentPose=@n84, Alpha=0.45, ConstraintWeights_0=1) {
    alphaBoolBlend {blendOption=Linear}
    boneToModify {boneName=dyn_loincloth_bk_r_02}
    constraintSetup [{targetBone {boneName=dyn_loincloth_bk_02}}]
    constraintWeights [1]
   }
   n86 = "Constraint - dyn_loincloth_bk_l_03 (P)"(ComponentPose=@n85, Alpha=0.3, ConstraintWeights_0=1) {
    alphaBoolBlend {blendOption=Linear}
    boneToModify {boneName=dyn_loincloth_bk_l_03}
    constraintSetup [{targetBone {boneName=dyn_loincloth_bk_03}}]
    constraintWeights [1]
   }
   n87 = "Constraint - dyn_loincloth_bk_r_03 (P)"(ComponentPose=@n86, Alpha=0.45, ConstraintWeights_0=1) {
    alphaBoolBlend {blendOption=Linear}
    boneToModify {boneName=dyn_loincloth_bk_r_03}
    constraintSetup [{targetBone {boneName=dyn_loincloth_bk_03}}]
    constraintWeights [1]
   }
   n88 = Look At - Bone: dyn_sleeve_l_MSTR(ComponentPose=Local To Component(LocalPose=Use cached pose 'Pose'()), Alpha=1) {
    alphaBoolBlend {blendOption=Linear}
    alphaScaleBiasClamp {bInterpResult=true, interpSpeedDecreasing=5, interpSpeedIncreasing=7}
    boneToModify {boneName=dyn_sleeve_l_MSTR}
    lookAtLocation=<-64.804 -25.7071 -307.9073>
    lookAtTarget {boneReference {boneName=pelvis}}
    lookAt_Axis {axis=<1 0>}
   }
   n89 = Look At - Bone: dyn_sleeve_r_MSTR(ComponentPose=@n88, Alpha=1) {
    alphaBoolBlend {blendOption=Linear}
    alphaScaleBiasClamp {bInterpResult=true, interpSpeedDecreasing=5, interpSpeedIncreasing=7}
    boneToModify {boneName=dyn_sleeve_r_MSTR}
    lookAtLocation=<-57.548 8.0847 70.0566>
    lookAtTarget {boneReference {boneName=pelvis}}
    lookAt_Axis {axis=<1 0>}
   }
   n90 = "Transform (Modify) Bone - Bone: dyn_loincloth_side_l_MSTR"!(ComponentPose=@n89, Alpha=1, Translation:Vector="0,0,1.5", Rotation="10,0,0") -> (Pose) {
    rotationMode=BMM_Replace
    rotationSpace=BCS_BoneSpace
    scaleMode=BMM_Additive
    scaleSpace=BCS_BoneSpace
    translationMode=BMM_Additive
    translationSpace=BCS_BoneSpace
    alphaBoolBlend {blendOption=Linear}
    boneToModify {boneName=dyn_loincloth_side_l_MSTR}
    rotation {pitch=10}
    translation {z=1.5}
   }
   n91 = "Transform (Modify) Bone - Bone: dyn_loincloth_bk_l_MSTR"(ComponentPose=@n90, Alpha=1, Translation="0,0,1", Rotation="10,0,0") {
    rotationMode=BMM_Replace
    rotationSpace=BCS_BoneSpace
    scaleMode=BMM_Additive
    scaleSpace=BCS_BoneSpace
    translationMode=BMM_Additive
    translationSpace=BCS_BoneSpace
    alphaBoolBlend {blendOption=Linear}
    boneToModify {boneName=dyn_loincloth_bk_l_MSTR}
    rotation {pitch=10}
    translation {z=1}
   }
   n92 = "Transform (Modify) Bone - Bone: dyn_loincloth_side_r_MSTR"!(ComponentPose=@n91, Alpha=1, Translation:Vector="0,0,1.5", Rotation="10,0,0") -> (Pose) {
    rotationMode=BMM_Replace
    rotationSpace=BCS_BoneSpace
    scaleMode=BMM_Additive
    scaleSpace=BCS_BoneSpace
    translationMode=BMM_Additive
    translationSpace=BCS_BoneSpace
    alphaBoolBlend {blendOption=Linear}
    boneToModify {boneName=dyn_loincloth_side_r_MSTR}
    rotation {pitch=10}
    translation {z=1.5}
   }
   n93 = "Transform (Modify) Bone - Bone: dyn_loincloth_bk_r_MSTR"(ComponentPose=@n92, Alpha=1, Translation="0,0,1", Rotation="10,0,0") {
    rotationMode=BMM_Replace
    rotationSpace=BCS_BoneSpace
    scaleMode=BMM_Additive
    scaleSpace=BCS_BoneSpace
    translationMode=BMM_Additive
    translationSpace=BCS_BoneSpace
    alphaBoolBlend {blendOption=Linear}
    boneToModify {boneName=dyn_loincloth_bk_r_MSTR}
    rotation {pitch=10}
    translation {z=1}
   }
   n94 = Blend Poses by bool(
    BlendPose_0=Component To Local(ComponentPose=@n93)
    BlendPose_1=Blend Poses by bool(
     BlendPose_0=Component To Local(ComponentPose=@n87)
     BlendPose_1=Blend Poses by bool(BlendPose_0=@n66, BlendPose_1=@n45, bActiveValue=@n67) {blendType=Linear}
     bActiveValue=Get bIsCrouching()
    ) {blendType=Linear}
    bActiveValue=Get bIsSkyDiving()
   ) {blendType=Linear}
   n95 = "Transform (Modify) Bone - Bone: dyn_sleeve_l_MSTR"!(
    ComponentPose=Local To Component(LocalPose=Use cached pose 'Pose'())
    Alpha=1
    Translation:Vector="-2.924,-3.419,-4.674"
    Rotation="47.557,-10.528,-31.369"
    Scale:Vector="1,1,1"
   ) -> (Pose) {
    rotationMode=BMM_Replace
    rotationSpace=BCS_BoneSpace
    scaleMode=BMM_Additive
    scaleSpace=BCS_BoneSpace
    translationMode=BMM_Additive
    translationSpace=BCS_BoneSpace
    alphaBoolBlend {blendOption=Linear}
    boneToModify {boneName=dyn_sleeve_l_MSTR}
    rotation=<rot 47.557 -10.528 -31.369>
    translation=<-2.924 -3.419 -4.674>
   }
   n96 = "Transform (Modify) Bone - Bone: dyn_sleeve_r_MSTR"!(ComponentPose=@n95, Alpha=1, Translation:Vector="-5.003,2.418,-5.427", Rotation="31.673,29.657,14.266", Scale:Vector="1,1,1") -> (Pose) {
    rotationMode=BMM_Replace
    rotationSpace=BCS_BoneSpace
    scaleMode=BMM_Additive
    scaleSpace=BCS_BoneSpace
    translationMode=BMM_Additive
    translationSpace=BCS_BoneSpace
    alphaBoolBlend {blendOption=Linear}
    boneToModify {boneName=dyn_sleeve_r_MSTR}
    rotation=<rot 31.673 29.657 14.266>
    translation=<-5.003 2.418 -5.427>
   }
   n97 = "Transform (Modify) Bone - Bone: dyn_loincloth_side_l_MSTR"!(ComponentPose=@n96, Alpha=1, Translation="5.762,-9.616,0.991", Rotation="-5.777,-29.459,-8.110", Scale:Vector="0.8,1,1") -> (Pose) {
    rotationMode=BMM_Replace
    rotationSpace=BCS_BoneSpace
    scaleMode=BMM_Additive
    scaleSpace=BCS_BoneSpace
    translationMode=BMM_Additive
    translationSpace=BCS_BoneSpace
    alphaBoolBlend {blendOption=Linear}
    boneToModify {boneName=dyn_loincloth_side_l_MSTR}
    rotation=<rot -5.777 -29.459 -8.11>
    scale {x=0.8}
    translation=<5.762 -9.616 0.991>
   }
   n98 = "Transform (Modify) Bone - Bone: dyn_loincloth_bk_l_MSTR"!(ComponentPose=@n97, Alpha=1, Translation="5.396,-8.331,7.567", Rotation="-14.829,-0.169,37.521", Scale:Vector="0.8,1,1") -> (Pose) {
    rotationMode=BMM_Replace
    rotationSpace=BCS_BoneSpace
    scaleMode=BMM_Additive
    scaleSpace=BCS_BoneSpace
    translationMode=BMM_Additive
    translationSpace=BCS_BoneSpace
    alphaBoolBlend {blendOption=Linear}
    boneToModify {boneName=dyn_loincloth_bk_l_MSTR}
    rotation=<rot -14.829 -0.169 37.521>
    scale {x=0.8}
    translation=<5.396 -8.331 7.567>
   }
   n99 = "Transform (Modify) Bone - Bone: dyn_loincloth_side_r_MSTR"!(ComponentPose=@n98, Alpha=1, Translation="0,0,0", Rotation="-14.470,31.620,7.887", Scale:Vector="0.8,1,1") -> (Pose) {
    rotationMode=BMM_Replace
    rotationSpace=BCS_BoneSpace
    scaleMode=BMM_Additive
    scaleSpace=BCS_BoneSpace
    translationMode=BMM_Additive
    translationSpace=BCS_BoneSpace
    alphaBoolBlend {blendOption=Linear}
    boneToModify {boneName=dyn_loincloth_side_r_MSTR}
    rotation=<rot -14.47 31.62 7.887>
    scale {x=0.8}
   }
   n100 = "Transform (Modify) Bone - Bone: dyn_loincloth_bk_r_MSTR"!(ComponentPose=@n99, Alpha=1, Translation="0.933,-0.301,2.608", Rotation="-0.041,8.713,0.182", Scale:Vector="0.8,1,1") -> (Pose) {
    rotationMode=BMM_Replace
    rotationSpace=BCS_BoneSpace
    scaleMode=BMM_Additive
    scaleSpace=BCS_BoneSpace
    translationMode=BMM_Additive
    translationSpace=BCS_BoneSpace
    alphaBoolBlend {blendOption=Linear}
    boneToModify {boneName=dyn_loincloth_bk_r_MSTR}
    rotation=<rot -0.041 8.713 0.182>
    scale {x=0.8}
    translation=<0.933 -0.301 2.608>
   }
   n101 = "Transform (Modify) Bone - Bone: dyn_loincloth_fr_l_MSTR"!(ComponentPose=@n100, Alpha=1, Translation="1.275,-0.171,1.827", Rotation="-37.717,3.335,1.341", Scale:Vector="0.8,1,1") -> (Pose) {
    rotationMode=BMM_Replace
    rotationSpace=BCS_BoneSpace
    scaleMode=BMM_Additive
    scaleSpace=BCS_BoneSpace
    translationMode=BMM_Additive
    translationSpace=BCS_BoneSpace
    alphaBoolBlend {blendOption=Linear}
    boneToModify {boneName=dyn_loincloth_fr_l_MSTR}
    rotation=<rot -37.717 3.335 1.341>
    scale {x=0.8}
    translation=<1.275 -0.171 1.827>
   }
   n102 = "Transform (Modify) Bone - Bone: dyn_loincloth_side_l_MSTR"(ComponentPose=@n101, Alpha=1, Translation="0,0,0", Rotation="-8.136,22.692,-1.000", Scale="0.8,1,1") {
    rotationMode=BMM_Replace
    rotationSpace=BCS_BoneSpace
    scaleMode=BMM_Additive
    scaleSpace=BCS_BoneSpace
    translationMode=BMM_Additive
    translationSpace=BCS_BoneSpace
    alphaBoolBlend {blendOption=Linear}
    boneToModify {boneName=dyn_loincloth_side_l_MSTR}
    rotation=<rot -8.136 22.692 -1>
    scale {x=0.8}
   }
   n103 = "Transform (Modify) Bone - Bone: dyn_loincloth_fr_r_MSTR"!(ComponentPose=@n102, Alpha=1, Translation:Vector="-8.067,1.554,3.378", Rotation="-24.977,5.605,-47.512", Scale="0.8,1,1") -> (Pose) {
    rotationMode=BMM_Replace
    rotationSpace=BCS_BoneSpace
    scaleMode=BMM_Additive
    scaleSpace=BCS_BoneSpace
    translationMode=BMM_Additive
    translationSpace=BCS_BoneSpace
    alphaBoolBlend {blendOption=Linear}
    boneToModify {boneName=dyn_loincloth_fr_r_MSTR}
    rotation=<rot -24.977 5.605 -47.512>
    scale {x=0.8}
    translation=<-8.067 1.554 3.378>
   }
   n104 = "Transform (Modify) Bone - Bone: dyn_loincloth_side_r_MSTR"(ComponentPose=@n103, Alpha=1, Translation="-1.794,-0.032,1.998", Rotation="-14.151,-25.621,-1.135", Scale="0.8,1,1") {
    rotationMode=BMM_Replace
    rotationSpace=BCS_BoneSpace
    scaleMode=BMM_Additive
    scaleSpace=BCS_BoneSpace
    translationMode=BMM_Additive
    translationSpace=BCS_BoneSpace
    alphaBoolBlend {blendOption=Linear}
    boneToModify {boneName=dyn_loincloth_side_r_MSTR}
    rotation=<rot -14.151 -25.621 -1.135>
    scale {x=0.8}
    translation=<-1.794 -0.032 1.998>
   }
   n105 = Look At - Bone: dyn_loincloth_bk_l_MSTR(ComponentPose=@n104, Alpha=1) {
    alphaBoolBlend {blendOption=Linear}
    boneToModify {boneName=dyn_loincloth_bk_l_MSTR}
    lookAtLocation=<-28.6895 -9.664 -16.1935>
    lookAtTarget {boneReference {boneName=pelvis}}
    lookAt_Axis {axis=<1 0>}
   }
   n106 = Look At - Bone: dyn_loincloth_bk_r_MSTR(ComponentPose=@n105, Alpha=1) {
    alphaBoolBlend {blendOption=Linear}
    boneToModify {boneName=dyn_loincloth_bk_r_MSTR}
    lookAtLocation=<14.652 -25.7815 17.4539>
    lookAtTarget {boneReference {boneName=pelvis}}
   }
   n107 = Component To Local(ComponentPose=@n106)
   n108 = Get bIsDBNO()
   n109 = Anim Dynamics - Bone: dyn_cuff_l(
    ComponentPose=Local To Component(LocalPose=Blend Poses by bool(BlendPose_0=@n107, BlendPose_1=@n94, bActiveValue=@n108) {blendType=Linear})
    Alpha=1
    ExternalForce="0,0,0"
   ) {
    angularDampingOverride=1
    angularSpringConstant=20000
    bOverrideLinearDamping=true
    bUseGravityOverride=true
    gravityScale=0.35
    linearDampingOverride=1
    linearSpringConstant=100
    simulationSpace=RootRelative
    alphaBoolBlend {blendOption=Linear}
    boundBone {boneName=dyn_cuff_l}
    gravityOverride {z=-980}
    physicsBodyDefinitions [{
      boundBone {boneName=dyn_cuff_l}
      boxExtents=<100 100 100>
      constraintSetup {angularLimitsMax=<0 0 0>, angularLimitsMin=<0 0 0>, angularTarget=<0 0 0>, linearAxesMax=<0.2 0 0>, linearAxesMin=<-0.5 0 0>}
      localJointOffset=<1 0 0>
     }]
    relativeSpaceBone {boneName=upperarm_r}
   }
   n110 = Anim Dynamics - Bone: dyn_cuff_r(ComponentPose=@n109, Alpha=1, ExternalForce="0,0,0") {
    angularDampingOverride=1
    angularSpringConstant=20000
    bOverrideLinearDamping=true
    bUseGravityOverride=true
    gravityScale=0.35
    linearDampingOverride=1
    linearSpringConstant=100
    simulationSpace=RootRelative
    alphaBoolBlend {blendOption=Linear}
    boundBone {boneName=dyn_cuff_r}
    gravityOverride {z=-980}
    physicsBodyDefinitions [{
      boundBone {boneName=dyn_cuff_r}
      boxExtents=<100 100 100>
      constraintSetup {angularLimitsMax=<0 0 0>, angularLimitsMin=<0 0 0>, angularTarget=<0 0 0>, linearAxesMax=<0.2 0 0>, linearAxesMin=<-0.5 0 0>}
      localJointOffset=<1 0 0>
     }]
    relativeSpaceBone {boneName=upperarm_r}
   }
   n111 = Anim Dynamics - Bone: dyn_chain_01(ComponentPose=@n110, Alpha=1, ExternalForce="0,0,0") {
    angularDampingOverride=0.9999
    angularSpringConstant=500
    bAngularSpring=true
    bGravityOverrideInSimSpace=true
    bOverrideAngularDamping=true
    simulationSpace=BoneRelative
    alphaBoolBlend {blendOption=Linear}
    boundBone {boneName=dyn_chain_01}
    componentAppliedLinearAccClamp=<5000 5000 5000>
    componentLinearAccScale=<1.2 1.2 1.2>
    componentLinearVelScale=<1.2 1.2 1.2>
    physicsBodyDefinitions [{
      sphereCollisionRadius=0
      boundBone {boneName=dyn_chain_01}
      boxExtents=<15 20 20>
      constraintSetup {angularLimitsMax=<1 5 30>, angularLimitsMin=<-1 -45 -35>, angularTarget=<0 0 0>}
      localJointOffset=<3 0 0>
     }]
    relativeSpaceBone {boneName=root}
   }
   n112 = Anim Dynamics - Bone: dyn_chain_02(ComponentPose=@n111, Alpha=1, ExternalForce="0,0,0") {
    angularDampingOverride=0.9999
    angularSpringConstant=500
    bAngularSpring=true
    bGravityOverrideInSimSpace=true
    bOverrideAngularDamping=true
    simulationSpace=BoneRelative
    alphaBoolBlend {blendOption=Linear}
    boundBone {boneName=dyn_chain_02}
    componentAppliedLinearAccClamp=<5000 5000 5000>
    componentLinearAccScale=<1.2 1.2 1.2>
    componentLinearVelScale=<1.2 1.2 1.2>
    physicsBodyDefinitions [{
      sphereCollisionRadius=0
      boundBone {boneName=dyn_chain_02}
      boxExtents=<15 20 20>
      constraintSetup {angularLimitsMax=<1 8 40>, angularLimitsMin=<-1 -45 -45>, angularTarget=<0 0 0>}
      localJointOffset=<5 0 0>
     }]
    relativeSpaceBone {boneName=root}
   }
   n113 = Anim Dynamics - Bone: dyn_shoetie_l_out_01(ComponentPose=@n112, Alpha=1, ExternalForce="0,0,0") {
    angularDampingOverride=0.9999
    angularSpringConstant=500
    bAngularSpring=true
    bGravityOverrideInSimSpace=true
    bOverrideAngularDamping=true
    simulationSpace=BoneRelative
    alphaBoolBlend {blendOption=Linear}
    boundBone {boneName=dyn_shoetie_l_out_01}
    componentAppliedLinearAccClamp=<3500 3500 3500>
    componentLinearAccScale=<1.2 1.2 1.2>
    componentLinearVelScale=<1.2 1.2 1.2>
    physicsBodyDefinitions [{
      sphereCollisionRadius=0
      boundBone {boneName=dyn_shoetie_l_out_01}
      boxExtents=<15 30 30>
      constraintSetup {angularLimitsMax=<1 5 15>, angularLimitsMin=<-1 -35 -15>, angularTarget=<0 0 0>}
      localJointOffset=<3 0 0>
     }]
    relativeSpaceBone {boneName=root}
   }
   n114 = Anim Dynamics - Bone: dyn_shoetie_l_out_02(ComponentPose=@n113, Alpha=1, ExternalForce="0,0,0") {
    angularDampingOverride=0.9999
    angularSpringConstant=500
    bAngularSpring=true
    bGravityOverrideInSimSpace=true
    bOverrideAngularDamping=true
    simulationSpace=BoneRelative
    alphaBoolBlend {blendOption=Linear}
    boundBone {boneName=dyn_shoetie_l_out_02}
    componentAppliedLinearAccClamp=<3500 3500 3500>
    componentLinearAccScale=<1.2 1.2 1.2>
    componentLinearVelScale=<1.2 1.2 1.2>
    physicsBodyDefinitions [{
      sphereCollisionRadius=0
      boundBone {boneName=dyn_shoetie_l_out_02}
      boxExtents=<15 30 30>
      constraintSetup {angularLimitsMax=<1 5 15>, angularLimitsMin=<-1 -35 -15>, angularTarget=<0 0 0>}
      localJointOffset=<3 0 0>
     }]
    relativeSpaceBone {boneName=root}
   }
   n115 = Anim Dynamics - Bone: dyn_shoelace_l_out_01(ComponentPose=@n114, Alpha=1, ExternalForce="0,0,0") {
    angularDampingOverride=0.9999
    angularSpringConstant=500
    bAngularSpring=true
    bGravityOverrideInSimSpace=true
    bOverrideAngularDamping=true
    simulationSpace=BoneRelative
    alphaBoolBlend {blendOption=Linear}
    boundBone {boneName=dyn_shoelace_l_out_01}
    componentAppliedLinearAccClamp=<5000 5000 5000>
    componentLinearAccScale=<1.2 1.2 1.2>
    componentLinearVelScale=<1.2 1.2 1.2>
    physicsBodyDefinitions [{
      sphereCollisionRadius=0
      boundBone {boneName=dyn_shoelace_l_out_01}
      boxExtents=<15 20 20>
      constraintSetup {angularLimitsMax=<1 10 20>, angularLimitsMin=<-1 -45 -60>, angularTarget=<0 0 0>}
      localJointOffset=<3 0 0>
     }]
    relativeSpaceBone {boneName=root}
   }
   n116 = Anim Dynamics - Bone: dyn_shoelace_l_out_02(ComponentPose=@n115, Alpha=1, ExternalForce="0,0,0") {
    angularDampingOverride=0.9999
    angularSpringConstant=500
    bAngularSpring=true
    bGravityOverrideInSimSpace=true
    bOverrideAngularDamping=true
    simulationSpace=BoneRelative
    alphaBoolBlend {blendOption=Linear}
    boundBone {boneName=dyn_shoelace_l_out_02}
    componentAppliedLinearAccClamp=<5000 5000 5000>
    componentLinearAccScale=<1.2 1.2 1.2>
    componentLinearVelScale=<1.2 1.2 1.2>
    physicsBodyDefinitions [{
      sphereCollisionRadius=0
      boundBone {boneName=dyn_shoelace_l_out_02}
      boxExtents=<15 20 20>
      constraintSetup {angularLimitsMax=<1 10 40>, angularLimitsMin=<-1 -45 -60>, angularTarget=<0 0 0>}
      localJointOffset=<3 0 0>
     }]
    relativeSpaceBone {boneName=root}
   }
   n117 = Anim Dynamics - Bone: dyn_shoelace_l_out_03(ComponentPose=@n116, Alpha=1, ExternalForce="0,0,0") {
    angularDampingOverride=0.9999
    angularSpringConstant=500
    bAngularSpring=true
    bGravityOverrideInSimSpace=true
    bOverrideAngularDamping=true
    simulationSpace=BoneRelative
    alphaBoolBlend {blendOption=Linear}
    boundBone {boneName=dyn_shoelace_l_out_03}
    componentAppliedLinearAccClamp=<5000 5000 5000>
    componentLinearAccScale=<1.2 1.2 1.2>
    componentLinearVelScale=<1.2 1.2 1.2>
    physicsBodyDefinitions [{
      sphereCollisionRadius=0
      boundBone {boneName=dyn_shoelace_l_out_03}
      boxExtents=<15 20 20>
      constraintSetup {angularLimitsMax=<1 10 20>, angularLimitsMin=<-1 -45 -60>, angularTarget=<0 0 0>}
      localJointOffset=<3 0 0>
     }]
    relativeSpaceBone {boneName=root}
   }
   n118 = Anim Dynamics - Bone: dyn_shoetie_l_in_01(ComponentPose=@n117, Alpha=1, ExternalForce="0,0,0") {
    angularDampingOverride=0.9999
    angularSpringConstant=500
    bAngularSpring=true
    bGravityOverrideInSimSpace=true
    bOverrideAngularDamping=true
    simulationSpace=BoneRelative
    alphaBoolBlend {blendOption=Linear}
    boundBone {boneName=dyn_shoetie_l_in_01}
    componentAppliedLinearAccClamp=<3500 3500 3500>
    componentLinearAccScale=<1.2 1.2 1.2>
    componentLinearVelScale=<1.2 1.2 1.2>
    physicsBodyDefinitions [{
      sphereCollisionRadius=0
      boundBone {boneName=dyn_shoetie_l_in_01}
      boxExtents=<15 30 30>
      constraintSetup {angularLimitsMax=<1 35 15>, angularLimitsMin=<-1 -5 -15>, angularTarget=<0 0 0>}
      localJointOffset=<3 0 0>
     }]
    relativeSpaceBone {boneName=root}
   }
   n119 = Anim Dynamics - Bone: dyn_shoetie_l_in_02(ComponentPose=@n118, Alpha=1, ExternalForce="0,0,0") {
    angularDampingOverride=0.9999
    angularSpringConstant=500
    bAngularSpring=true
    bGravityOverrideInSimSpace=true
    bOverrideAngularDamping=true
    simulationSpace=BoneRelative
    alphaBoolBlend {blendOption=Linear}
    boundBone {boneName=dyn_shoetie_l_in_02}
    componentAppliedLinearAccClamp=<3500 3500 3500>
    componentLinearAccScale=<1.2 1.2 1.2>
    componentLinearVelScale=<1.2 1.2 1.2>
    physicsBodyDefinitions [{
      sphereCollisionRadius=0
      boundBone {boneName=dyn_shoetie_l_in_02}
      boxExtents=<15 30 30>
      constraintSetup {angularLimitsMax=<1 35 15>, angularLimitsMin=<-1 -5 -15>, angularTarget=<0 0 0>}
      localJointOffset=<3 0 0>
     }]
    relativeSpaceBone {boneName=root}
   }
   n120 = Anim Dynamics - Bone: dyn_shoelace_l_in_01(ComponentPose=@n119, Alpha=1, ExternalForce="0,0,0") {
    angularDampingOverride=0.9999
    angularSpringConstant=500
    bAngularSpring=true
    bGravityOverrideInSimSpace=true
    bOverrideAngularDamping=true
    simulationSpace=BoneRelative
    alphaBoolBlend {blendOption=Linear}
    boundBone {boneName=dyn_shoelace_l_in_01}
    componentAppliedLinearAccClamp=<5000 5000 5000>
    componentLinearAccScale=<1.2 1.2 1.2>
    componentLinearVelScale=<1.2 1.2 1.2>
    physicsBodyDefinitions [{
      sphereCollisionRadius=0
      boundBone {boneName=dyn_shoelace_l_in_01}
      boxExtents=<15 20 20>
      constraintSetup {angularLimitsMax=<1 10 20>, angularLimitsMin=<-1 -0.05 -60>, angularTarget=<0 0 0>}
      localJointOffset=<3 0 0>
     }]
    relativeSpaceBone {boneName=root}
   }
   n121 = Anim Dynamics - Bone: dyn_shoelace_l_in_02(ComponentPose=@n120, Alpha=1, ExternalForce="0,0,0") {
    angularDampingOverride=0.9999
    angularSpringConstant=500
    bAngularSpring=true
    bGravityOverrideInSimSpace=true
    bOverrideAngularDamping=true
    simulationSpace=BoneRelative
    alphaBoolBlend {blendOption=Linear}
    boundBone {boneName=dyn_shoelace_l_in_02}
    componentAppliedLinearAccClamp=<5000 5000 5000>
    componentLinearAccScale=<1.2 1.2 1.2>
    componentLinearVelScale=<1.2 1.2 1.2>
    physicsBodyDefinitions [{
      sphereCollisionRadius=0
      boundBone {boneName=dyn_shoelace_l_in_02}
      boxExtents=<15 20 20>
      constraintSetup {angularLimitsMax=<1 45 40>, angularLimitsMin=<-1 -0.5 -60>, angularTarget=<0 0 0>}
      localJointOffset=<3 0 0>
     }]
    relativeSpaceBone {boneName=root}
   }
   n122 = Anim Dynamics - Bone: dyn_shoetie_r_out_01(ComponentPose=@n121, Alpha=1, ExternalForce="0,0,0") {
    angularDampingOverride=0.9999
    angularSpringConstant=500
    bAngularSpring=true
    bGravityOverrideInSimSpace=true
    bOverrideAngularDamping=true
    bUseGravityOverride=true
    simulationSpace=BoneRelative
    alphaBoolBlend {blendOption=Linear}
    boundBone {boneName=dyn_shoetie_r_out_01}
    componentAppliedLinearAccClamp=<3500 3500 3500>
    componentLinearAccScale=<1.2 1.2 1.2>
    componentLinearVelScale=<1.2 1.2 1.2>
    gravityOverride {z=-980}
    physicsBodyDefinitions [{
      sphereCollisionRadius=0
      boundBone {boneName=dyn_shoetie_r_out_01}
      boxExtents=<15 30 30>
      constraintSetup {angularLimitsMax=<1 5 15>, angularLimitsMin=<-1 -10 -15>, angularTarget=<0 0 0>}
      localJointOffset=<-2.5 0 0>
     }]
    relativeSpaceBone {boneName=root}
   }
   n123 = Anim Dynamics - Bone: dyn_shoetie_r_out_02(ComponentPose=@n122, Alpha=1, ExternalForce="0,0,0") {
    angularDampingOverride=0.9999
    angularSpringConstant=500
    bAngularSpring=true
    bGravityOverrideInSimSpace=true
    bOverrideAngularDamping=true
    bUseGravityOverride=true
    simulationSpace=BoneRelative
    alphaBoolBlend {blendOption=Linear}
    boundBone {boneName=dyn_shoetie_r_out_02}
    componentAppliedLinearAccClamp=<3500 3500 3500>
    componentLinearAccScale=<1.2 1.2 1.2>
    componentLinearVelScale=<1.2 1.2 1.2>
    gravityOverride {z=-980}
    physicsBodyDefinitions [{
      sphereCollisionRadius=0
      boundBone {boneName=dyn_shoetie_r_out_02}
      boxExtents=<15 30 30>
      constraintSetup {angularLimitsMax=<1 5 15>, angularLimitsMin=<-1 -20 -15>, angularTarget=<0 0 0>}
      localJointOffset=<-1 0 0>
     }]
    relativeSpaceBone {boneName=root}
   }
   n124 = Anim Dynamics - Bone: dyn_shoelace_r_out_01(ComponentPose=@n123, Alpha=1, ExternalForce="0,0,0") {
    angularDampingOverride=0.9999
    angularSpringConstant=500
    bAngularSpring=true
    bGravityOverrideInSimSpace=true
    bOverrideAngularDamping=true
    bUseGravityOverride=true
    simulationSpace=BoneRelative
    alphaBoolBlend {blendOption=Linear}
    boundBone {boneName=dyn_shoelace_r_out_01}
    componentAppliedLinearAccClamp=<5000 5000 5000>
    componentLinearAccScale=<1.2 1.2 1.2>
    componentLinearVelScale=<1.2 1.2 1.2>
    gravityOverride {z=-980}
    physicsBodyDefinitions [{
      sphereCollisionRadius=0
      boundBone {boneName=dyn_shoelace_r_out_01}
      boxExtents=<15 20 20>
      constraintSetup {angularLimitsMax=<1 5 10>, angularLimitsMin=<-1 -5 -6>, angularTarget=<0 0 0>}
      localJointOffset=<3 0 0>
     }]
    relativeSpaceBone {boneName=root}
   }
   n125 = Anim Dynamics - Bone: dyn_shoelace_r_out_02(ComponentPose=@n124, Alpha=1, ExternalForce="0,0,0") {
    angularDampingOverride=0.9999
    angularSpringConstant=500
    bAngularSpring=true
    bGravityOverrideInSimSpace=true
    bOverrideAngularDamping=true
    bUseGravityOverride=true
    simulationSpace=BoneRelative
    alphaBoolBlend {blendOption=Linear}
    boundBone {boneName=dyn_shoelace_r_out_02}
    componentAppliedLinearAccClamp=<5000 5000 5000>
    componentLinearAccScale=<1.2 1.2 1.2>
    componentLinearVelScale=<1.2 1.2 1.2>
    gravityOverride {z=-980}
    physicsBodyDefinitions [{
      sphereCollisionRadius=0
      boundBone {boneName=dyn_shoelace_r_out_02}
      boxExtents=<15 20 20>
      constraintSetup {angularLimitsMax=<1 5 10>, angularLimitsMin=<-1 -5 -6>, angularTarget=<0 0 0>}
      localJointOffset=<3 0 0>
     }]
    relativeSpaceBone {boneName=root}
   }
   n126 = Anim Dynamics - Bone: dyn_shoelace_r_out_03(ComponentPose=@n125, Alpha=1, ExternalForce="0,0,0") {
    angularDampingOverride=0.9999
    angularSpringConstant=500
    bAngularSpring=true
    bGravityOverrideInSimSpace=true
    bOverrideAngularDamping=true
    bUseGravityOverride=true
    simulationSpace=BoneRelative
    alphaBoolBlend {blendOption=Linear}
    boundBone {boneName=dyn_shoelace_r_out_03}
    componentAppliedLinearAccClamp=<5000 5000 5000>
    componentLinearAccScale=<1.2 1.2 1.2>
    componentLinearVelScale=<1.2 1.2 1.2>
    gravityOverride {z=-980}
    physicsBodyDefinitions [{
      sphereCollisionRadius=0
      boundBone {boneName=dyn_shoelace_r_out_03}
      boxExtents=<15 20 20>
      constraintSetup {angularLimitsMax=<1 5 10>, angularLimitsMin=<-1 -5 -6>, angularTarget=<0 0 0>}
      localJointOffset=<3 0 0>
     }]
    relativeSpaceBone {boneName=root}
   }
   n127 = Anim Dynamics - Bone: dyn_shoetie_r_in_01(ComponentPose=@n126, Alpha=1, ExternalForce="0,0,0") {
    angularDampingOverride=0.9999
    angularSpringConstant=500
    bAngularSpring=true
    bGravityOverrideInSimSpace=true
    bOverrideAngularDamping=true
    bUseGravityOverride=true
    simulationSpace=BoneRelative
    alphaBoolBlend {blendOption=Linear}
    boundBone {boneName=dyn_shoetie_r_in_01}
    componentAppliedLinearAccClamp=<3500 3500 3500>
    componentLinearAccScale=<1.2 1.2 1.2>
    componentLinearVelScale=<1.2 1.2 1.2>
    gravityOverride {z=-980}
    physicsBodyDefinitions [{
      sphereCollisionRadius=0
      boundBone {boneName=dyn_shoetie_r_in_01}
      boxExtents=<15 30 30>
      constraintSetup {angularLimitsMax=<1 5 15>, angularLimitsMin=<-1 -35 -15>, angularTarget=<0 0 0>}
      localJointOffset=<-3 0 0>
     }]
    relativeSpaceBone {boneName=root}
   }
   n128 = Anim Dynamics - Bone: dyn_shoetie_r_in_02(ComponentPose=@n127, Alpha=1, ExternalForce="0,0,0") {
    angularDampingOverride=0.9999
    angularSpringConstant=500
    bAngularSpring=true
    bGravityOverrideInSimSpace=true
    bOverrideAngularDamping=true
    bUseGravityOverride=true
    simulationSpace=BoneRelative
    alphaBoolBlend {blendOption=Linear}
    boundBone {boneName=dyn_shoetie_r_in_02}
    componentAppliedLinearAccClamp=<3500 3500 3500>
    componentLinearAccScale=<1.2 1.2 1.2>
    componentLinearVelScale=<1.2 1.2 1.2>
    gravityOverride {z=-980}
    physicsBodyDefinitions [{
      sphereCollisionRadius=0
      boundBone {boneName=dyn_shoetie_r_in_02}
      boxExtents=<15 30 30>
      constraintSetup {angularLimitsMax=<1 5 15>, angularLimitsMin=<-1 -35 -15>, angularTarget=<0 0 0>}
      localJointOffset=<-3 0 0>
     }]
    relativeSpaceBone {boneName=root}
   }
   n129 = Anim Dynamics - Bone: dyn_shoelace_r_in_01(ComponentPose=@n128, Alpha=1, ExternalForce="0,0,0") {
    angularDampingOverride=0.9999
    angularSpringConstant=500
    bAngularSpring=true
    bGravityOverrideInSimSpace=true
    bOverrideAngularDamping=true
    bUseGravityOverride=true
    simulationSpace=BoneRelative
    alphaBoolBlend {blendOption=Linear}
    boundBone {boneName=dyn_shoelace_r_in_01}
    componentAppliedLinearAccClamp=<5000 5000 5000>
    componentLinearAccScale=<1.2 1.2 1.2>
    componentLinearVelScale=<1.2 1.2 1.2>
    gravityOverride {z=-980}
    physicsBodyDefinitions [{
      sphereCollisionRadius=0
      boundBone {boneName=dyn_shoelace_r_in_01}
      boxExtents=<15 20 20>
      constraintSetup {angularLimitsMax=<1 5 2>, angularLimitsMin=<-1 -5 -5>, angularTarget=<0 0 0>}
      localJointOffset=<3 0 0>
     }]
    relativeSpaceBone {boneName=root}
   }
   n130 = Anim Dynamics - Bone: dyn_shoelace_r_in_02(ComponentPose=@n129, Alpha=1, ExternalForce="0,0,0") {
    angularDampingOverride=0.9999
    angularSpringConstant=500
    bAngularSpring=true
    bGravityOverrideInSimSpace=true
    bOverrideAngularDamping=true
    bUseGravityOverride=true
    simulationSpace=BoneRelative
    alphaBoolBlend {blendOption=Linear}
    boundBone {boneName=dyn_shoelace_r_in_02}
    componentAppliedLinearAccClamp=<5000 5000 5000>
    componentLinearAccScale=<1.2 1.2 1.2>
    componentLinearVelScale=<1.2 1.2 1.2>
    gravityOverride {z=-980}
    physicsBodyDefinitions [{
      sphereCollisionRadius=0
      boundBone {boneName=dyn_shoelace_r_in_02}
      boxExtents=<15 20 20>
      constraintSetup {angularLimitsMax=<1 5 2>, angularLimitsMin=<-1 -5 -5>, angularTarget=<0 0 0>}
      localJointOffset=<3 0 0>
     }]
    relativeSpaceBone {boneName=root}
   }
   Output Pose(Result=Component To Local(ComponentPose=@n130)) {layerGroup=None}
   n131 = Control Rig(Source=Copy Pose From Mesh() {bUseAttachedParent=true}, Alpha=Get Parachute Open() -> (Parachute Open:double)) {
    alphaBoolBlend {blendOption=Linear}
    controlRigAssetReference {blueprintRigClass=&/M_MED_Jonesy_Orange_Deform_Master_ControlRig.M_MED_Jonesy_Orange_Deform_Master_ControlRig_C}
   }
   n132 = "Transform (Modify) Bone - Bone: dyn_strap_top_l"!(
    ComponentPose=Local To Component(LocalPose=@n131)
    Alpha=Get L_arm_up() -> (L_arm_up:double)
    Translation="6.966,0.756,1.700"
    Rotation="0.512,14.683,-9.984"
   ) -> (Pose) {
    rotationMode=BMM_Replace
    rotationSpace=BCS_BoneSpace
    scaleSpace=BCS_BoneSpace
    translationMode=BMM_Additive
    translationSpace=BCS_BoneSpace
    alphaBoolBlend {blendOption=Linear}
    boneToModify {boneName=dyn_strap_top_l}
    rotation=<rot 0.512 14.683 -9.984>
    translation=<6.966 0.756 1.7>
   }
   n133 = "Transform (Modify) Bone - Bone: dyn_strap_top_r"(
    ComponentPose=@n132
    Alpha=Get R_arm_up() -> (R_arm_up:double)
    Translation="2.442,0.021,-1.066"
    Rotation="-2.126,44.759,-0.357"
    Scale="1,1,1"
   ) {
    rotationMode=BMM_Replace
    rotationSpace=BCS_BoneSpace
    scaleMode=BMM_Additive
    scaleSpace=BCS_BoneSpace
    translationMode=BMM_Additive
    translationSpace=BCS_BoneSpace
    alphaBoolBlend {blendOption=Linear}
    boneToModify {boneName=dyn_strap_top_r}
    rotation=<rot -2.126 44.759 -0.357>
    translation=<2.442 0.021 -1.066>
   }
   n134 = "Transform (Modify) Bone - Bone: dyn_strap_top_r"(
    ComponentPose=@n133
    Alpha=Get R_Arm_Down() -> (R_Arm_Down:double)
    Translation="-1.074,-0.061,-2.410"
    Rotation="4.468,0.769,0.575"
    Scale="1.000,2.250,1.250"
   ) {
    rotationMode=BMM_Replace
    rotationSpace=BCS_BoneSpace
    scaleMode=BMM_Additive
    scaleSpace=BCS_BoneSpace
    translationMode=BMM_Additive
    translationSpace=BCS_BoneSpace
    alphaBoolBlend {blendOption=Linear}
    boneToModify {boneName=dyn_strap_top_r}
    rotation=<rot 4.468 0.769 0.575>
    scale {y=2.25, z=1.25}
    translation=<-1.074 -0.061 -2.41>
   }
   n135 = "Constraint - dyn_sleeve_r_MSTR (P)"(ComponentPose=@n134, Alpha=0.6, ConstraintWeights_0=1) {
    alphaBoolBlend {blendOption=Linear}
    alphaScaleBiasClamp {bClampResult=true, clampMax=0.6}
    boneToModify {boneName=dyn_sleeve_r_MSTR}
    constraintSetup [{targetBone {boneName=thigh_r}}]
    constraintWeights [1]
   }
   n136 = "Transform (Modify) Bone - Bone: dyn_sleeve_r_MSTR"!(ComponentPose=@n135, Alpha=1, Translation="0.414,-0.228,-1.696", Rotation="14.885,-10.06,-1.408") -> (Pose) {
    rotationMode=BMM_Replace
    rotationSpace=BCS_BoneSpace
    translationMode=BMM_Replace
    translationSpace=BCS_BoneSpace
    alphaBoolBlend {blendOption=Linear}
    alphaScaleBiasClamp {bClampResult=true, bInterpResult=true, clampMax=0.6, interpSpeedDecreasing=5, interpSpeedIncreasing=7}
    boneToModify {boneName=dyn_sleeve_r_MSTR}
    rotation=<rot 14.885 -10.06 -1.408>
    translation=<0.414 -0.228 -1.696>
   }
   n137 = "Constraint - dyn_sleeve_l_MSTR (P)"(ComponentPose=@n136, Alpha=0.6, ConstraintWeights_0=1) {
    alphaBoolBlend {blendOption=Linear}
    alphaScaleBiasClamp {bClampResult=true, clampMax=0.6}
    boneToModify {boneName=dyn_sleeve_l_MSTR}
    constraintSetup [{targetBone {boneName=thigh_l}}]
    constraintWeights [1]
   }
   n138 = "Transform (Modify) Bone - Bone: dyn_sleeve_l_MSTR"!(ComponentPose=@n137, Alpha=1, Translation="10.393,8.249,-0.907", Rotation="9.994,-0.004,-0.002") -> (Pose) {
    rotationMode=BMM_Replace
    rotationSpace=BCS_BoneSpace
    translationMode=BMM_Replace
    translationSpace=BCS_BoneSpace
    alphaBoolBlend {blendOption=Linear}
    alphaScaleBiasClamp {bClampResult=true, bInterpResult=true, clampMax=0.6, interpSpeedDecreasing=5, interpSpeedIncreasing=7}
    boneToModify {boneName=dyn_sleeve_l_MSTR}
    rotation=<rot 9.9941 -0.004 -0.002>
    translation=<10.393 8.249 -0.907>
   }
   n139 = Look At - Bone: dyn_loincloth_fr_l_MSTR(ComponentPose=@n138, Alpha=1) {
    alphaBoolBlend {blendOption=Linear}
    boneToModify {boneName=dyn_loincloth_fr_l_MSTR}
    lookAtLocation=<-21.6467 -5.8862 -13.0523>
    lookAtTarget {boneReference {boneName=thigh_l}}
    lookAt_Axis {axis=<1 0>}
   }
   n140 = Look At - Bone: dyn_loincloth_side_l_MSTR(ComponentPose=@n139, Alpha=1) {
    alphaBoolBlend {blendOption=Linear}
    boneToModify {boneName=dyn_loincloth_side_l_MSTR}
    lookAtLocation=<-23.3568 -29.7073 -14.2694>
    lookAtTarget {boneReference {boneName=thigh_l}}
    lookAt_Axis {axis=<1 0>}
   }
   n141 = Look At - Bone: dyn_loincloth_bk_l_MSTR(ComponentPose=@n140, Alpha=1) {
    alphaBoolBlend {blendOption=Linear}
    alphaScaleBiasClamp {bClampResult=true, bInterpResult=true, clampMax=0.6, interpSpeedDecreasing=5, interpSpeedIncreasing=7}
    boneToModify {boneName=dyn_loincloth_bk_l_MSTR}
    lookAtLocation=<-12.5598 -16.0606 -4.723>
    lookAtTarget {boneReference {boneName=thigh_l}}
    lookAt_Axis {axis=<1 0>}
   }
   n142 = Look At - Bone: dyn_loincloth_fr_r_MSTR(ComponentPose=@n141, Alpha=1) {
    alphaBoolBlend {blendOption=Linear}
    boneToModify {boneName=dyn_loincloth_fr_r_MSTR}
    lookAtLocation=<39.6499 -3.0182 16.2489>
    lookAtTarget {boneReference {boneName=thigh_r}}
    lookAt_Axis {axis=<1 0>}
   }
   n143 = Look At - Bone: dyn_loincloth_side_r_MSTR(ComponentPose=@n142, Alpha=1) {
    alphaBoolBlend {blendOption=Linear}
    boneToModify {boneName=dyn_loincloth_side_r_MSTR}
    lookAtLocation=<15.2883 5.9511 14.5933>
    lookAtTarget {boneReference {boneName=thigh_r}}
    lookAt_Axis {axis=<1 0>}
   }
   n144 = Look At - Bone: dyn_loincloth_bk_r_MSTR(ComponentPose=@n143, Alpha=1) {
    alphaBoolBlend {blendOption=Linear}
    alphaScaleBiasClamp {bClampResult=true, bInterpResult=true, clampMax=0.6, interpSpeedDecreasing=5, interpSpeedIncreasing=7}
    boneToModify {boneName=dyn_loincloth_bk_r_MSTR}
    lookAtLocation=<6.8998 16.6406 2.8904>
    lookAtTarget {boneReference {boneName=thigh_r}}
    lookAt_Axis {axis=<1 0>}
   }
   n145 = Look At - Bone: dyn_loincloth_bk_MSTR(ComponentPose=@n144, Alpha=1) {
    alphaBoolBlend {blendOption=Linear}
    boneToModify {boneName=dyn_loincloth_bk_MSTR}
    lookAtLocation=<0 -12>
    lookAtTarget {boneReference {boneName=pelvis}}
    lookAt_Axis {axis=<1 0>}
   }
   n146 = "Transform (Modify) Bone - Bone: dyn_loincloth_fr_l_03"!(ComponentPose=@n145, Alpha=1, Rotation="23.252,-0.472,-9.300") -> (Pose) {
    rotationMode=BMM_Replace
    rotationSpace=BCS_BoneSpace
    translationMode=BMM_Replace
    translationSpace=BCS_BoneSpace
    alphaBoolBlend {blendOption=Linear}
    alphaScaleBiasClamp {bClampResult=true, bInterpResult=true, clampMax=0.6, interpSpeedDecreasing=5, interpSpeedIncreasing=7}
    boneToModify {boneName=dyn_loincloth_fr_l_03}
    rotation=<rot 23.252 -0.472 -9.3>
    translation=<-0.008 -0.2018 0.6709>
   }
   n147 = "Transform (Modify) Bone - Bone: dyn_loincloth_fr_r_03"(ComponentPose=@n146, Alpha=1, Rotation="23.600,1.967,8.413") {
    rotationMode=BMM_Replace
    rotationSpace=BCS_BoneSpace
    translationMode=BMM_Replace
    translationSpace=BCS_BoneSpace
    alphaBoolBlend {blendOption=Linear}
    alphaScaleBiasClamp {bClampResult=true, bInterpResult=true, clampMax=0.6, interpSpeedDecreasing=5, interpSpeedIncreasing=7}
    boneToModify {boneName=dyn_loincloth_fr_r_03}
    rotation=<rot 23.6001 1.967 8.4131>
    translation=<-0.8103 -1.3985 0.9605>
   }
   n148 = Anim Dynamics - Bone: dyn_loincloth_fr_l_02(ComponentPose=@n147, Alpha=1, ExternalForce="0,0,0") {
    angularDampingOverride=0.9999
    angularSpringConstant=500
    bAngularSpring=true
    bGravityOverrideInSimSpace=true
    bOverrideAngularDamping=true
    bUseGravityOverride=true
    bUseSphericalLimits=true
    gravityScale=0.5
    simulationSpace=RootRelative
    alphaBoolBlend {blendOption=Linear}
    boundBone {boneName=dyn_loincloth_fr_l_02}
    componentAppliedLinearAccClamp=<5000 5000 5000>
    componentLinearAccScale=<0.2 0.2 0.4>
    componentLinearVelScale=<0.5 0.5 0.5>
    gravityOverride {z=-980}
    physicsBodyDefinitions [{
      sphereCollisionRadius=0
      boundBone {boneName=dyn_loincloth_fr_l_02}
      boxExtents=<5 20 20>
      constraintSetup {angularLimitsMax=<1 3 15>, angularLimitsMin=<-1 -30 -15>, angularTarget=<0 0 0>}
      localJointOffset=<5 0 0>
     }]
    relativeSpaceBone {boneName=root}
   }
   n149 = Anim Dynamics - Bone: dyn_loincloth_fr_l_03(ComponentPose=@n148, Alpha=1, ExternalForce="0,0,0") {
    angularDampingOverride=0.9999
    angularSpringConstant=500
    bAngularSpring=true
    bGravityOverrideInSimSpace=true
    bOverrideAngularDamping=true
    bUseGravityOverride=true
    bUseSphericalLimits=true
    gravityScale=0.5
    simulationSpace=RootRelative
    alphaBoolBlend {blendOption=Linear}
    boundBone {boneName=dyn_loincloth_fr_l_03}
    componentAppliedLinearAccClamp=<3500 3500 3500>
    componentLinearAccScale=<0.2 0.2 0.4>
    componentLinearVelScale=<0.5 0.5 0.5>
    gravityOverride {z=-980}
    physicsBodyDefinitions [{
      sphereCollisionRadius=0
      boundBone {boneName=dyn_loincloth_fr_l_03}
      boxExtents=<5 20 20>
      constraintSetup {angularLimitsMax=<1 3 15>, angularLimitsMin=<-1 -30 -15>, angularTarget=<0 0 0>}
      localJointOffset=<5 0 0>
     }]
    relativeSpaceBone {boneName=root}
   }
   n150 = Anim Dynamics - Bone: dyn_loincloth_fr_l_04(ComponentPose=@n149, Alpha=1, ExternalForce="0,0,0") {
    angularDampingOverride=0.9999
    angularSpringConstant=500
    bAngularSpring=true
    bGravityOverrideInSimSpace=true
    bOverrideAngularDamping=true
    bUseGravityOverride=true
    bUseSphericalLimits=true
    gravityScale=0.5
    simulationSpace=RootRelative
    alphaBoolBlend {blendOption=Linear}
    boundBone {boneName=dyn_loincloth_fr_l_04}
    componentAppliedLinearAccClamp=<5000 5000 5000>
    componentLinearAccScale=<0.2 0.2 0.4>
    componentLinearVelScale=<0.5 0.5 0.5>
    gravityOverride {z=-980}
    physicsBodyDefinitions [{
      sphereCollisionRadius=0
      boundBone {boneName=dyn_loincloth_fr_l_04}
      boxExtents=<5 20 20>
      constraintSetup {angularLimitsMax=<1 3 15>, angularLimitsMin=<-1 -30 -15>, angularTarget=<0 0 0>}
      localJointOffset=<5 0 0>
     }]
    relativeSpaceBone {boneName=root}
   }
   n151 = Anim Dynamics - Bone: dyn_loincloth_side_l_01(ComponentPose=@n150, Alpha=1, ExternalForce="0,0,0") {
    angularDampingOverride=0.9999
    angularSpringConstant=500
    bAngularSpring=true
    bGravityOverrideInSimSpace=true
    bOverrideAngularDamping=true
    bUseGravityOverride=true
    bUseSphericalLimits=true
    gravityScale=0.5
    simulationSpace=RootRelative
    alphaBoolBlend {blendOption=Linear}
    boundBone {boneName=dyn_loincloth_side_l_01}
    componentAppliedLinearAccClamp=<5000 5000 5000>
    componentLinearAccScale=<0.2 0.2 0.4>
    componentLinearVelScale=<0.5 0.5 0.5>
    gravityOverride {z=-980}
    physicsBodyDefinitions [{
      sphereCollisionRadius=0
      boundBone {boneName=dyn_loincloth_side_l_01}
      boxExtents=<8 20 20>
      constraintSetup {angularLimitsMax=<1 3 15>, angularLimitsMin=<-1 -30 -15>, angularTarget=<0 0 0>}
      localJointOffset=<5 0 0>
     }]
    relativeSpaceBone {boneName=root}
    sphericalLimits [{limitRadius=5, drivingBone {boneName=thigh_l}, sphereLocalOffset=<0 0 0>}]
   }
   n152 = Anim Dynamics - Bone: dyn_loincloth_side_l_02(ComponentPose=@n151, Alpha=1, ExternalForce="0,0,0") {
    angularDampingOverride=0.9999
    angularSpringConstant=500
    bAngularSpring=true
    bGravityOverrideInSimSpace=true
    bOverrideAngularDamping=true
    bUseGravityOverride=true
    bUseSphericalLimits=true
    gravityScale=0.5
    simulationSpace=RootRelative
    alphaBoolBlend {blendOption=Linear}
    boundBone {boneName=dyn_loincloth_side_l_02}
    componentAppliedLinearAccClamp=<5000 5000 5000>
    componentLinearAccScale=<0.2 0.2 0.4>
    componentLinearVelScale=<0.5 0.5 0.5>
    gravityOverride {z=-980}
    physicsBodyDefinitions [{
      sphereCollisionRadius=0
      boundBone {boneName=dyn_loincloth_side_l_02}
      boxExtents=<8 20 20>
      constraintSetup {angularLimitsMax=<1 3 15>, angularLimitsMin=<-1 -30 -15>, angularTarget=<0 0 0>}
      localJointOffset=<5 0 0>
     }]
    relativeSpaceBone {boneName=root}
   }
   n153 = Anim Dynamics - Bone: dyn_loincloth_side_l_03(ComponentPose=@n152, Alpha=1, ExternalForce="0,0,0") {
    angularDampingOverride=0.9999
    angularSpringConstant=500
    bAngularSpring=true
    bGravityOverrideInSimSpace=true
    bOverrideAngularDamping=true
    bUseGravityOverride=true
    bUseSphericalLimits=true
    gravityScale=0.5
    simulationSpace=RootRelative
    alphaBoolBlend {blendOption=Linear}
    boundBone {boneName=dyn_loincloth_side_l_03}
    componentAppliedLinearAccClamp=<3500 3500 3500>
    componentLinearAccScale=<0.2 0.2 0.4>
    componentLinearVelScale=<0.5 0.5 0.5>
    gravityOverride {z=-980}
    physicsBodyDefinitions [{
      sphereCollisionRadius=0
      boundBone {boneName=dyn_loincloth_side_l_03}
      boxExtents=<8 20 20>
      constraintSetup {angularLimitsMax=<1 3 15>, angularLimitsMin=<-1 -30 -15>, angularTarget=<0 0 0>}
      localJointOffset=<5 0 0>
     }]
    relativeSpaceBone {boneName=root}
   }
   n154 = Anim Dynamics - Bone: dyn_loincloth_bk_l_01(ComponentPose=@n153, Alpha=1, ExternalForce="0,0,0") {
    angularDampingOverride=0.9999
    angularSpringConstant=500
    bAngularSpring=true
    bGravityOverrideInSimSpace=true
    bOverrideAngularDamping=true
    bUseGravityOverride=true
    bUseSphericalLimits=true
    gravityScale=0.5
    simulationSpace=RootRelative
    alphaBoolBlend {blendOption=Linear}
    boundBone {boneName=dyn_loincloth_bk_l_01}
    componentAppliedLinearAccClamp=<5000 5000 5000>
    componentLinearAccScale=<0.2 0.2 0.4>
    componentLinearVelScale=<0.5 0.5 0.5>
    gravityOverride {z=-980}
    physicsBodyDefinitions [{
      sphereCollisionRadius=0
      boundBone {boneName=dyn_loincloth_bk_l_01}
      boxExtents=<8 20 20>
      constraintSetup {angularLimitsMax=<0.2 2 15>, angularLimitsMin=<-1 -30 -15>, angularTarget=<0 0 0>}
      localJointOffset=<5 0 0>
     }]
    relativeSpaceBone {boneName=root}
    sphericalLimits [{limitRadius=10, drivingBone {boneName=thigh_l}, sphereLocalOffset=<0 0 0>}]
   }
   n155 = Anim Dynamics - Bone: dyn_loincloth_bk_l_02(ComponentPose=@n154, Alpha=1, ExternalForce="0,0,0") {
    angularDampingOverride=0.9999
    angularSpringConstant=500
    bAngularSpring=true
    bGravityOverrideInSimSpace=true
    bOverrideAngularDamping=true
    bUseGravityOverride=true
    bUseSphericalLimits=true
    gravityScale=0.5
    simulationSpace=RootRelative
    alphaBoolBlend {blendOption=Linear}
    boundBone {boneName=dyn_loincloth_bk_l_02}
    componentAppliedLinearAccClamp=<5000 5000 5000>
    componentLinearAccScale=<0.2 0.2 0.4>
    componentLinearVelScale=<0.5 0.5 0.5>
    gravityOverride {z=-980}
    physicsBodyDefinitions [{
      sphereCollisionRadius=0
      boundBone {boneName=dyn_loincloth_bk_l_02}
      boxExtents=<8 20 20>
      constraintSetup {angularLimitsMax=<1 2 15>, angularLimitsMin=<-1 -30 -15>, angularTarget=<0 0 0>}
      localJointOffset=<5 0 0>
     }]
    relativeSpaceBone {boneName=root}
    sphericalLimits [{limitRadius=10, drivingBone {boneName=thigh_l}, sphereLocalOffset=<-10 -5 0>}]
   }
   n156 = Anim Dynamics - Bone: dyn_loincloth_bk_l_03(ComponentPose=@n155, Alpha=1, ExternalForce="0,0,0") {
    angularDampingOverride=0.9999
    angularSpringConstant=500
    bAngularSpring=true
    bGravityOverrideInSimSpace=true
    bOverrideAngularDamping=true
    bUseGravityOverride=true
    bUseSphericalLimits=true
    gravityScale=0.5
    simulationSpace=RootRelative
    alphaBoolBlend {blendOption=Linear}
    boundBone {boneName=dyn_loincloth_bk_l_03}
    componentAppliedLinearAccClamp=<3500 3500 3500>
    componentLinearAccScale=<0.2 0.2 0.4>
    componentLinearVelScale=<0.5 0.5 0.5>
    gravityOverride {z=-980}
    physicsBodyDefinitions [{
      sphereCollisionRadius=0
      boundBone {boneName=dyn_loincloth_bk_l_03}
      boxExtents=<8 20 20>
      constraintSetup {angularLimitsMax=<1 2 15>, angularLimitsMin=<-1 -30 -15>, angularTarget=<0 0 0>}
      localJointOffset=<5 0 0>
     }]
    relativeSpaceBone {boneName=root}
   }
   n157 = Anim Dynamics - Bone: dyn_loincloth_fr_r_02(ComponentPose=@n156, Alpha=1, ExternalForce="0,0,0") {
    angularDampingOverride=0.9999
    angularSpringConstant=500
    bAngularSpring=true
    bGravityOverrideInSimSpace=true
    bOverrideAngularDamping=true
    bUseGravityOverride=true
    gravityScale=0.5
    simulationSpace=RootRelative
    alphaBoolBlend {blendOption=Linear}
    boundBone {boneName=dyn_loincloth_fr_r_02}
    componentAppliedLinearAccClamp=<5000 5000 5000>
    componentLinearAccScale=<0.2 0.2 0.4>
    componentLinearVelScale=<0.5 0.5 0.5>
    gravityOverride {z=-980}
    physicsBodyDefinitions [{
      sphereCollisionRadius=0
      boundBone {boneName=dyn_loincloth_fr_r_02}
      boxExtents=<5 20 20>
      constraintSetup {angularLimitsMax=<1 3 15>, angularLimitsMin=<-1 -30 -15>, angularTarget=<0 0 0>}
      localJointOffset=<5 0 0>
     }]
    relativeSpaceBone {boneName=root}
   }
   n158 = Anim Dynamics - Bone: dyn_loincloth_fr_r_03(ComponentPose=@n157, Alpha=1, ExternalForce="0,0,0") {
    angularDampingOverride=0.9999
    angularSpringConstant=500
    bAngularSpring=true
    bGravityOverrideInSimSpace=true
    bOverrideAngularDamping=true
    bUseGravityOverride=true
    gravityScale=0.5
    simulationSpace=RootRelative
    alphaBoolBlend {blendOption=Linear}
    boundBone {boneName=dyn_loincloth_fr_r_03}
    componentAppliedLinearAccClamp=<3500 3500 3500>
    componentLinearAccScale=<0.2 0.2 0.4>
    componentLinearVelScale=<0.5 0.5 0.5>
    gravityOverride {z=-980}
    physicsBodyDefinitions [{
      sphereCollisionRadius=0
      boundBone {boneName=dyn_loincloth_fr_r_03}
      boxExtents=<5 20 20>
      constraintSetup {angularLimitsMax=<1 3 15>, angularLimitsMin=<-1 -30 -15>, angularTarget=<0 0 0>}
      localJointOffset=<5 0 0>
     }]
    relativeSpaceBone {boneName=root}
   }
   n159 = Anim Dynamics - Bone: dyn_loincloth_fr_r_04(ComponentPose=@n158, Alpha=1, ExternalForce="0,0,0") {
    angularDampingOverride=0.9999
    angularSpringConstant=500
    bAngularSpring=true
    bGravityOverrideInSimSpace=true
    bOverrideAngularDamping=true
    bUseGravityOverride=true
    gravityScale=0.5
    simulationSpace=RootRelative
    alphaBoolBlend {blendOption=Linear}
    boundBone {boneName=dyn_loincloth_fr_r_04}
    componentAppliedLinearAccClamp=<5000 5000 5000>
    componentLinearAccScale=<0.2 0.2 0.4>
    componentLinearVelScale=<0.5 0.5 0.5>
    gravityOverride {z=-980}
    physicsBodyDefinitions [{
      sphereCollisionRadius=0
      boundBone {boneName=dyn_loincloth_fr_r_04}
      boxExtents=<5 20 20>
      constraintSetup {angularLimitsMax=<1 3 15>, angularLimitsMin=<-1 -30 -15>, angularTarget=<0 0 0>}
      localJointOffset=<5 0 0>
     }]
    relativeSpaceBone {boneName=root}
   }
   n160 = Anim Dynamics - Bone: dyn_loincloth_side_r_01(ComponentPose=@n159, Alpha=1, ExternalForce="0,0,0") {
    angularDampingOverride=0.9999
    angularSpringConstant=500
    bAngularSpring=true
    bGravityOverrideInSimSpace=true
    bOverrideAngularDamping=true
    bUseGravityOverride=true
    gravityScale=0.5
    simulationSpace=RootRelative
    alphaBoolBlend {blendOption=Linear}
    boundBone {boneName=dyn_loincloth_side_r_01}
    componentAppliedLinearAccClamp=<5000 5000 5000>
    componentLinearAccScale=<0.2 0.2 0.4>
    componentLinearVelScale=<0.5 0.5 0.5>
    gravityOverride {z=-980}
    physicsBodyDefinitions [{
      sphereCollisionRadius=0
      boundBone {boneName=dyn_loincloth_side_r_01}
      boxExtents=<8 20 20>
      constraintSetup {angularLimitsMax=<1 3 15>, angularLimitsMin=<-1 -30 -15>, angularTarget=<0 0 0>}
      localJointOffset=<5 0 0>
     }]
    relativeSpaceBone {boneName=root}
   }
   n161 = Anim Dynamics - Bone: dyn_loincloth_side_r_02(ComponentPose=@n160, Alpha=1, ExternalForce="0,0,0") {
    angularDampingOverride=0.9999
    angularSpringConstant=500
    bAngularSpring=true
    bGravityOverrideInSimSpace=true
    bOverrideAngularDamping=true
    bUseGravityOverride=true
    bUseSphericalLimits=true
    gravityScale=0.5
    simulationSpace=RootRelative
    alphaBoolBlend {blendOption=Linear}
    boundBone {boneName=dyn_loincloth_side_r_02}
    componentAppliedLinearAccClamp=<5000 5000 5000>
    componentLinearAccScale=<0.2 0.2 0.4>
    componentLinearVelScale=<0.5 0.5 0.5>
    gravityOverride {z=-980}
    physicsBodyDefinitions [{
      sphereCollisionRadius=0
      boundBone {boneName=dyn_loincloth_side_r_02}
      boxExtents=<8 20 20>
      constraintSetup {angularLimitsMax=<1 3 15>, angularLimitsMin=<-1 -30 -15>, angularTarget=<0 0 0>}
      localJointOffset=<5 0 0>
     }]
    relativeSpaceBone {boneName=root}
   }
   n162 = Anim Dynamics - Bone: dyn_loincloth_side_r_03(ComponentPose=@n161, Alpha=1, ExternalForce="0,0,0") {
    angularDampingOverride=0.9999
    angularSpringConstant=500
    bAngularSpring=true
    bGravityOverrideInSimSpace=true
    bOverrideAngularDamping=true
    bUseGravityOverride=true
    bUseSphericalLimits=true
    gravityScale=0.5
    simulationSpace=RootRelative
    alphaBoolBlend {blendOption=Linear}
    boundBone {boneName=dyn_loincloth_side_r_03}
    componentAppliedLinearAccClamp=<3500 3500 3500>
    componentLinearAccScale=<0.2 0.2 0.4>
    componentLinearVelScale=<0.5 0.5 0.5>
    gravityOverride {z=-980}
    physicsBodyDefinitions [{
      sphereCollisionRadius=0
      boundBone {boneName=dyn_loincloth_side_r_03}
      boxExtents=<8 20 20>
      constraintSetup {angularLimitsMax=<1 3 15>, angularLimitsMin=<-1 -30 -15>, angularTarget=<0 0 0>}
      localJointOffset=<5 0 0>
     }]
    relativeSpaceBone {boneName=root}
   }
   n163 = Anim Dynamics - Bone: dyn_loincloth_bk_r_01(ComponentPose=@n162, Alpha=1, ExternalForce="0,0,0") {
    angularDampingOverride=0.9999
    angularSpringConstant=500
    bAngularSpring=true
    bGravityOverrideInSimSpace=true
    bOverrideAngularDamping=true
    bUseGravityOverride=true
    gravityScale=0.5
    simulationSpace=RootRelative
    alphaBoolBlend {blendOption=Linear}
    boundBone {boneName=dyn_loincloth_bk_r_01}
    componentAppliedLinearAccClamp=<5000 5000 5000>
    componentLinearAccScale=<0.2 0.2 0.4>
    componentLinearVelScale=<0.5 0.5 0.5>
    gravityOverride {z=-980}
    physicsBodyDefinitions [{
      sphereCollisionRadius=0
      boundBone {boneName=dyn_loincloth_bk_r_01}
      boxExtents=<8 20 20>
      constraintSetup {angularLimitsMax=<1 2 15>, angularLimitsMin=<-1 -30 -15>, angularTarget=<0 0 0>}
      localJointOffset=<5 0 0>
     }]
    relativeSpaceBone {boneName=root}
   }
   n164 = Anim Dynamics - Bone: dyn_loincloth_bk_r_02(ComponentPose=@n163, Alpha=1, ExternalForce="0,0,0") {
    angularDampingOverride=0.9999
    angularSpringConstant=500
    bAngularSpring=true
    bGravityOverrideInSimSpace=true
    bOverrideAngularDamping=true
    bUseGravityOverride=true
    bUseSphericalLimits=true
    gravityScale=0.5
    simulationSpace=RootRelative
    alphaBoolBlend {blendOption=Linear}
    boundBone {boneName=dyn_loincloth_bk_r_02}
    componentAppliedLinearAccClamp=<5000 5000 5000>
    componentLinearAccScale=<0.2 0.2 0.4>
    componentLinearVelScale=<0.5 0.5 0.5>
    gravityOverride {z=-980}
    physicsBodyDefinitions [{
      sphereCollisionRadius=0
      boundBone {boneName=dyn_loincloth_bk_r_02}
      boxExtents=<8 20 20>
      constraintSetup {angularLimitsMax=<1 2 15>, angularLimitsMin=<-1 -30 -15>, angularTarget=<0 0 0>}
      localJointOffset=<5 0 0>
     }]
    relativeSpaceBone {boneName=root}
   }
   n165 = Anim Dynamics - Bone: dyn_loincloth_bk_r_03(ComponentPose=@n164, Alpha=1, ExternalForce="0,0,0") {
    angularDampingOverride=0.9999
    angularSpringConstant=500
    bAngularSpring=true
    bGravityOverrideInSimSpace=true
    bOverrideAngularDamping=true
    bUseGravityOverride=true
    bUseSphericalLimits=true
    gravityScale=0.5
    simulationSpace=RootRelative
    alphaBoolBlend {blendOption=Linear}
    boundBone {boneName=dyn_loincloth_bk_r_03}
    componentAppliedLinearAccClamp=<3500 3500 3500>
    componentLinearAccScale=<0.2 0.2 0.4>
    componentLinearVelScale=<0.5 0.5 0.5>
    gravityOverride {z=-980}
    physicsBodyDefinitions [{
      sphereCollisionRadius=0
      boundBone {boneName=dyn_loincloth_bk_r_03}
      boxExtents=<8 20 20>
      constraintSetup {angularLimitsMax=<1 2 15>, angularLimitsMin=<-1 -30 -15>, angularTarget=<0 0 0>}
      localJointOffset=<5 0 0>
     }]
    relativeSpaceBone {boneName=root}
   }
   n166 = Anim Dynamics - Bone: dyn_loincloth_bk_01(ComponentPose=@n165, Alpha=1, ExternalForce="0,0,0") {
    angularDampingOverride=0.9999
    angularSpringConstant=500
    bAngularSpring=true
    bGravityOverrideInSimSpace=true
    bOverrideAngularDamping=true
    bUseGravityOverride=true
    gravityScale=0.5
    simulationSpace=BoneRelative
    alphaBoolBlend {blendOption=Linear}
    boundBone {boneName=dyn_loincloth_bk_01}
    componentAppliedLinearAccClamp=<5000 5000 5000>
    componentLinearAccScale=<0.2 0.2 0.4>
    componentLinearVelScale=<0.5 0.5 0.5>
    gravityOverride {z=-980}
    physicsBodyDefinitions [{
      sphereCollisionRadius=0
      boundBone {boneName=dyn_loincloth_bk_01}
      boxExtents=<5 20 20>
      constraintSetup {angularLimitsMax=<1 3 15>, angularLimitsMin=<-1 -30 -15>, angularTarget=<0 0 0>}
      localJointOffset=<5 0 0>
     }]
    relativeSpaceBone {boneName=root}
   }
   n167 = Anim Dynamics - Bone: dyn_loincloth_bk_02(ComponentPose=@n166, Alpha=1, ExternalForce="0,0,0") {
    angularDampingOverride=0.9999
    angularSpringConstant=500
    bAngularSpring=true
    bGravityOverrideInSimSpace=true
    bOverrideAngularDamping=true
    bUseGravityOverride=true
    gravityScale=0.5
    simulationSpace=BoneRelative
    alphaBoolBlend {blendOption=Linear}
    boundBone {boneName=dyn_loincloth_bk_02}
    componentAppliedLinearAccClamp=<5000 5000 5000>
    componentLinearAccScale=<0.2 0.2 0.4>
    componentLinearVelScale=<0.5 0.5 0.5>
    gravityOverride {z=-980}
    physicsBodyDefinitions [{
      sphereCollisionRadius=0
      boundBone {boneName=dyn_loincloth_bk_02}
      boxExtents=<5 20 20>
      constraintSetup {angularLimitsMax=<1 3 15>, angularLimitsMin=<-1 -30 -15>, angularTarget=<0 0 0>}
      localJointOffset=<5 0 0>
     }]
    relativeSpaceBone {boneName=root}
   }
   n168 = Anim Dynamics - Bone: dyn_loincloth_bk_03(ComponentPose=@n167, Alpha=1, ExternalForce="0,0,0") {
    angularDampingOverride=0.9999
    angularSpringConstant=500
    bAngularSpring=true
    bGravityOverrideInSimSpace=true
    bOverrideAngularDamping=true
    bUseGravityOverride=true
    gravityScale=0.5
    simulationSpace=BoneRelative
    alphaBoolBlend {blendOption=Linear}
    boundBone {boneName=dyn_loincloth_bk_03}
    componentAppliedLinearAccClamp=<3500 3500 3500>
    componentLinearAccScale=<0.2 0.2 0.4>
    componentLinearVelScale=<0.5 0.5 0.5>
    gravityOverride {z=-980}
    physicsBodyDefinitions [{
      sphereCollisionRadius=0
      boundBone {boneName=dyn_loincloth_bk_03}
      boxExtents=<5 20 20>
      constraintSetup {angularLimitsMax=<1 3 15>, angularLimitsMin=<-1 -30 -15>, angularTarget=<0 0 0>}
      localJointOffset=<5 0 0>
     }]
    relativeSpaceBone {boneName=root}
   }
   n169 = Anim Dynamics - Bone: dyn_sleeve_l_01(ComponentPose=@n168, Alpha=1, ExternalForce="0,0,0") {
    angularDampingOverride=0.9999
    angularSpringConstant=500
    bAngularSpring=true
    bGravityOverrideInSimSpace=true
    bOverrideAngularDamping=true
    bUseGravityOverride=true
    gravityScale=0.5
    simulationSpace=BoneRelative
    alphaBoolBlend {blendOption=Linear}
    boundBone {boneName=dyn_sleeve_l_01}
    componentAppliedLinearAccClamp=<5000 5000 5000>
    componentLinearAccScale=<0.2 0.2 0.4>
    componentLinearVelScale=<0.5 0.5 0.5>
    gravityOverride {z=-980}
    physicsBodyDefinitions [{
      sphereCollisionRadius=0
      boundBone {boneName=dyn_sleeve_l_01}
      boxExtents=<10 20 20>
      constraintSetup {angularLimitsMax=<1 3 15>, angularLimitsMin=<-1 -30 -15>, angularTarget=<0 0 0>}
      localJointOffset=<5 0 0>
     }]
    relativeSpaceBone {boneName=root}
   }
   n170 = Anim Dynamics - Bone: dyn_sleeve_l_02(ComponentPose=@n169, Alpha=1, ExternalForce="0,0,0") {
    angularDampingOverride=0.9999
    angularSpringConstant=500
    bAngularSpring=true
    bGravityOverrideInSimSpace=true
    bOverrideAngularDamping=true
    bUseGravityOverride=true
    gravityScale=0.5
    simulationSpace=BoneRelative
    alphaBoolBlend {blendOption=Linear}
    boundBone {boneName=dyn_sleeve_l_02}
    componentAppliedLinearAccClamp=<5000 5000 5000>
    componentLinearAccScale=<0.2 0.2 0.4>
    componentLinearVelScale=<0.5 0.5 0.5>
    gravityOverride {z=-980}
    physicsBodyDefinitions [{
      sphereCollisionRadius=0
      boundBone {boneName=dyn_sleeve_l_02}
      boxExtents=<10 20 20>
      constraintSetup {angularLimitsMax=<1 3 15>, angularLimitsMin=<-1 -30 -15>, angularTarget=<0 0 0>}
      localJointOffset=<5 0 0>
     }]
    relativeSpaceBone {boneName=root}
   }
   n171 = Anim Dynamics - Bone: dyn_sleeve_l_03(ComponentPose=@n170, Alpha=1, ExternalForce="0,0,0") {
    angularDampingOverride=0.9999
    angularSpringConstant=500
    bAngularSpring=true
    bGravityOverrideInSimSpace=true
    bOverrideAngularDamping=true
    bUseGravityOverride=true
    gravityScale=0.5
    simulationSpace=BoneRelative
    alphaBoolBlend {blendOption=Linear}
    boundBone {boneName=dyn_sleeve_l_03}
    componentAppliedLinearAccClamp=<3500 3500 3500>
    componentLinearAccScale=<0.2 0.2 0.4>
    componentLinearVelScale=<0.5 0.5 0.5>
    gravityOverride {z=-980}
    physicsBodyDefinitions [{
      sphereCollisionRadius=0
      boundBone {boneName=dyn_sleeve_l_03}
      boxExtents=<10 20 20>
      constraintSetup {angularLimitsMax=<1 3 15>, angularLimitsMin=<-1 -30 -15>, angularTarget=<0 0 0>}
      localJointOffset=<5 0 0>
     }]
    relativeSpaceBone {boneName=root}
   }
   n172 = Anim Dynamics - Bone: dyn_sleeve_r_01(ComponentPose=@n171, Alpha=1, ExternalForce="0,0,0") {
    angularDampingOverride=0.9999
    angularSpringConstant=500
    bAngularSpring=true
    bGravityOverrideInSimSpace=true
    bOverrideAngularDamping=true
    bUseGravityOverride=true
    gravityScale=0.5
    simulationSpace=BoneRelative
    alphaBoolBlend {blendOption=Linear}
    boundBone {boneName=dyn_sleeve_r_01}
    componentAppliedLinearAccClamp=<5000 5000 5000>
    componentLinearAccScale=<0.2 0.2 0.4>
    componentLinearVelScale=<0.5 0.5 0.5>
    gravityOverride {z=-980}
    physicsBodyDefinitions [{
      sphereCollisionRadius=0
      boundBone {boneName=dyn_sleeve_r_01}
      boxExtents=<10 20 20>
      constraintSetup {angularLimitsMax=<1 3 15>, angularLimitsMin=<-1 -30 -15>, angularTarget=<0 0 0>}
      localJointOffset=<5 0 0>
     }]
    relativeSpaceBone {boneName=root}
   }
   n173 = Anim Dynamics - Bone: dyn_sleeve_r_02(ComponentPose=@n172, Alpha=1, ExternalForce="0,0,0") {
    angularDampingOverride=0.9999
    angularSpringConstant=500
    bAngularSpring=true
    bGravityOverrideInSimSpace=true
    bOverrideAngularDamping=true
    bUseGravityOverride=true
    gravityScale=0.5
    simulationSpace=BoneRelative
    alphaBoolBlend {blendOption=Linear}
    boundBone {boneName=dyn_sleeve_r_02}
    componentAppliedLinearAccClamp=<5000 5000 5000>
    componentLinearAccScale=<0.2 0.2 0.4>
    componentLinearVelScale=<0.5 0.5 0.5>
    gravityOverride {z=-980}
    physicsBodyDefinitions [{
      sphereCollisionRadius=0
      boundBone {boneName=dyn_sleeve_r_02}
      boxExtents=<10 20 20>
      constraintSetup {angularLimitsMax=<1 3 15>, angularLimitsMin=<-1 -30 -15>, angularTarget=<0 0 0>}
      localJointOffset=<5 0 0>
     }]
    relativeSpaceBone {boneName=root}
   }
   n174 = Anim Dynamics - Bone: dyn_sleeve_r_03(ComponentPose=@n173, Alpha=1, ExternalForce="0,0,0") {
    angularDampingOverride=0.9999
    angularSpringConstant=500
    bAngularSpring=true
    bGravityOverrideInSimSpace=true
    bOverrideAngularDamping=true
    bUseGravityOverride=true
    gravityScale=0.5
    simulationSpace=BoneRelative
    alphaBoolBlend {blendOption=Linear}
    boundBone {boneName=dyn_sleeve_r_03}
    componentAppliedLinearAccClamp=<3500 3500 3500>
    componentLinearAccScale=<0.2 0.2 0.4>
    componentLinearVelScale=<0.5 0.5 0.5>
    gravityOverride {z=-980}
    physicsBodyDefinitions [{
      sphereCollisionRadius=0
      boundBone {boneName=dyn_sleeve_r_03}
      boxExtents=<10 20 20>
      constraintSetup {angularLimitsMax=<1 3 15>, angularLimitsMin=<-1 -30 -15>, angularTarget=<0 0 0>}
      localJointOffset=<5 0 0>
     }]
    relativeSpaceBone {boneName=root}
   }
   n175 = Anim Dynamics - Bone: dyn_zipper_01(ComponentPose=@n174, Alpha=1, ExternalForce="0,0,0") {
    angularDampingOverride=0.9999
    angularSpringConstant=500
    bAngularSpring=true
    bGravityOverrideInSimSpace=true
    bOverrideAngularDamping=true
    simulationSpace=BoneRelative
    alphaBoolBlend {blendOption=Linear}
    boundBone {boneName=dyn_zipper_01}
    componentAppliedLinearAccClamp=<5000 5000 5000>
    componentLinearAccScale=<1.2 1.2 1.2>
    componentLinearVelScale=<1.2 1.2 1.2>
    physicsBodyDefinitions [{
      sphereCollisionRadius=0
      boundBone {boneName=dyn_zipper_01}
      boxExtents=<15 20 20>
      constraintSetup {angularLimitsMax=<1 5 30>, angularLimitsMin=<-1 -20 -35>, angularTarget=<0 0 0>}
      localJointOffset=<2 0 0>
     }]
    relativeSpaceBone {boneName=root}
   }
   n176 = Anim Dynamics - Bone: dyn_zipper_02(ComponentPose=@n175, Alpha=1, ExternalForce="0,0,0") {
    angularDampingOverride=0.9999
    angularSpringConstant=500
    bAngularSpring=true
    bGravityOverrideInSimSpace=true
    bOverrideAngularDamping=true
    simulationSpace=BoneRelative
    alphaBoolBlend {blendOption=Linear}
    boundBone {boneName=dyn_zipper_02}
    componentAppliedLinearAccClamp=<5000 5000 5000>
    componentLinearAccScale=<1.2 1.2 1.2>
    componentLinearVelScale=<1.2 1.2 1.2>
    physicsBodyDefinitions [{
      sphereCollisionRadius=0
      boundBone {boneName=dyn_zipper_02}
      boxExtents=<10 20 20>
      constraintSetup {angularLimitsMax=<1 5 45>, angularLimitsMin=<-1 -10 -45>, angularTarget=<0 0 0>}
      localJointOffset=<10 0 0>
     }]
    relativeSpaceBone {boneName=root}
   }
   Save cached pose 'Pose'(Pose=Component To Local(ComponentPose=@n176)) {cacheName=Pose}
  ]}
 EventGraph {nodes [
   Event BlueprintUpdateAnimation()
   > n1 = Sequence!(execute) -> (then_0, then_1, then_2:Exec, then_3:Exec, then_4:Exec, then_5:Exec)
   > Branch(Condition=OR Boolean(A=Get bIsSkyDiving(), B=Get bIsInFrontEnd()))
   n2 = Branch(execute=@n1.then_1, Condition=OR Boolean(A=Get bIsJumping(), B=Get bIsFalling()))
   > Set JumpingFalling(JumpingFalling:bool=true) -> (Output_Get:bool)
   Set JumpingFalling(execute=@n2.else, JumpingFalling=false)
   n3 = Get PawnSpeed2D()
   Set run_alpha(execute=@n1.then_2, run_alpha:double=SelectFloat(A=1, B=0, bPickA="float > float"(A=@n3, B=500) {is=title K2Node_CallFunction})) -> (Output_Get:double)
   n4 = NOT Boolean(A=Get bIsMoving2D())
   n5 = Get bIsCrouching()
   > Set Crouch_idle(Crouch_idle:double=SelectFloat(A=1, B=0, bPickA=AND Boolean(A=@n5, B=@n4))) -> (Output_Get:double)
   > Set is crouch(is crouch:double=SelectFloat(A=1, B=0, bPickA=Get bIsCrouching())) -> (Output_Get:double)
   > Set Crouch off(Crouch off:double=SelectFloat(A=0, B=1, bPickA=Get bIsCrouching())) -> (Output_Get:double)
   > Set Parachute Open(Parachute Open:double=SelectFloat(A=0.5, B=1, bPickA=Get bIsParachuteOpened())) -> (Output_Get:double)
   n6 = NOT Boolean(A=Get bIsPlayingEmote())
   n7 = Get bIsInFrontEnd()
   > Set Emote off(Emote off:double=SelectFloat(A=0, B=1, bPickA=AND Boolean(A=@n7, B=@n6))) -> (Output_Get:double)
   n8 = BreakTransform(InTransform=GetDeltaTransformFromRefPose(self=GetOwnerSkeletalMeshComponent(), BoneName=thigh_l, BaseName=pelvis))
   Set thigh_rot_l(execute=@n1.then_3, thigh_rot_l=MapRangeClamped(Value=BreakRotator(InRot=@n8.Rotation).Yaw, InRangeA=-15, InRangeB=-100))
   n9 = BreakTransform(InTransform=GetDeltaTransformFromRefPose(self=GetOwnerSkeletalMeshComponent(), BoneName=thigh_r, BaseName=pelvis))
   > Set thigh_rot_r(thigh_rot_r=MapRangeClamped(Value=BreakRotator(InRot=@n9.Rotation).Yaw, InRangeA=-15, InRangeB=-100))
   n10 = BreakTransform(InTransform=GetDeltaTransformFromRefPose(self=GetOwnerSkeletalMeshComponent(), BoneName=upperarm_l, BaseName=spine_05))
   > Set L_arm_up(L_arm_up:double=MapRangeClamped(Value=BreakRotator(InRot=@n10.Rotation).Yaw, InRangeA=70, InRangeB=110)) -> (Output_Get:double)
   n11 = GetOwnerSkeletalMeshComponent()
   n12 = BreakRotator(InRot=BreakTransform(InTransform=GetDeltaTransformFromRefPose(self=@n11, BoneName=upperarm_r, BaseName=spine_05)).Rotation)
   > Set R_arm_up(R_arm_up:double=SelectFloat(A=MapRangeClamped(Value=@n12.Yaw, InRangeA=38, InRangeB=136), B=0, bPickA=Get bIsTargeting())) -> (Output_Get:double)
   > Set R_Arm_Down(R_Arm_Down:double=SelectFloat(A=MapRangeClamped(Value=@n12.Yaw, InRangeA=120, InRangeB=2), B=0, bPickA=Get bIsTargeting())) -> (Output_Get:double)
   n13 = Get PawnMovementDirectionAzimuth() -> (PawnMovementDirectionAzimuth:float)
   n14 = float - float(A=MapRangeClamped(Value=@n13, InRangeA=45, InRangeB=90), B=MapRangeClamped(Value=@n13, InRangeA=90, InRangeB=135)) {is=title K2Node_CallFunction}
   n15 = Get bIsMoving2D()
   > Set Right_Direct(Right_Direct:double=SelectFloat(A="Absolute (Float)"(A=@n14), B=0, bPickA=@n15)) -> (Output_Get:double)
   n16 = Get PawnMovementDirectionAzimuth()
   n17 = float - float(A=MapRangeClamped(Value=@n16, InRangeA=-45, InRangeB=-90), B=MapRangeClamped(Value=@n16, InRangeA=-90, InRangeB=-135))
   n18 = Get bIsMoving2D()
   > Set Left_Direct(Left_Direct:double=SelectFloat(A="Absolute (Float)"(A=@n17), B=0, bPickA=@n18)) -> (Output_Get:double)
   n19 = Get PawnMovementDirectionAzimuth()
   n20 = float + float(
    A=MapRangeClamped(Value=@n19, InRangeA=-135, InRangeB=-180, OutRangeB=0.75)
    B=MapRangeClamped(Value=@n19, InRangeA=135, InRangeB=180, OutRangeB=0.75)
   ) {is=title K2Node_CommutativeAssociativeBinaryOperator}
   n21 = Get bIsMoving2D()
   > Set Backward_Direct(Backward_Direct:double=SelectFloat(A="Absolute (Float)"(A=@n20), B=0, bPickA=@n21)) -> (Output_Get:double)
   Event OnCharacterCustomizationLoaded()
   n22 = CheckForBackbling()
   > Set isWearingBackbling(isWearingBackbling=@n22)
   > Set isWearingCape(isWearingCape=@n22.bIsBackblingLong)
   AND Boolean(A=Get bIsCrouching(), B=@n15)
   AND Boolean(A=Get bIsCrouching(), B=@n21)
   AND Boolean(A=Get bIsCrouching(), B=@n18)
  ]}
]
variables [
 Backward_Direct {type=number}
 Crouch off {type=number}
 Crouch_idle {type=number}
 Emote off {type=number}
 is crouch {type=number}
 isWearingBackbling {type=boolean}
 isWearingCape {type=boolean}
 JumpingFalling {type=boolean}
 L_arm_up {type=number}
 Left_Direct {type=number}
 Parachute Open {type=number}
 R_Arm_Down {type=number}
 R_arm_up {type=number}
 Right_Direct {type=number}
 run_alpha {type=number}
 Skel_Mesh {type=SkeletalMeshComponent}
 thigh_rot_l {type=number}
 thigh_rot_r {type=number}
]
